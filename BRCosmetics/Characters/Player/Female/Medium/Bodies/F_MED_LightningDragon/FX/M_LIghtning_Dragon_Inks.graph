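# Translucent unlit Niagara sprite material for the ink effect on the female Lightning Dragon player cosmetic.
# An ink texture, boosted and saturated, has the per-particle Dissolve parameter subtracted, then is multiplied by a noise texture and particle alpha for opacity.
# A UV-space sphere mask softens sprite edges, a world-space mask limits opacity around particle position, and depth fade softens intersections.
# Particle color feeds emissive through a shared adjustment function.
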Material /BRCosmetics/Characters/Player/Female/Medium/Bodies/F_MED_LightningDragon/FX/M_LIghtning_Dragon_Inks
graph {
 properties {bUsedWithNiagaraSprites=true, blendMode=BLEND_Translucent, pixelDepthOffsetMode=PDOM_Legacy, shadingModel=MSM_Unlit, usedShadingModels=Unlit}
 nodes [
  n1 = WorldPosition()
  n2 = ParticlePositionWS()
  n3 = ScalarParameter() {defaultValue=300, parameterName=Radius}
  n4 = ScalarParameter() {defaultValue=50, parameterName=Hardness}
  n5 = ParticleColor()
  n6 = TextureSampleParameter2D(UVs=TextureCoordinate()) {parameterName=Base Tex, texture=&/T_LIghtning_Dragon_Inks}
  n7 = Saturate(None=Multiply(A=@n6.R) {constB=2})
  n8 = DynamicParameter() {defaultValue=<rgba 0 1 1 1>, paramNames [Dissolve,Param2,Param3,Param4]}
  n9 = Multiply(
   A=@n5.A
   B=Multiply(
    A=Subtract(A=@n7, B=@n8.Dissolve)
    B=TextureSampleParameter2D() {parameterName=Noise Tex, samplerType=SAMPLERTYPE_Grayscale, texture=&LowResBlurredNoise}.R
   )
  )
  n10 = SphereMask(A=TextureCoordinate(), B=Constant2Vector() {g=0.5, r=0.5}) {attenuationRadius=0.5, hardnessPercent=0}
  DepthFade(
   Opacity=Multiply(A=Multiply(A=@n9, B=@n10), B=SphereMask(A=@n1.XYZ, B=@n2, Radius=@n3, Hardness=@n4) {attenuationRadius=300, hardnessPercent=50})
   FadeDistance=ScalarParameter() {defaultValue=4, parameterName=FadeDistance}
  ) {fadeDistanceDefault=5}
  MaterialFunctionCall(Emissive=@n5.RGB) {materialFunction=&MF_FX_EmissiveAdjustments}
 ]
}
