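# Deformation Control Rig for the F_MED_LightningDragon character body that drives the limb twist bones.
# Upper arm and thigh twist bones counter-rotate by scaled X-axis swing-twist angles from their parent bones, with thighs measured against the initial pose.
# Forearm and calf 01 twist bones follow hand and foot through weighted X-rotation-only parent constraints, and each 02 bone takes half that twist.
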ControlRigBlueprint /BRCosmetics/Characters/Player/Female/Medium/Bodies/F_MED_LightningDragon/Meshes/F_MED_LightningDragon_Deform_CtrlRig_twists
graphs [
 RigVMFunctionLibrary {links [], nodes []}
 RigVMModel {nodes [
   RigUnit_BeginExecution = &RigUnit_BeginExecution()
   n1 = /Script/RigVM.RigVMFunction_MathQuaternionToAxisAndAngle(Value=&RigVMFunction_MathQuaternionSwingTwist(
     TwistAxis=(X=1,Y=0,Z=0)
     Input=&RigUnit_GetTransform(Item=(Type=Bone,Name="upperarm_l"), Space=LocalSpace).Transform.Rotation
    ).Twist) {is=struct RigVMUnitNode, template="ToAxisAndAngle::Execute(in Value,out Axis,out Angle)"}
   RerouteNode_2_1 = RigVMDispatchNode(Value=-0.8) {template="DISPATCH_RigVMDispatch_Constant(io Value)"}
   RigUnit_SetRotation = &RigUnit_SetRotation(
    Item=(Type=Bone,Name="upperarm_twist_01_l")
    Weight=1
    Value=/Script/RigVM.RigVMFunction_MathQuaternionFromAxisAndAngle(
     Angle=&RigVMFunction_MathFloatMul(A=@n1.Angle, B=@RerouteNode_2_1.Value) {template="Multiply::Execute(in A,in B,out Result)"}.Result
     Axis=@n1.Axis
    ) {is=struct RigVMUnitNode, template="FromAxisAndAngle::Execute(in Axis,in Angle,out Result)"}.Result
    ExecutePin=@RigUnit_BeginExecution.ExecutePin
   )
   RerouteNode_2 = RigVMDispatchNode(Value=-0.5) {template="DISPATCH_RigVMDispatch_Constant(io Value)"}
   > &RigUnit_SetRotation(
    Item=(Type=Bone,Name="upperarm_twist_02_l")
    Weight=1
    Value=&RigVMFunction_MathQuaternionFromAxisAndAngle(
     Angle=&RigVMFunction_MathFloatMul(A=@n1.Angle, B=@RerouteNode_2.Value) {template="Multiply::Execute(in A,in B,out Result)"}.Result
     Axis=@n1.Axis
    ) {template="FromAxisAndAngle::Execute(in Axis,in Angle,out Result)"}.Result
   )
   n2 = &RigVMFunction_MathQuaternionToAxisAndAngle(Value=&RigVMFunction_MathQuaternionSwingTwist(
     TwistAxis=(X=1,Y=0,Z=0)
     Input=&RigUnit_GetTransform(Item=(Type=Bone,Name="upperarm_r"), Space=LocalSpace).Transform.Rotation
    ).Twist) {template="ToAxisAndAngle::Execute(in Value,out Axis,out Angle)"}
   > &RigUnit_SetRotation(
    Item=(Type=Bone,Name="upperarm_twist_01_r")
    Weight=1
    Value=&RigVMFunction_MathQuaternionFromAxisAndAngle(
     Angle=&RigVMFunction_MathFloatMul(A=@n2.Angle, B=@RerouteNode_2_1.Value) {template="Multiply::Execute(in A,in B,out Result)"}.Result
     Axis=@n2.Axis
    ) {template="FromAxisAndAngle::Execute(in Axis,in Angle,out Result)"}.Result
   )
   > &RigUnit_SetRotation(
    Item=(Type=Bone,Name="upperarm_twist_02_r")
    Weight=1
    Value=&RigVMFunction_MathQuaternionFromAxisAndAngle(
     Angle=&RigVMFunction_MathFloatMul(A=@n2.Angle, B=@RerouteNode_2.Value) {template="Multiply::Execute(in A,in B,out Result)"}.Result
     Axis=@n2.Axis
    ) {template="FromAxisAndAngle::Execute(in Axis,in Angle,out Result)"}.Result
   )
   RigUnit_GetTransform = &RigUnit_GetTransform(Item=(Type=Bone,Name="thigh_l"), Space=LocalSpace)
   n3 = &RigUnit_GetTransform(Item=(Type=Bone,Name="thigh_l"), Space=LocalSpace, bInitial=true)
   n4 = &RigVMFunction_MathQuaternionToAxisAndAngle(Value=&RigVMFunction_MathQuaternionSwingTwist(
     TwistAxis=(X=1,Y=0,Z=0)
     Input=&RigVMFunction_MathTransformMakeRelative(Global=@RigUnit_GetTransform.Transform, Parent=@n3.Transform).Local.Rotation
    ).Twist) {template="ToAxisAndAngle::Execute(in Value,out Axis,out Angle)"}
   > &RigUnit_SetRotation(
    Item=(Type=Bone,Name="thigh_twist_01_l")
    Weight=1
    Value=&RigVMFunction_MathQuaternionFromAxisAndAngle(Angle=&RigVMFunction_MathFloatMul(B=-1, A=@n4.Angle) {template="Multiply::Execute(in A,in B,out Result)"}.Result, Axis=@n4.Axis) {template="FromAxisAndAngle::Execute(in Axis,in Angle,out Result)"}.Result
   )
   n5 = &RigUnit_GetTransform(Item=(Type=Bone,Name="thigh_r"), Space=LocalSpace, bInitial=true)
   n6 = &RigUnit_GetTransform(Item=(Type=Bone,Name="thigh_r"), Space=LocalSpace)
   n7 = &RigVMFunction_MathQuaternionToAxisAndAngle(Value=&RigVMFunction_MathQuaternionSwingTwist(
     TwistAxis=(X=1,Y=0,Z=0)
     Input=&RigVMFunction_MathTransformMakeRelative(Parent=@n5.Transform, Global=@n6.Transform).Local.Rotation
    ).Twist) {template="ToAxisAndAngle::Execute(in Value,out Axis,out Angle)"}
   > &RigUnit_SetRotation(
    Item=(Type=Bone,Name="thigh_twist_01_r")
    Weight=1
    Value=&RigVMFunction_MathQuaternionFromAxisAndAngle(Angle=&RigVMFunction_MathFloatMul(B=-1, A=@n7.Angle) {template="Multiply::Execute(in A,in B,out Result)"}.Result, Axis=@n7.Axis) {template="FromAxisAndAngle::Execute(in Axis,in Angle,out Result)"}.Result
   )
   > &RigUnit_ParentConstraint(
    AdvancedSettings=(RotationOrderForFilter=ZYX)
    Child=(Type=Bone,Name="lowerarm_twist_01_l")
    Filter=(TranslationFilter=(bX=false,bY=false,bZ=false),RotationFilter=(bX=True,bY=false,bZ=false),ScaleFilter=(bX=false,bY=false,bZ=false))
    Parents=((Item=(Type=Bone,Name="hand_l"),Weight=0.75),(Item=(Type=Bone,Name="lowerarm_twist_01_l"),Weight=0.25))
    Weight=1
   )
   n8 = &RigVMFunction_MathQuaternionToEuler(
    RotationOrder=ZYX
    Value=&RigUnit_GetTransform(Item=(Type=Bone,Name="lowerarm_twist_01_l"), Space=LocalSpace).Transform.Rotation
   )
   > &RigUnit_SetRotation(
    Item=(Type=Bone,Name="lowerarm_twist_02_l")
    Weight=1
    Value=&RigVMFunction_MathQuaternionFromEuler(
     Euler=(X=0,Y=0,Z=0)
     Euler.X=&RigVMFunction_MathFloatMul(B=0.5, A=@n8.Result.X) {template="Multiply::Execute(in A,in B,out Result)"}.Result
    ).Result
   )
   > &RigUnit_ParentConstraint(
    AdvancedSettings=(RotationOrderForFilter=ZYX)
    Child=(Type=Bone,Name="lowerarm_twist_01_r")
    Filter=(TranslationFilter=(bX=false,bY=false,bZ=false),RotationFilter=(bX=True,bY=false,bZ=false),ScaleFilter=(bX=false,bY=false,bZ=false))
    Parents=((Item=(Type=Bone,Name="hand_r"),Weight=0.75),(Item=(Type=Bone,Name="lowerarm_twist_01_r"),Weight=0.25))
    Weight=1
   )
   n9 = &RigVMFunction_MathQuaternionToEuler(
    RotationOrder=ZYX
    Value=&RigUnit_GetTransform(Item=(Type=Bone,Name="lowerarm_twist_01_r"), Space=LocalSpace).Transform.Rotation
   )
   > &RigUnit_SetRotation(
    Item=(Type=Bone,Name="lowerarm_twist_02_r")
    Weight=1
    Value=&RigVMFunction_MathQuaternionFromEuler(
     Euler=(X=0,Y=0,Z=0)
     Euler.X=&RigVMFunction_MathFloatMul(B=0.5, A=@n9.Result.X) {template="Multiply::Execute(in A,in B,out Result)"}.Result
    ).Result
   )
   > &RigUnit_ParentConstraint(
    AdvancedSettings=(RotationOrderForFilter=ZYX)
    Child=(Type=Bone,Name="calf_twist_01_l")
    Filter=(TranslationFilter=(bX=false,bY=false,bZ=false),RotationFilter=(bX=True,bY=false,bZ=false),ScaleFilter=(bX=false,bY=false,bZ=false))
    Parents=((Item=(Type=Bone,Name="foot_l"),Weight=0.75),(Item=(Type=Bone,Name="calf_twist_01_l"),Weight=0.25))
    Weight=1
   )
   n10 = &RigVMFunction_MathQuaternionToEuler(
    RotationOrder=ZYX
    Value=&RigUnit_GetTransform(Item=(Type=Bone,Name="calf_twist_01_l"), Space=LocalSpace).Transform.Rotation
   )
   > &RigUnit_SetRotation(
    Item=(Type=Bone,Name="calf_twist_02_l")
    Weight=1
    Value=&RigVMFunction_MathQuaternionFromEuler(
     Euler=(X=0,Y=0,Z=0)
     Euler.X=&RigVMFunction_MathFloatMul(B=0.5, A=@n10.Result.X) {template="Multiply::Execute(in A,in B,out Result)"}.Result
    ).Result
   )
   > &RigUnit_ParentConstraint(
    AdvancedSettings=(RotationOrderForFilter=ZYX)
    Child=(Type=Bone,Name="calf_twist_01_r")
    Filter=(TranslationFilter=(bX=false,bY=false,bZ=false),RotationFilter=(bX=True,bY=false,bZ=false),ScaleFilter=(bX=false,bY=false,bZ=false))
    Parents=((Item=(Type=Bone,Name="foot_r"),Weight=0.75),(Item=(Type=Bone,Name="calf_twist_01_r"),Weight=0.25))
    Weight=1
   )
   n11 = &RigVMFunction_MathQuaternionToEuler(
    RotationOrder=ZYX
    Value=&RigUnit_GetTransform(Item=(Type=Bone,Name="calf_twist_01_r"), Space=LocalSpace).Transform.Rotation
   )
   > &RigUnit_SetRotation(
    Item=(Type=Bone,Name="calf_twist_02_r")
    Weight=1
    Value=&RigVMFunction_MathQuaternionFromEuler(
     Euler=(X=0,Y=0,Z=0)
     Euler.X=&RigVMFunction_MathFloatMul(B=0.5, A=@n11.Result.X) {template="Multiply::Execute(in A,in B,out Result)"}.Result
    ).Result
   )
   > /Script/ControlRig.RigUnit_DebugHierarchy(
    Color=(R=1,G=1,B=1,A=1)
    DebugDrawSettings=(DepthPriority=SDPG_Foreground,Lifetime=-1)
    Items=()
    Scale=5
    Thickness=0.2
    WorldOffset=(Rotation=(X=0,Y=0,Z=0,W=1),Translation=(X=0,Y=0,Z=0),Scale3D=(X=1,Y=1,Z=1))
    bEnabled=false
   ) {is=struct RigVMUnitNode}
   RigVMVariableNode() {variable=Upperarm_Twist}
  ]}
]
hierarchy [
 Root {loc=<0 0 0>, rot=<0 0 0 1>}
 pelvis {parent=Root, loc=<0 -0.5301 98.0571>, rot=<0.0127 0.707 -0.0127 -0.707>}
 Spine_01 {parent=pelvis, loc=<6.1208 -0.9328 0>, rot=<0 0 0.034 -0.9994>}
 Spine_02 {parent=Spine_01, loc=<6.5996 -0.2095 0>, rot=<0 0 0.0238 -0.9997>}
 Spine_03 {parent=Spine_02, loc=<6.4914 -0.1237 0>, rot=<0 0 0.0149 -0.9999>}
 spine_04 {parent=Spine_03, loc=<6.9598 0.0794 0>, rot=<0 0 -0.0132 -0.9999>}
 spine_05 {parent=spine_04, loc=<6.9987 0 0>, rot=<0 0 -0.0015 -1>}
 clavicle_l {parent=spine_05, loc=<5.4831 -0.1053 -3.2946>, rot=<0.0465 -0.7289 -0.009 -0.6829>}
 upperarm_l {parent=clavicle_l, loc=<9.8658 -1.2136 -0.4105>, rot=<0.0164 -0.3245 -0.0269 -0.9454>}
 lowerarm_l {parent=upperarm_l, loc=<26.465 -0.8345 0.259>, rot=<0.0036 0.0161 -0.2298 -0.9731>}
 hand_l {parent=lowerarm_l, loc=<22.7355 -0.3203 -0.6105>, rot=<-0.6784 -0.0223 0.0735 -0.7307>}
 index_metacarpal_l {parent=hand_l, loc=<1.6535 0.0183 -2.96>, rot=<0.0158 -0.0124 -0.0173 -0.9996>}
 index_01_l {parent=index_metacarpal_l, loc=<8.3008 0.0357 -0.0524>, rot=<-0.0044 0.0485 0.0924 -0.9945>}
 index_02_l {parent=index_01_l, loc=<3.3587 -0.3003 0.0452>, rot=<-0.002 0.0027 0.3225 -0.9466>}
 index_03_l {parent=index_02_l, loc=<2.5578 -0.123 -0.0252>, rot=<-0.002 0.0067 0.2558 -0.9667>}
 middle_metacarpal_l {parent=hand_l, loc=<2.1856 -0.1725 -0.5984>, rot=<0.006 0.0168 -0.0329 -0.9993>}
 middle_01_l {parent=middle_metacarpal_l, loc=<7.3275 -0.2447 0.0129>, rot=<-0.0208 0.0042 0.0858 -0.9961>}
 middle_02_l {parent=middle_01_l, loc=<4.0123 0.1401 0.1175>, rot=<-0.0218 0.0258 0.3512 -0.9357>}
 middle_03_l {parent=middle_02_l, loc=<2.6077 0.0534 0.1373>, rot=<0.0082 0.0188 0.1708 -0.9851>}
 pinky_metacarpal_l {parent=hand_l, loc=<1.7616 -0.9753 3.3814>, rot=<-0.0397 0.0814 -0.069 -0.9935>}
 pinky_01_l {parent=pinky_metacarpal_l, loc=<7.1709 -0.4831 -0.1094>, rot=<-0.0654 -0.0035 0.2242 -0.9723>}
 pinky_02_l {parent=pinky_01_l, loc=<2.8274 0.0893 -0.1043>, rot=<0.0037 0.0011 0.2854 -0.9584>}
 pinky_03_l {parent=pinky_02_l, loc=<2.2046 0.0308 -0.0078>, rot=<-0.0002 0.0098 0.1445 -0.9895>}
 ring_metacarpal_l {parent=hand_l, loc=<2.2074 -0.3625 1.3638>, rot=<-0.0284 0.0722 -0.0458 -0.9959>}
 ring_01_l {parent=ring_metacarpal_l, loc=<7.0295 -0.6197 -0.1446>, rot=<-0.0489 -0.0252 0.1892 -0.9804>}
 ring_02_l {parent=ring_01_l, loc=<3.6174 0.4647 -0.1027>, rot=<0 0 0.2806 -0.9598>}
 ring_03_l {parent=ring_02_l, loc=<2.4152 0.1141 0.0199>, rot=<0 0 0.1189 -0.9929>}
 thumb_01_l {parent=hand_l, loc=<1.3118 -0.3191 -2.459>, rot=<0.6546 -0.4042 0.0664 -0.6354>}
 thumb_02_l {parent=thumb_01_l, loc=<5.9336 -0.0448 0.0401>, rot=<-0.0037 0.0213 0.298 -0.9543>}
 thumb_03_l {parent=thumb_02_l, loc=<2.2223 0.3682 -0.0907>, rot=<0.0118 -0.013 0.0066 -0.9998>}
 lowerarm_twist_01_l {parent=lowerarm_l, loc=<18.316 0 0>, rot=<0 0 0 1>}
 lowerarm_twist_02_l {parent=lowerarm_l, loc=<10.6707 0 0>, rot=<0 0 0 1>}
 dyn_sleeves_l_MSTR {parent=lowerarm_l, loc=<-2.5828 3.0383 -1.1208>, rot=<0.2065 -0.6801 0.32 -0.6264>}
 dyn_sleeves_l_up {parent=dyn_sleeves_l_MSTR, loc=<-7.569 -0.0305 2.6618>, rot=<0.0012 0.7379 0.0033 -0.6749>}
 dyn_sleeves_l_dn {parent=dyn_sleeves_l_MSTR, loc=<7.4403 0.6549 2.4233>, rot=<0.7326 0.0016 0.6806 0.0008>}
 dyn_sleeves_l_ft {parent=dyn_sleeves_l_MSTR, loc=<-1.838 5.7907 3.4408>, rot=<-0.5146 -0.5202 -0.4846 0.4793>}
 dyn_sleeves_l_bk {parent=dyn_sleeves_l_MSTR, loc=<-0.9941 -6.2344 1.6936>, rot=<0.5216 -0.5091 0.4782 0.4899>}
 deform_elbow_in_base_l {parent=lowerarm_l, loc=<0.1834 3.6062 -0.2572>, rot=<0 0 0 -1>}
 deform_elbow_in_l {parent=deform_elbow_in_base_l, loc=<0 0 0>, rot=<0 0 0 1>}
 deform_elbow_out_base_l {parent=lowerarm_l, loc=<0 -4 0>, rot=<0 0 0 1>}
 deform_elbow_out_l {parent=deform_elbow_out_base_l, loc=<0 0 0>, rot=<0 0 0 1>}
 upperarm_twist_01_l {parent=upperarm_l, loc=<7.2135 -0.1134 0>, rot=<0 0 0 -1>}
 dyn_arm_straps_l_01 {parent=upperarm_twist_01_l, loc=<1.5871 -6.6585 1.3646>, rot=<-0.6406 0.1797 0.008 -0.7465>}
 dyn_arm_straps_l_02 {parent=upperarm_twist_01_l, loc=<-0.6618 -0.0932 6.3869>, rot=<-0.0075 0.1877 0.0049 -0.9822>}
 dyn_arm_straps_l_03 {parent=upperarm_twist_01_l, loc=<0.6848 6.2838 3.1454>, rot=<0.4263 0.1067 0.0361 -0.8975>}
 upperarm_twist_02_l {parent=upperarm_l, loc=<19.7998 -0.1134 0>, rot=<0 0 0 -1>}
 clavicle_r {parent=spine_05, loc=<5.4831 -0.1053 3.2946>, rot=<0.7289 0.0465 -0.6829 0.009>}
 upperarm_r {parent=clavicle_r, loc=<-9.8658 1.2136 0.4104>, rot=<0.0164 -0.3245 -0.0269 -0.9454>}
 lowerarm_r {parent=upperarm_r, loc=<-26.4651 0.8345 -0.259>, rot=<0.0036 0.0161 -0.2298 -0.9731>}
 hand_r {parent=lowerarm_r, loc=<-22.7356 0.3203 0.6106>, rot=<-0.6784 -0.0223 0.0735 -0.7307>}
 index_metacarpal_r {parent=hand_r, loc=<-1.6535 -0.0183 2.96>, rot=<0.0158 -0.0124 -0.0173 -0.9996>}
 index_01_r {parent=index_metacarpal_r, loc=<-8.3008 -0.0357 0.0524>, rot=<-0.0044 0.0485 0.0924 -0.9945>}
 index_02_r {parent=index_01_r, loc=<-3.3587 0.3004 -0.0452>, rot=<-0.002 0.0027 0.3225 -0.9466>}
 index_03_r {parent=index_02_r, loc=<-2.5578 0.123 0.0252>, rot=<-0.002 0.0067 0.2558 -0.9667>}
 middle_metacarpal_r {parent=hand_r, loc=<-2.1856 0.1726 0.5984>, rot=<0.006 0.0168 -0.0329 -0.9993>}
 middle_01_r {parent=middle_metacarpal_r, loc=<-7.3275 0.2447 -0.0129>, rot=<-0.0208 0.0042 0.0858 -0.9961>}
 middle_02_r {parent=middle_01_r, loc=<-4.0123 -0.1401 -0.1175>, rot=<-0.0218 0.0258 0.3512 -0.9357>}
 middle_03_r {parent=middle_02_r, loc=<-2.6077 -0.0533 -0.1373>, rot=<0.0082 0.0188 0.1708 -0.9851>}
 pinky_metacarpal_r {parent=hand_r, loc=<-1.7616 0.9753 -3.3814>, rot=<-0.0397 0.0814 -0.069 -0.9935>}
 pinky_01_r {parent=pinky_metacarpal_r, loc=<-7.1709 0.4831 0.1094>, rot=<-0.0654 -0.0035 0.2242 -0.9723>}
 pinky_02_r {parent=pinky_01_r, loc=<-2.8274 -0.0892 0.1043>, rot=<0.0037 0.0011 0.2854 -0.9584>}
 pinky_03_r {parent=pinky_02_r, loc=<-2.2046 -0.0307 0.0078>, rot=<-0.0002 0.0098 0.1445 -0.9895>}
 ring_metacarpal_r {parent=hand_r, loc=<-2.2074 0.3625 -1.3638>, rot=<-0.0284 0.0722 -0.0458 -0.9959>}
 ring_01_r {parent=ring_metacarpal_r, loc=<-7.0295 0.6197 0.1446>, rot=<-0.0489 -0.0252 0.1892 -0.9804>}
 ring_02_r {parent=ring_01_r, loc=<-3.6174 -0.4647 0.1027>, rot=<0 0 0.2806 -0.9598>}
 ring_03_r {parent=ring_02_r, loc=<-2.4152 -0.1141 -0.0199>, rot=<0 0 0.1189 -0.9929>}
 thumb_01_r {parent=hand_r, loc=<-1.3118 0.3191 2.459>, rot=<0.6546 -0.4042 0.0664 -0.6354>}
 thumb_02_r {parent=thumb_01_r, loc=<-5.9336 0.0448 -0.0401>, rot=<-0.0037 0.0213 0.298 -0.9543>}
 thumb_03_r {parent=thumb_02_r, loc=<-2.2223 -0.3682 0.0907>, rot=<0.0118 -0.013 0.0066 -0.9998>}
 lowerarm_twist_01_r {parent=lowerarm_r, loc=<-18.3161 0 0>, rot=<0 0 0 1>}
 lowerarm_twist_02_r {parent=lowerarm_r, loc=<-10.6707 0 0>, rot=<0 0 0 1>}
 dyn_sleeves_r_MSTR {parent=lowerarm_r, loc=<2.5831 -3.0385 1.1205>, rot=<0.32 0.6264 -0.2065 -0.6801>}
 dyn_sleeves_r_up {parent=dyn_sleeves_r_MSTR, loc=<-7.5686 0.0302 2.6623>, rot=<-0.0012 0.7379 -0.0033 -0.6749>}
 dyn_sleeves_r_dn {parent=dyn_sleeves_r_MSTR, loc=<7.4408 -0.6552 2.4238>, rot=<0.7326 -0.0016 0.6806 -0.0008>}
 dyn_sleeves_r_ft {parent=dyn_sleeves_r_MSTR, loc=<-1.8375 -5.791 3.4413>, rot=<0.5146 -0.5202 0.4846 0.4793>}
 dyn_sleeves_r_bk {parent=dyn_sleeves_r_MSTR, loc=<-0.9936 6.2341 1.6941>, rot=<0.5216 0.5091 0.4782 -0.4899>}
 deform_elbow_in_base_r {parent=lowerarm_r, loc=<-0.1834 -3.6062 0.2572>, rot=<0 0 0 -1>}
 deform_elbow_in_r {parent=deform_elbow_in_base_r, loc=<0 0 0>, rot=<0 0 0 1>}
 deform_elbow_out_base_r {parent=lowerarm_r, loc=<0 4 -0.0001>, rot=<0 0 0 -1>}
 deform_elbow_out_r {parent=deform_elbow_out_base_r, loc=<0 0 0>, rot=<0 0 0 1>}
 upperarm_twist_01_r {parent=upperarm_r, loc=<-7.2135 0.1134 0>, rot=<0 0 0 1>}
 dyn_arm_straps_r_01 {parent=upperarm_twist_01_r, loc=<-1.587 6.6585 -1.3647>, rot=<-0.008 -0.7465 -0.6406 -0.1797>}
 dyn_arm_straps_r_02 {parent=upperarm_twist_01_r, loc=<0.6621 0.0932 -6.3872>, rot=<-0.0049 -0.9822 -0.0075 -0.1877>}
 dyn_arm_straps_r_03 {parent=upperarm_twist_01_r, loc=<-0.6849 -6.2838 -3.1453>, rot=<-0.0361 -0.8975 0.4263 -0.1067>}
 upperarm_twist_02_r {parent=upperarm_r, loc=<-19.7998 0.1134 0>, rot=<0 0 0 1>}
 neck_01 {parent=spine_05, loc=<11.3138 0.1529 0>, rot=<0 0 -0.2003 -0.9797>}
 neck_02 {parent=neck_01, loc=<4.5627 0 0>, rot=<0 0 0.0504 -0.9987>}
 head {parent=neck_02, loc=<5.3914 0 0>, rot=<0 0 0.115 -0.9934>}
 attach_cape {parent=spine_05, loc=<7.7592 -12.2707 0>, rot=<0.0616 -0.7044 -0.0616 -0.7044>}
 attach_backpack {parent=spine_05, loc=<-3.1836 -12.9397 0>, rot=<0.0261 -0.7066 -0.0261 -0.7066>}
 dyn_hood_r_MSTR {parent=spine_05, loc=<9.986 -4.3912 11.1355>, rot=<-0.2236 -0.0139 -0.0399 -0.9738>}
 dyn_hood_r {parent=dyn_hood_r_MSTR, loc=<0 0 0>, rot=<0 0 0 1>}
 dyn_hood_l_MSTR {parent=spine_05, loc=<9.986 -4.3912 -11.1355>, rot=<0.9738 0.0399 0.0139 0.2236>}
 dyn_hood_l {parent=dyn_hood_l_MSTR, loc=<0 0 0>, rot=<0 0 0 -1>}
 dyn_hood_bk_MSTR {parent=spine_05, loc=<7.986 -9.5619 0>, rot=<-0.7063 -0.0335 -0.0335 -0.7063>}
 dyn_hood_bk {parent=dyn_hood_bk_MSTR, loc=<0 0 0>, rot=<0 0 0 1>}
 thigh_l {parent=pelvis, loc=<-1.0855 -0.2109 -8.7053>, rot=<0.0009 0.0355 0.0155 -0.9993>}
 calf_l {parent=thigh_l, loc=<-42.5284 -0.0019 -0.7297>, rot=<0.0006 -0.0161 -0.0391 -0.9991>}
 foot_l {parent=calf_l, loc=<-46.7401 -2.5014 -0.592>, rot=<0.0427 -0.0238 0.0353 -0.9982>}
 ball_l {parent=foot_l, loc=<-3.5658 10.8066 -0.1544>, rot=<0.0001 -0.0001 -0.7063 -0.7079>}
 calf_twist_01_l {parent=calf_l, loc=<-35.8483 -1.9197 -0.5794>, rot=<0.0237 0.0095 -0.0035 -0.9997>}
 calf_twist_02_l {parent=calf_l, loc=<-19.1546 -1.7943 -0.2638>, rot=<0.0237 0.0095 -0.0035 -0.9997>}
 dyn_pants_bk_l {parent=calf_twist_02_l, loc=<14.044 -12.1209 1.5588>, rot=<-0.0461 -0.7221 0.6893 0.0361>}
 dyn_pants_ft_l {parent=calf_twist_02_l, loc=<15.5153 9.6288 0.1613>, rot=<0.005 0.6885 0.725 0.0208>}
 dyn_pants_otr_l {parent=calf_twist_02_l, loc=<14.9997 -1.2351 -10.2894>, rot=<-0.0208 -0.9989 -0.0252 -0.0345>}
 dyn_pants_inr_l {parent=calf_twist_02_l, loc=<14.8066 -0.1211 10.4214>, rot=<-0.0303 -0.0261 0.9991 0.0114>}
 thigh_twist_01_l {parent=thigh_l, loc=<-27.952 0 0>, rot=<0.0013 0 0 -1>}
 thigh_r {parent=pelvis, loc=<-1.0855 -0.2109 8.7053>, rot=<0.0355 -0.0009 0.9993 0.0155>}
 calf_r {parent=thigh_r, loc=<42.5284 0.0019 0.7297>, rot=<0.0006 -0.0161 -0.0391 -0.9991>}
 foot_r {parent=calf_r, loc=<46.7401 2.5014 0.592>, rot=<0.0427 -0.0238 0.0353 -0.9982>}
 ball_r {parent=foot_r, loc=<3.5658 -10.8066 0.1544>, rot=<0.0001 -0.0001 -0.7063 -0.7079>}
 calf_twist_01_r {parent=calf_r, loc=<35.8483 1.9197 0.5794>, rot=<0.0237 0.0095 -0.0035 -0.9997>}
 calf_twist_02_r {parent=calf_r, loc=<19.1546 1.7943 0.2638>, rot=<0.0237 0.0095 -0.0035 -0.9997>}
 dyn_pants_bk_r {parent=calf_twist_02_r, loc=<-14.044 12.1209 -1.5588>, rot=<0.6893 -0.0361 0.0461 -0.7221>}
 dyn_pants_ft_r {parent=calf_twist_02_r, loc=<-15.5154 -9.6288 -0.1613>, rot=<-0.725 0.0208 0.005 -0.6885>}
 dyn_pants_otr_r {parent=calf_twist_02_r, loc=<-14.9997 1.2351 10.2894>, rot=<-0.0252 0.0345 0.0208 -0.9989>}
 dyn_pants_inr_r {parent=calf_twist_02_r, loc=<-14.8065 0.1211 -10.4214>, rot=<0.9991 -0.0114 0.0303 -0.0261>}
 thigh_twist_01_r {parent=thigh_r, loc=<27.952 0 0>, rot=<0.0013 0 0 -1>}
 dyn_bag_TGT {parent=thigh_r, loc=<25.3109 -9.0293 2.8173>, rot=<-0.5759 0.077 -0.0448 -0.8127>}
 dyn_jacket_bk_MSTR {parent=pelvis, loc=<19.445 -15.4598 0>, rot=<-0.0122 -0.7068 0.7071 0.0145>}
 dyn_jacket_l_bk_MSTR {parent=dyn_jacket_bk_MSTR, loc=<24.1566 8.3976 -0.0668>, rot=<0.0755 0.0687 -0.0773 -0.9918>}
 dyn_jacket_l_bk_01 {parent=dyn_jacket_l_bk_MSTR, loc=<0 0 0>, rot=<0 0 0 1>}
 dyn_jacket_l_bk_02 {parent=dyn_jacket_l_bk_01, loc=<12.9491 0 0>, rot=<0 0.0239 0 -0.9997>}
 dyn_jacket_l_bk_03 {parent=dyn_jacket_l_bk_02, loc=<12.1335 -0.702 -0.4897>, rot=<0.0016 0.0675 0.0241 -0.9974>}
 dyn_jacket_l_bk_04 {parent=dyn_jacket_l_bk_03, loc=<15 0 0>, rot=<0 0 0 1>}
 dyn_jacket_r_bk_MSTR {parent=dyn_jacket_bk_MSTR, loc=<24.2121 -8.2362 -0.0582>, rot=<-0.0752 0.0685 0.074 -0.9921>}
 dyn_jacket_r_bk_01 {parent=dyn_jacket_r_bk_MSTR, loc=<0 0 0>, rot=<0 0 0 1>}
 dyn_jacket_r_bk_02 {parent=dyn_jacket_r_bk_01, loc=<12.9491 0 0>, rot=<0 0.0239 0 -0.9997>}
 dyn_jacket_r_bk_03 {parent=dyn_jacket_r_bk_02, loc=<12.1335 0.7021 -0.4897>, rot=<-0.0016 0.0675 -0.0241 -0.9974>}
 dyn_jacket_r_bk_04 {parent=dyn_jacket_r_bk_03, loc=<15 0 0>, rot=<0 0 0 1>}
 dyn_jacket_l_fr_MSTR {parent=pelvis, loc=<19.9442 8.6417 -10.1808>, rot=<-0.0065 -0.8821 -0.4677 -0.0563>}
 dyn_jacket_l_fr_01 {parent=dyn_jacket_l_fr_MSTR, loc=<0 0 0>, rot=<0 0 0 1>}
 dyn_jacket_l_fr_02 {parent=dyn_jacket_l_fr_01, loc=<13.8223 0 0>, rot=<0 0.0202 0.0299 -0.9993>}
 dyn_jacket_l_fr_03 {parent=dyn_jacket_l_fr_02, loc=<13.4971 0 0>, rot=<0.0015 -0.0043 0.0093 -0.9999>}
 dyn_jacket_l_fr_04 {parent=dyn_jacket_l_fr_03, loc=<12.6528 0 0>, rot=<-0.0041 0.0277 0.0041 -0.9996>}
 dyn_jacket_l_fr_05 {parent=dyn_jacket_l_fr_04, loc=<14.4221 0.123 -0.5859>, rot=<0 0.0276 0 -0.9996>}
 dyn_jacket_l_fr_06 {parent=dyn_jacket_l_fr_05, loc=<15 0 0>, rot=<0 0 0 1>}
 dyn_jacket_l_MSTR {parent=pelvis, loc=<6.017 -1.9089 -16.6246>, rot=<-0.0327 -0.993 0.0003 -0.1135>}
 dyn_jacket_l_01 {parent=dyn_jacket_l_MSTR, loc=<0 0 0>, rot=<0 0 0 1>}
 dyn_jacket_a_l_01 {parent=dyn_jacket_l_01, loc=<13.72 3.5376 -0.35>, rot=<0.0564 0.0106 0.0003 -0.9984>}
 dyn_jacket_a_l_02 {parent=dyn_jacket_a_l_01, loc=<12.3151 0 0>, rot=<0.0011 0.0293 -0.0067 -0.9995>}
 dyn_jacket_a_l_03 {parent=dyn_jacket_a_l_02, loc=<13.1481 -0.1502 -0.1671>, rot=<0.0006 0.0196 0.0317 -0.9993>}
 dyn_jacket_a_l_04 {parent=dyn_jacket_a_l_03, loc=<15 0 0>, rot=<0 0 0 1>}
 dyn_jacket_b_l_01 {parent=dyn_jacket_l_01, loc=<12.5302 -5.1905 -1.5609>, rot=<-0.2473 0.019 0.0574 -0.967>}
 dyn_jacket_b_l_02 {parent=dyn_jacket_b_l_01, loc=<12.8692 0 0>, rot=<-0.0008 0.0181 0.0195 -0.9996>}
 dyn_jacket_b_l_03 {parent=dyn_jacket_b_l_02, loc=<11 0 0>, rot=<0 0 0 1>}
 dyn_jacket_r_fr_MSTR {parent=pelvis, loc=<19.9442 8.6417 10.1808>, rot=<0.0563 0.4677 0.8821 0.0065>}
 dyn_jacket_r_fr_01 {parent=dyn_jacket_r_fr_MSTR, loc=<0 0 0>, rot=<0 0 0 1>}
 dyn_jacket_r_fr_02 {parent=dyn_jacket_r_fr_01, loc=<13.8228 0 0>, rot=<0 0.0202 -0.0299 -0.9993>}
 dyn_jacket_r_fr_03 {parent=dyn_jacket_r_fr_02, loc=<13.4967 0.0001 0.0001>, rot=<-0.0015 -0.0043 -0.0093 -0.9999>}
 dyn_jacket_r_fr_04 {parent=dyn_jacket_r_fr_03, loc=<12.6528 0 0>, rot=<0.0041 0.0277 -0.0041 -0.9996>}
 dyn_jacket_r_fr_05 {parent=dyn_jacket_r_fr_04, loc=<14.4221 -0.123 -0.5858>, rot=<0 0.0276 0 -0.9996>}
 dyn_jacket_r_fr_06 {parent=dyn_jacket_r_fr_05, loc=<15 0 0>, rot=<0 0 0 1>}
 dyn_jacket_r_MSTR {parent=pelvis, loc=<6.0173 -1.9089 16.6246>, rot=<0.1135 -0.0003 0.993 0.0327>}
 dyn_jacket_r_01 {parent=dyn_jacket_r_MSTR, loc=<0 0 0>, rot=<0 0 0 1>}
 dyn_jacket_a_r_01 {parent=dyn_jacket_r_01, loc=<13.7204 -3.5376 -0.3501>, rot=<-0.0564 0.0106 -0.0003 -0.9984>}
 dyn_jacket_a_r_02 {parent=dyn_jacket_a_r_01, loc=<12.3151 0 -0.0001>, rot=<-0.0011 0.0293 0.0067 -0.9995>}
 dyn_jacket_a_r_03 {parent=dyn_jacket_a_r_02, loc=<13.1482 0.1502 -0.1671>, rot=<-0.0006 0.0196 -0.0317 -0.9993>}
 dyn_jacket_a_r_04 {parent=dyn_jacket_a_r_03, loc=<15 0 0>, rot=<0 0 0 1>}
 dyn_jacket_b_r_01 {parent=dyn_jacket_r_01, loc=<12.5306 5.1905 -1.561>, rot=<0.2473 0.019 -0.0574 -0.967>}
 dyn_jacket_b_r_02 {parent=dyn_jacket_b_r_01, loc=<12.8692 0 0>, rot=<0.0008 0.0181 -0.0195 -0.9996>}
 dyn_jacket_b_r_03 {parent=dyn_jacket_b_r_02, loc=<11 0 0>, rot=<0 0 0 1>}
 dyn_laces_MATR {parent=pelvis, loc=<3.3349 8.6322 -0.0436>, rot=<-0.0123 -0.707 0.0123 -0.707>}
 dyn_laces_l_01 {parent=dyn_laces_MATR, loc=<1.2291 1.1698 -1.0749>, rot=<0.4867 -0.4877 -0.513 -0.5119>}
 dyn_laces_l_02 {parent=dyn_laces_l_01, loc=<11.0009 0 0>, rot=<0 -0.0115 0.0034 -0.9999>}
 dyn_laces_r_01 {parent=dyn_laces_MATR, loc=<-1.3162 1.1698 -1.0746>, rot=<0.4877 -0.4867 -0.5119 -0.513>}
 dyn_laces_r_02 {parent=dyn_laces_r_01, loc=<11.0011 0 0>, rot=<0 -0.0115 -0.0034 -0.9999>}
 dyn_groin_collision {parent=pelvis, loc=<-15 0 0>, rot=<0.0127 0.707 0.707 0.0127>}
 dyn_bag_MSTR {parent=pelvis, loc=<3.2043 8.2139 5.9348>, rot=<0.1146 0.5719 0.8105 -0.0526>}
 dyn_bag {parent=dyn_bag_MSTR, loc=<0 0 0>, rot=<0 0 0 1>}
 NewControl {type=Control, loc=<0 0 0>, rot=<0 0 0 1>}
 r_knee_corrective {type=Curve, loc=<0 0 0>, rot=<0 0 0 1>}
 l_knee_corrective {type=Curve, loc=<0 0 0>, rot=<0 0 0 1>}
 Male_Med_Head_01_Pout {type=Curve, loc=<0 0 0>, rot=<0 0 0 1>}
 Male_Med_Head_02_Pout {type=Curve, loc=<0 0 0>, rot=<0 0 0 1>}
 Male_Med_Head_03_Pout {type=Curve, loc=<0 0 0>, rot=<0 0 0 1>}
 Male_Med_Head_05_Effort {type=Curve, loc=<0 0 0>, rot=<0 0 0 1>}
 Stretch {type=Curve, loc=<0 0 0>, rot=<0 0 0 1>}
 Squash {type=Curve, loc=<0 0 0>, rot=<0 0 0 1>}
 polySurface151 {type=Curve, loc=<0 0 0>, rot=<0 0 0 1>}
 l_cheek_puff {type=Curve, loc=<0 0 0>, rot=<0 0 0 1>}
 r_cheek_puff {type=Curve, loc=<0 0 0>, rot=<0 0 0 1>}
 l_smile_fix {type=Curve, loc=<0 0 0>, rot=<0 0 0 1>}
 r_smile_fix {type=Curve, loc=<0 0 0>, rot=<0 0 0 1>}
 jaw_open_fix {type=Curve, loc=<0 0 0>, rot=<0 0 0 1>}
 r_lip_sneer_fix {type=Curve, loc=<0 0 0>, rot=<0 0 0 1>}
 l_lip_sneer_fix {type=Curve, loc=<0 0 0>, rot=<0 0 0 1>}
 L_Blink {type=Curve, loc=<0 0 0>, rot=<0 0 0 1>}
 R_Blink {type=Curve, loc=<0 0 0>, rot=<0 0 0 1>}
 L_Arm_Stretch {type=Curve, loc=<0 0 0>, rot=<0 0 0 1>}
 Hair_Hat {type=Curve, loc=<0 0 0>, rot=<0 0 0 1>}
 Hair_Helmet {type=Curve, loc=<0 0 0>, rot=<0 0 0 1>}
 Male_Med_Head_02_What {type=Curve, loc=<0 0 0>, rot=<0 0 0 1>}
 Male_Med_Head_02_Smile {type=Curve, loc=<0 0 0>, rot=<0 0 0 1>}
 Male_Med_Head_02_Scowl {type=Curve, loc=<0 0 0>, rot=<0 0 0 1>}
 Male_Med_Head_02_Hmmph {type=Curve, loc=<0 0 0>, rot=<0 0 0 1>}
 Male_Med_Head_02_Grrr {type=Curve, loc=<0 0 0>, rot=<0 0 0 1>}
 Male_Med_Head_02_Effort {type=Curve, loc=<0 0 0>, rot=<0 0 0 1>}
 Male_Med_Head_03_What {type=Curve, loc=<0 0 0>, rot=<0 0 0 1>}
 Male_Med_Head_03_Grrr {type=Curve, loc=<0 0 0>, rot=<0 0 0 1>}
 Male_Med_Head_03_Effort {type=Curve, loc=<0 0 0>, rot=<0 0 0 1>}
 Male_Med_Head_03_Hmmph {type=Curve, loc=<0 0 0>, rot=<0 0 0 1>}
 Male_Med_Head_03_Scowl {type=Curve, loc=<0 0 0>, rot=<0 0 0 1>}
 Male_Med_Head_03_Smile {type=Curve, loc=<0 0 0>, rot=<0 0 0 1>}
 Male_Med_Head_05_Teeth_Low_What {type=Curve, loc=<0 0 0>, rot=<0 0 0 1>}
 Male_Med_Head_05_Teeth_Low_Pout {type=Curve, loc=<0 0 0>, rot=<0 0 0 1>}
 Male_Med_Head_05_Teeth_Low_Hmmph {type=Curve, loc=<0 0 0>, rot=<0 0 0 1>}
 Male_Med_Head_05_Teeth_Low_Grrr {type=Curve, loc=<0 0 0>, rot=<0 0 0 1>}
 Male_Med_Head_05_Teeth_Low_Effort {type=Curve, loc=<0 0 0>, rot=<0 0 0 1>}
 Male_Med_Head_05_What {type=Curve, loc=<0 0 0>, rot=<0 0 0 1>}
 Male_Med_Head_05_Grrr {type=Curve, loc=<0 0 0>, rot=<0 0 0 1>}
 Male_Med_Head_05_Hmmph {type=Curve, loc=<0 0 0>, rot=<0 0 0 1>}
 Male_Med_Head_05_Pout {type=Curve, loc=<0 0 0>, rot=<0 0 0 1>}
 Male_Med_Head_05_Tongue_What {type=Curve, loc=<0 0 0>, rot=<0 0 0 1>}
 Male_Med_Head_05_Tongue_Pout {type=Curve, loc=<0 0 0>, rot=<0 0 0 1>}
 Male_Med_Head_05_Tongue_Hmmph {type=Curve, loc=<0 0 0>, rot=<0 0 0 1>}
 Male_Med_Head_05_Tongue_Grrr {type=Curve, loc=<0 0 0>, rot=<0 0 0 1>}
 Male_Med_Head_05_Tongue_Effort {type=Curve, loc=<0 0 0>, rot=<0 0 0 1>}
 AdditiveLayer {type=Curve, loc=<0 0 0>, rot=<0 0 0 1>}
 RightHandIK {type=Curve, loc=<0 0 0>, rot=<0 0 0 1>}
 LeftHandIK {type=Curve, loc=<0 0 0>, rot=<0 0 0 1>}
 LeftHandFK {type=Curve, loc=<0 0 0>, rot=<0 0 0 1>}
 Male_Med_Head_01_Grrr {type=Curve, loc=<0 0 0>, rot=<0 0 0 1>}
 WeaponAiming {type=Curve, loc=<0 0 0>, rot=<0 0 0 1>}
 Male_Med_Head_01_Hmmph {type=Curve, loc=<0 0 0>, rot=<0 0 0 1>}
 DragonSlash {type=Curve, loc=<0 0 0>, rot=<0 0 0 1>}
 Male_Med_Head_01_Effort {type=Curve, loc=<0 0 0>, rot=<0 0 0 1>}
 LeftHankFK {type=Curve, loc=<0 0 0>, rot=<0 0 0 1>}
 blendOrient1 {type=Curve, loc=<0 0 0>, rot=<0 0 0 1>}
 blendParent1 {type=Curve, loc=<0 0 0>, rot=<0 0 0 1>}
 bottom_R_Corner_Down {type=Curve, loc=<0 0 0>, rot=<0 0 0 1>}
 bottom_R_Corner_Up {type=Curve, loc=<0 0 0>, rot=<0 0 0 1>}
 top_L_Corner_Up {type=Curve, loc=<0 0 0>, rot=<0 0 0 1>}
 top_L_Corner_Down {type=Curve, loc=<0 0 0>, rot=<0 0 0 1>}
 top_R_Corner_Down {type=Curve, loc=<0 0 0>, rot=<0 0 0 1>}
 top_R_Corner_Up {type=Curve, loc=<0 0 0>, rot=<0 0 0 1>}
 bottom_L_Corner_Down {type=Curve, loc=<0 0 0>, rot=<0 0 0 1>}
 bottomt_L_Corner_Up {type=Curve, loc=<0 0 0>, rot=<0 0 0 1>}
 TrailCurve1 {type=Curve, loc=<0 0 0>, rot=<0 0 0 1>}
 BothHandFK {type=Curve, loc=<0 0 0>, rot=<0 0 0 1>}
 bottom_R_Corner_Down1 {type=Curve, loc=<0 0 0>, rot=<0 0 0 1>}
 blendPoint1 {type=Curve, loc=<0 0 0>, rot=<0 0 0 1>}
 WeaponAimingFreeze {type=Curve, loc=<0 0 0>, rot=<0 0 0 1>}
 l_calf_top_taper {type=Curve, loc=<0 0 0>, rot=<0 0 0 1>}
 JogBlendSpaceRotation {type=Curve, loc=<0 0 0>, rot=<0 0 0 1>}
 ZeroOutPitchWeight {type=Curve, loc=<0 0 0>, rot=<0 0 0 1>}
 MeleeTwist {type=Curve, loc=<0 0 0>, rot=<0 0 0 1>}
 MeleeAnim {type=Curve, loc=<0 0 0>, rot=<0 0 0 1>}
 l_sneer_fix {type=Curve, loc=<0 0 0>, rot=<0 0 0 1>}
 jaw_fix {type=Curve, loc=<0 0 0>, rot=<0 0 0 1>}
 r_sneer_fix {type=Curve, loc=<0 0 0>, rot=<0 0 0 1>}
 TurnRotationAmount {type=Curve, loc=<0 0 0>, rot=<0 0 0 1>}
 TurnEnd {type=Curve, loc=<0 0 0>, rot=<0 0 0 1>}
 BlendIn {type=Curve, loc=<0 0 0>, rot=<0 0 0 1>}
 BlendOut {type=Curve, loc=<0 0 0>, rot=<0 0 0 1>}
 BodyRotateX {type=Curve, loc=<0 0 0>, rot=<0 0 0 1>}
 BodyTranslateX {type=Curve, loc=<0 0 0>, rot=<0 0 0 1>}
 CrescentKickScale {type=Curve, loc=<0 0 0>, rot=<0 0 0 1>}
 mask_geo {type=Curve, loc=<0 0 0>, rot=<0 0 0 1>}
 RemoveJumpAdditive {type=Curve, loc=<0 0 0>, rot=<0 0 0 1>}
 RightLeg_IK {type=Curve, loc=<0 0 0>, rot=<0 0 0 1>}
 SnapWeapon_LHand {type=Curve, loc=<0 0 0>, rot=<0 0 0 1>}
 PivotAnim {type=Curve, loc=<0 0 0>, rot=<0 0 0 1>}
 DefaultAdditiveStateWeight {type=Curve, loc=<0 0 0>, rot=<0 0 0 1>}
 StartAnim {type=Curve, loc=<0 0 0>, rot=<0 0 0 1>}
 StopAnim {type=Curve, loc=<0 0 0>, rot=<0 0 0 1>}
 DisableSpeedWarping {type=Curve, loc=<0 0 0>, rot=<0 0 0 1>}
 DisableDBNOSlopeWarp {type=Curve, loc=<0 0 0>, rot=<0 0 0 1>}
 MaskRightArm {type=Curve, loc=<0 0 0>, rot=<0 0 0 1>}
 TurnInPlaceAnim {type=Curve, loc=<0 0 0>, rot=<0 0 0 1>}
 Head_MESH {type=Curve, loc=<0 0 0>, rot=<0 0 0 1>}
 Hair_MESH {type=Curve, loc=<0 0 0>, rot=<0 0 0 1>}
 LaunchPadTransitionOff {type=Curve, loc=<0 0 0>, rot=<0 0 0 1>}
 C_glabella_up_pose {type=Curve, loc=<0 0 0>, rot=<0 0 0 1>}
 jaw_open_pose {type=Curve, loc=<0 0 0>, rot=<0 0 0 1>}
 L_blink_pose {type=Curve, loc=<0 0 0>, rot=<0 0 0 1>}
 L_brow_down_pose {type=Curve, loc=<0 0 0>, rot=<0 0 0 1>}
 L_brow_squeeze_in_pose {type=Curve, loc=<0 0 0>, rot=<0 0 0 1>}
 L_brow_up_pose {type=Curve, loc=<0 0 0>, rot=<0 0 0 1>}
 L_frown_pose {type=Curve, loc=<0 0 0>, rot=<0 0 0 1>}
 L_lip_corner_narrow_pose {type=Curve, loc=<0 0 0>, rot=<0 0 0 1>}
 L_squint_inner_pose {type=Curve, loc=<0 0 0>, rot=<0 0 0 1>}
 L_wide_pose {type=Curve, loc=<0 0 0>, rot=<0 0 0 1>}
 look_down_pose {type=Curve, loc=<0 0 0>, rot=<0 0 0 1>}
 look_left_pose {type=Curve, loc=<0 0 0>, rot=<0 0 0 1>}
 look_right_pose {type=Curve, loc=<0 0 0>, rot=<0 0 0 1>}
 look_up_pose {type=Curve, loc=<0 0 0>, rot=<0 0 0 1>}
 phoneme_fv_pose {type=Curve, loc=<0 0 0>, rot=<0 0 0 1>}
 phoneme_oo_pose {type=Curve, loc=<0 0 0>, rot=<0 0 0 1>}
 R_blink_pose {type=Curve, loc=<0 0 0>, rot=<0 0 0 1>}
 R_brow_down_pose {type=Curve, loc=<0 0 0>, rot=<0 0 0 1>}
 R_brow_squeeze_in_pose {type=Curve, loc=<0 0 0>, rot=<0 0 0 1>}
 R_brow_up_pose {type=Curve, loc=<0 0 0>, rot=<0 0 0 1>}
 R_frown_pose {type=Curve, loc=<0 0 0>, rot=<0 0 0 1>}
 R_lip_corner_narrow_pose {type=Curve, loc=<0 0 0>, rot=<0 0 0 1>}
 R_squint_inner_pose {type=Curve, loc=<0 0 0>, rot=<0 0 0 1>}
 R_wide_pose {type=Curve, loc=<0 0 0>, rot=<0 0 0 1>}
 DisableArmsHeadAdditive {type=Curve, loc=<0 0 0>, rot=<0 0 0 1>}
 FrontendRotation {type=Curve, loc=<0 0 0>, rot=<0 0 0 1>}
 BothLegFK {type=Curve, loc=<0 0 0>, rot=<0 0 0 1>}
 AllowTurnInPlace {type=Curve, loc=<0 0 0>, rot=<0 0 0 1>}
 MontageTimeStretchCurve {type=Curve, loc=<0 0 0>, rot=<0 0 0 1>}
 ParaGlideLeanAlpha {type=Curve, loc=<0 0 0>, rot=<0 0 0 1>}
 WeaponRInChestSpace {type=Curve, loc=<0 0 0>, rot=<0 0 0 1>}
 JimBeam_AO {type=Curve, loc=<0 0 0>, rot=<0 0 0 1>}
 R_squeeze_pose {type=Curve, loc=<0 0 0>, rot=<0 0 0 1>}
 L_squeeze_pose {type=Curve, loc=<0 0 0>, rot=<0 0 0 1>}
 R_upper_lip_raiser_pose {type=Curve, loc=<0 0 0>, rot=<0 0 0 1>}
 L_cheek_up_pose {type=Curve, loc=<0 0 0>, rot=<0 0 0 1>}
 R_cheek_up_pose {type=Curve, loc=<0 0 0>, rot=<0 0 0 1>}
 DisableFaceOverride {type=Curve, loc=<0 0 0>, rot=<0 0 0 1>}
 EnableFullBodyOverride {type=Curve, loc=<0 0 0>, rot=<0 0 0 1>}
 C_glabella_down_pose {type=Curve, loc=<0 0 0>, rot=<0 0 0 1>}
 C_nose_wrinkler_pose {type=Curve, loc=<0 0 0>, rot=<0 0 0 1>}
 eye_convergence_pose {type=Curve, loc=<0 0 0>, rot=<0 0 0 1>}
 jaw_back_pose {type=Curve, loc=<0 0 0>, rot=<0 0 0 1>}
 jaw_forward_pose {type=Curve, loc=<0 0 0>, rot=<0 0 0 1>}
 jaw_left_pose {type=Curve, loc=<0 0 0>, rot=<0 0 0 1>}
 jaw_right_pose {type=Curve, loc=<0 0 0>, rot=<0 0 0 1>}
 jaw_up_pose {type=Curve, loc=<0 0 0>, rot=<0 0 0 1>}
 L_lip_corner_wide_pose {type=Curve, loc=<0 0 0>, rot=<0 0 0 1>}
 L_lower_lip_down_pose {type=Curve, loc=<0 0 0>, rot=<0 0 0 1>}
 L_lower_lip_up_pose {type=Curve, loc=<0 0 0>, rot=<0 0 0 1>}
 L_smile_pose {type=Curve, loc=<0 0 0>, rot=<0 0 0 1>}
 L_upper_lip_lower_pose {type=Curve, loc=<0 0 0>, rot=<0 0 0 1>}
 L_upper_lip_raiser_pose {type=Curve, loc=<0 0 0>, rot=<0 0 0 1>}
 L_upperLidDown_pose {type=Curve, loc=<0 0 0>, rot=<0 0 0 1>}
 mouth_down_pose {type=Curve, loc=<0 0 0>, rot=<0 0 0 1>}
 mouth_up_pose {type=Curve, loc=<0 0 0>, rot=<0 0 0 1>}
 phoneme_ch_pose {type=Curve, loc=<0 0 0>, rot=<0 0 0 1>}
 phoneme_mbp_pose {type=Curve, loc=<0 0 0>, rot=<0 0 0 1>}
 R_lip_corner_wide_pose {type=Curve, loc=<0 0 0>, rot=<0 0 0 1>}
 R_lower_lip_down_pose {type=Curve, loc=<0 0 0>, rot=<0 0 0 1>}
 R_lower_lip_up_pose {type=Curve, loc=<0 0 0>, rot=<0 0 0 1>}
 R_smile_pose {type=Curve, loc=<0 0 0>, rot=<0 0 0 1>}
 R_upper_lip_lower_pose {type=Curve, loc=<0 0 0>, rot=<0 0 0 1>}
 R_upperLidDown_pose {type=Curve, loc=<0 0 0>, rot=<0 0 0 1>}
 tongue_back_pose {type=Curve, loc=<0 0 0>, rot=<0 0 0 1>}
 tongue_down_pose {type=Curve, loc=<0 0 0>, rot=<0 0 0 1>}
 tongue_forward_pose {type=Curve, loc=<0 0 0>, rot=<0 0 0 1>}
 tongue_up_pose {type=Curve, loc=<0 0 0>, rot=<0 0 0 1>}
 LiveLink {type=Curve, loc=<0 0 0>, rot=<0 0 0 1>}
 ParaGlideDeploy {type=Curve, loc=<0 0 0>, rot=<0 0 0 1>}
 HideLeftWeapon {type=Curve, loc=<0 0 0>, rot=<0 0 0 1>}
 DisableLeftHandIKGolfCart {type=Curve, loc=<0 0 0>, rot=<0 0 0 1>}
 DisableRightHandIKGolfCart {type=Curve, loc=<0 0 0>, rot=<0 0 0 1>}
 DisableHandIKRetargeting {type=Curve, loc=<0 0 0>, rot=<0 0 0 1>}
 DisableConsumableOffsetPose {type=Curve, loc=<0 0 0>, rot=<0 0 0 1>}
 UpperBodyNoAdditive {type=Curve, loc=<0 0 0>, rot=<0 0 0 1>}
 MaskLeftArm {type=Curve, loc=<0 0 0>, rot=<0 0 0 1>}
 ForceUpperBody {type=Curve, loc=<0 0 0>, rot=<0 0 0 1>}
 RightHandFK {type=Curve, loc=<0 0 0>, rot=<0 0 0 1>}
 ParaGlideLeanTransition {type=Curve, loc=<0 0 0>, rot=<0 0 0 1>}
 InMotionEmote {type=Curve, loc=<0 0 0>, rot=<0 0 0 1>}
 ForceFullBodyInFrontEnd {type=Curve, loc=<0 0 0>, rot=<0 0 0 1>}
 WeaponLinFloorSpace {type=Curve, loc=<0 0 0>, rot=<0 0 0 1>}
 FrontEnd_Idle_Bypass {type=Curve, loc=<0 0 0>, rot=<0 0 0 1>}
 DisableSlopeWarping {type=Curve, loc=<0 0 0>, rot=<0 0 0 1>}
 Disable_Custom {type=Curve, loc=<0 0 0>, rot=<0 0 0 1>}
 EmoteRequiresWeapon {type=Curve, loc=<0 0 0>, rot=<0 0 0 1>}
 DisableIKRootAdditive {type=Curve, loc=<0 0 0>, rot=<0 0 0 1>}
 DisableFullBodySprint {type=Curve, loc=<0 0 0>, rot=<0 0 0 1>}
 DetachFeet_Vehicle {type=Curve, loc=<0 0 0>, rot=<0 0 0 1>}
 IndigoBoost {type=Curve, loc=<0 0 0>, rot=<0 0 0 1>}
 DisableUpperBodyAdditiveMeshSpace {type=Curve, loc=<0 0 0>, rot=<0 0 0 1>}
 MeleeAOWeight {type=Curve, loc=<0 0 0>, rot=<0 0 0 1>}
 HandsInRootSpace {type=Curve, loc=<0 0 0>, rot=<0 0 0 1>}
 CrouchPelvisAdjustment {type=Curve, loc=<0 0 0>, rot=<0 0 0 1>}
 EnableAccelYawCorrection {type=Curve, loc=<0 0 0>, rot=<0 0 0 1>}
 ParaGlideModifyRoot {type=Curve, loc=<0 0 0>, rot=<0 0 0 1>}
 UseScalarParameterForBlink {type=Curve, loc=<0 0 0>, rot=<0 0 0 1>}
 BlinkControl {type=Curve, loc=<0 0 0>, rot=<0 0 0 1>}
 ForceTargeting {type=Curve, loc=<0 0 0>, rot=<0 0 0 1>}
 BlinkControlProp {type=Curve, loc=<0 0 0>, rot=<0 0 0 1>}
 UseScalarParameterForBlinkProp {type=Curve, loc=<0 0 0>, rot=<0 0 0 1>}
 CrackshotCozyHeadCorrection {type=Curve, loc=<0 0 0>, rot=<0 0 0 1>}
 ForwardMotion {type=Curve, loc=<0 0 0>, rot=<0 0 0 1>}
 EmoteForcedRootYaw {type=Curve, loc=<0 0 0>, rot=<0 0 0 1>}
 MeleeTrailWidth {type=Curve, loc=<0 0 0>, rot=<0 0 0 1>}
 space_world {type=Curve, loc=<0 0 0>, rot=<0 0 0 1>}
 PeelyBlenderFace {type=Curve, loc=<0 0 0>, rot=<0 0 0 1>}
 Width1 {type=Curve, loc=<0 0 0>, rot=<0 0 0 1>}
 OverrideIKHands {type=Curve, loc=<0 0 0>, rot=<0 0 0 1>}
 ParaGlideAddAlpha {type=Curve, loc=<0 0 0>, rot=<0 0 0 1>}
 ParaGlideTurn {type=Curve, loc=<0 0 0>, rot=<0 0 0 1>}
 DisablePelvisAdjustment {type=Curve, loc=<0 0 0>, rot=<0 0 0 1>}
 UncleBrollyDeploy {type=Curve, loc=<0 0 0>, rot=<0 0 0 1>}
 ForceRootSlopeOrientation {type=Curve, loc=<0 0 0>, rot=<0 0 0 1>}
 DisableAdditiveOffsetPose {type=Curve, loc=<0 0 0>, rot=<0 0 0 1>}
 DisableIdleNoiseAdditive {type=Curve, loc=<0 0 0>, rot=<0 0 0 1>}
 UsingConsumable {type=Curve, loc=<0 0 0>, rot=<0 0 0 1>}
 DisableRightArmAdditive {type=Curve, loc=<0 0 0>, rot=<0 0 0 1>}
]
variables [Upperarm_Twist {type=double, default=0}]
