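# Physics asset for the dynamic jacket and laces on the medium female LightningDragon character body cosmetic.
# Capsules cover pelvis, lower spine, legs, feet, groin and bag alongside the cloth bone chains.
# Front and branching side chains anchor through MSTR bones to spine_02, while back chains and laces anchor to pelvis, with limited swing and twist along each chain.
# Cross-constraints limiting only swing1 tie side branches to neighboring front and back panels.
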PhysicsAsset /BRCosmetics/Characters/Player/Female/Medium/Bodies/F_MED_LightningDragon/Meshes/F_MED_LightningDragon_Dyn_Jacket
bodies {
 calf_l [{shapeType=Capsule, length=17.3282, radius=10.5288, center=<-11.6217 -2.2618 -0.4465>, rotation=<rot 84.897 178.7789 -0.2947>}]
 calf_r [{shapeType=Capsule, length=16.8282, radius=11.0288, center=<14.9785 2.2076 0.7463>, rotation=<rot 84.897 178.7789 179.7053>}]
 calf_twist_01_l [{shapeType=Capsule, length=13.9532, radius=8.6538, center=<4.2478 -0.5544 0.4508>, rotation=<rot 88.9235 177.677 -1.3957>}]
 calf_twist_01_r [{shapeType=Capsule, length=13.9532, radius=8.6538, center=<-4.2478 0.5544 -0.4508>, rotation=<rot 88.9235 177.677 178.6043>}]
 dyn_bag [
  {shapeType=Capsule, length=11.25, radius=4.125, center=<9.3679 -1.263 -1.2843>, rotation=<rot 90 0 0>}
  {shapeType=Capsule, length=11.25, radius=4.125, center=<9.3679 2.2819 -1.2681>, rotation=<rot 90 0 0>}
 ]
 dyn_groin_collision [{shapeType=Capsule, length=13.1, radius=11.7122, center=<2.0336 -0.7751 -1.3756>, rotation=<rot 0 0 90>}]
 dyn_jacket_a_l_01 [{shapeType=Capsule, length=5.5803, radius=6.5415, shapeName=dyn_jacket_l_bk_01_capsule, center=<7 2 4>, rotation=<rot 0.0001 0.0006 99.9999>}]
 dyn_jacket_a_l_02 [{shapeType=Capsule, length=5.5803, radius=6.5415, shapeName=dyn_jacket_l_bk_02_capsule, center=<7 2 4>, rotation=<rot 0.0001 0.0006 99.9998>}]
 dyn_jacket_a_l_03 [{shapeType=Capsule, length=5.5, radius=4.5, shapeName=dyn_jacket_l_bk_02_capsule, center=<8 0 3>, rotation=<rot 90 0 0>}]
 dyn_jacket_a_r_01 [{shapeType=Capsule, length=5.5803, radius=6.5415, shapeName=dyn_jacket_l_bk_01_capsule, center=<7 -2 4>, rotation=<rot 0 0 80>}]
 dyn_jacket_a_r_02 [{shapeType=Capsule, length=5.5803, radius=6.5415, shapeName=dyn_jacket_l_bk_02_capsule, center=<7 -2 4>, rotation=<rot 0 0 80>}]
 dyn_jacket_a_r_03 [{shapeType=Capsule, length=5.5, radius=4.5, shapeName=dyn_jacket_l_bk_02_capsule, center=<8 0 3>, rotation=<rot 90 0 0>}]
 dyn_jacket_b_l_01 [{shapeType=Capsule, length=8.375, radius=6.925, shapeName=dyn_jacket_l_bk_01_capsule, center=<7 -2 4>, rotation=<rot -0.0007 0.0009 89.9998>}]
 dyn_jacket_b_l_02 [{shapeType=Capsule, length=10.375, radius=6.925, shapeName=dyn_jacket_l_bk_02_capsule, center=<7 -2 4>, rotation=<rot -0.0007 0.0008 89.9999>}]
 dyn_jacket_b_r_01 [{shapeType=Capsule, length=11, radius=6.3, shapeName=dyn_jacket_l_bk_01_capsule, center=<7 2 4>, rotation=<rot -0.0004 0.0004 95>}]
 dyn_jacket_b_r_02 [{shapeType=Capsule, length=11, radius=6.3, shapeName=dyn_jacket_l_bk_02_capsule, center=<7 2 4>, rotation=<rot -0.0004 0.0004 95>}]
 dyn_jacket_bk_MSTR [{shapeType=Capsule, length=1, radius=1}]
 dyn_jacket_l_01 [{shapeType=Capsule, length=8.2053, radius=7.1665, shapeName=dyn_jacket_l_01_capsule, center=<7 0 4>, rotation=<rot 0 0.0004 79.9999>}]
 dyn_jacket_l_MSTR [{shapeType=Capsule, length=1, radius=1}]
 dyn_jacket_l_bk_01 [{shapeType=Capsule, length=13.4553, radius=7.6665, shapeName=dyn_jacket_l_bk_01_capsule, center=<7 0 4>, rotation=<rot -0.0003 0 94.9999>}]
 dyn_jacket_l_bk_02 [{shapeType=Capsule, length=16.5803, radius=7.0415, shapeName=dyn_jacket_l_bk_02_capsule, center=<7 0 2.5>, rotation=<rot -0.0001 0 95>}]
 dyn_jacket_l_bk_03 [{shapeType=Capsule, length=5.5, radius=4.5, shapeName=dyn_jacket_l_bk_02_capsule, center=<8 0 3>, rotation=<rot 90 0 0>}]
 dyn_jacket_l_bk_MSTR [{shapeType=Capsule, length=1, radius=1}]
 dyn_jacket_l_fr_01 [{shapeType=Capsule, length=5.125, radius=6.875, shapeName=dyn_jacket_l_fr_01_capsule, center=<6 0 5>, rotation=<rot 90 0 0>}]
 dyn_jacket_l_fr_02 [{shapeType=Capsule, length=6.75, radius=7.725, shapeName=dyn_jacket_l_fr_02_capsule, center=<6 -2 5>, rotation=<rot 0 0 90>}]
 dyn_jacket_l_fr_03 [{shapeType=Capsule, length=6.875, radius=7.6, shapeName=dyn_jacket_l_fr_02_capsule, center=<6 -2 5>, rotation=<rot 0 0 90>}]
 dyn_jacket_l_fr_04 [{shapeType=Capsule, length=6.25, radius=8.225, shapeName=dyn_jacket_l_fr_02_capsule, center=<6 -2 5>, rotation=<rot 0 0 90>}]
 dyn_jacket_l_fr_05 [{shapeType=Capsule, length=5.5, radius=4.5, shapeName=dyn_jacket_l_fr_02_capsule, center=<8 0 3>, rotation=<rot 90 0 0>}]
 dyn_jacket_l_fr_MSTR [{shapeType=Capsule, length=1, radius=1}]
 dyn_jacket_r_01 [{shapeType=Capsule, length=9.0803, radius=6.2915, shapeName=dyn_jacket_l_01_capsule, center=<7 0 4>, rotation=<rot 90 0 0>}]
 dyn_jacket_r_MSTR [{shapeType=Capsule, length=1, radius=1}]
 dyn_jacket_r_bk_01 [{shapeType=Capsule, length=17.2053, radius=6.4165, shapeName=dyn_jacket_l_bk_01_capsule, center=<7 0 4>, rotation=<rot -0.0001 0 85>}]
 dyn_jacket_r_bk_02 [{shapeType=Capsule, length=17.5803, radius=6.0415, shapeName=dyn_jacket_l_bk_02_capsule, center=<7 0 2.5>, rotation=<rot -0.0001 0 85>}]
 dyn_jacket_r_bk_03 [{shapeType=Capsule, length=5.5, radius=4.5, shapeName=dyn_jacket_l_bk_02_capsule, center=<8 0 3>, rotation=<rot 90 0 0>}]
 dyn_jacket_r_bk_MSTR [{shapeType=Capsule, length=1, radius=1}]
 dyn_jacket_r_fr_01 [{shapeType=Capsule, length=5, radius=7, shapeName=dyn_jacket_l_fr_01_capsule, center=<6 0 5>, rotation=<rot 90 0 0>}]
 dyn_jacket_r_fr_02 [{shapeType=Capsule, length=9.5, radius=6.875, shapeName=dyn_jacket_l_fr_02_capsule, center=<6 3 5>, rotation=<rot 0 0 90>}]
 dyn_jacket_r_fr_03 [{shapeType=Capsule, length=9.5, radius=6.875, shapeName=dyn_jacket_l_fr_02_capsule, center=<6 3 5>, rotation=<rot 0 0 90>}]
 dyn_jacket_r_fr_04 [{shapeType=Capsule, length=8.875, radius=7.5, shapeName=dyn_jacket_l_fr_02_capsule, center=<6 3 5>, rotation=<rot 0 0 90>}]
 dyn_jacket_r_fr_05 [{shapeType=Capsule, length=5.5, radius=4.5, shapeName=dyn_jacket_l_fr_02_capsule, center=<8 0 3>, rotation=<rot 90 0 0>}]
 dyn_jacket_r_fr_MSTR [{shapeType=Capsule, length=1, radius=1}]
 dyn_laces_MATR [{shapeType=Capsule, length=1, radius=1}]
 dyn_laces_l_01 [{shapeType=Capsule, length=3.75, radius=2.5, shapeName=dyn_jacket_l_fr_02_capsule, center=<7.8787 0 1.5>, rotation=<rot 90 0 0>}]
 dyn_laces_l_02 [{shapeType=Capsule, length=3.625, radius=2.5, shapeName=dyn_jacket_l_fr_02_capsule, center=<6.5258 0 1.5>, rotation=<rot 90 0 0>}]
 dyn_laces_r_01 [{shapeType=Capsule, length=3.75, radius=2.5, shapeName=dyn_jacket_l_fr_02_capsule, center=<7.8787 0 1.5>, rotation=<rot 90 0 0>}]
 dyn_laces_r_02 [{shapeType=Capsule, length=4.625, radius=2.5, shapeName=dyn_jacket_l_fr_02_capsule, center=<8.5258 0 1.5>, rotation=<rot 90 0 0>}]
 foot_l [{shapeType=Capsule, length=18.4532, radius=9.4038, center=<-2.4892 6.214 0>, rotation=<rot 0 0 -90>}]
 foot_r [{shapeType=Capsule, length=18.4532, radius=9.4038, center=<2.4892 -6.214 0>, rotation=<rot 0 0 90>}]
 pelvis [{shapeType=Capsule, length=6.35, radius=12.2122, center=<-5.4884 -1.3188 0>}]
 spine_01 [{shapeType=Capsule, length=7.2756, radius=10.4006, center=<-3.1705 0.9137 0>}]
 spine_02 [{shapeType=Capsule, length=5.6506, radius=10.4006, center=<0 0.6443 0>}]
 thigh_twist_01_l [{shapeType=Capsule, length=29.5782, radius=12.0288, center=<3.6306 -1.3849 1.2565>, rotation=<rot 81.2732 179.3758 -2.2807>}]
 thigh_twist_01_r [{shapeType=Capsule, length=24.7032, radius=12.0288, center=<-1.1478 1.4545 -1.4297>, rotation=<rot 84.2159 178.2126 176.5652>}]
}
constraints [
 {bone1Name=dyn_jacket_r_bk_01, bone2Name=dyn_jacket_r_bk_MSTR, swing1LimitDegrees=20, swing2LimitDegrees=25, twistLimitDegrees=2}
 {bone1Name=dyn_jacket_r_bk_02, bone2Name=dyn_jacket_r_bk_01, swing1LimitDegrees=7, swing2LimitDegrees=25, twistLimitDegrees=2}
 {bone1Name=dyn_jacket_l_bk_01, bone2Name=dyn_jacket_l_bk_MSTR, swing1LimitDegrees=20, swing2LimitDegrees=25, twistLimitDegrees=2}
 {bone1Name=dyn_jacket_l_bk_02, bone2Name=dyn_jacket_l_bk_01, swing1LimitDegrees=7, swing2LimitDegrees=25, twistLimitDegrees=2}
 {bone1Name=dyn_jacket_l_fr_01, bone2Name=dyn_jacket_l_fr_MSTR, swing1LimitDegrees=7, swing2LimitDegrees=25, twistLimitDegrees=2}
 {bone1Name=dyn_jacket_l_fr_02, bone2Name=dyn_jacket_l_fr_01, swing1LimitDegrees=7, swing2LimitDegrees=25, twistLimitDegrees=2}
 {bone1Name=dyn_jacket_r_fr_01, bone2Name=dyn_jacket_r_fr_MSTR, swing1LimitDegrees=7, swing2LimitDegrees=25, twistLimitDegrees=2}
 {bone1Name=dyn_jacket_r_fr_02, bone2Name=dyn_jacket_r_fr_01, swing1LimitDegrees=7, swing2LimitDegrees=25, twistLimitDegrees=2}
 {bone1Name=dyn_jacket_l_01, bone2Name=dyn_jacket_l_MSTR, swing1LimitDegrees=7, swing2LimitDegrees=25, twistLimitDegrees=1}
 {bone1Name=dyn_jacket_r_01, bone2Name=dyn_jacket_r_MSTR, swing1LimitDegrees=7, swing2LimitDegrees=25, twistLimitDegrees=2}
 {
  bone1Name=dyn_jacket_r_bk_MSTR
  bone2Name=dyn_jacket_bk_MSTR
  swing1LimitDegrees=45
  swing1Motion=Free
  swing2LimitDegrees=45
  swing2Motion=Free
  twistLimitDegrees=45
  twistMotion=Free
 }
 {
  bone1Name=dyn_jacket_l_bk_MSTR
  bone2Name=dyn_jacket_bk_MSTR
  swing1LimitDegrees=45
  swing1Motion=Free
  swing2LimitDegrees=45
  swing2Motion=Free
  twistLimitDegrees=45
  twistMotion=Free
 }
 {
  bone1Name=dyn_jacket_r_MSTR
  bone2Name=spine_02
  swing1LimitDegrees=45
  swing1Motion=Free
  swing2LimitDegrees=45
  swing2Motion=Free
  twistLimitDegrees=45
  twistMotion=Free
 }
 {
  bone1Name=dyn_jacket_l_MSTR
  bone2Name=spine_02
  swing1LimitDegrees=45
  swing1Motion=Free
  swing2LimitDegrees=45
  swing2Motion=Free
  twistLimitDegrees=45
  twistMotion=Free
 }
 {
  bone1Name=dyn_jacket_l_fr_MSTR
  bone2Name=spine_02
  swing1LimitDegrees=45
  swing1Motion=Free
  swing2LimitDegrees=45
  swing2Motion=Free
  twistLimitDegrees=45
  twistMotion=Free
 }
 {
  bone1Name=dyn_jacket_r_fr_MSTR
  bone2Name=spine_02
  swing1LimitDegrees=45
  swing1Motion=Free
  swing2LimitDegrees=45
  swing2Motion=Free
  twistLimitDegrees=45
  twistMotion=Free
 }
 {
  bone1Name=thigh_twist_01_l
  bone2Name=pelvis
  swing1LimitDegrees=45
  swing1Motion=Free
  swing2LimitDegrees=45
  swing2Motion=Free
  twistLimitDegrees=45
  twistMotion=Free
 }
 {
  bone1Name=spine_01
  bone2Name=pelvis
  swing1LimitDegrees=45
  swing1Motion=Free
  swing2LimitDegrees=45
  swing2Motion=Free
  twistLimitDegrees=45
  twistMotion=Free
 }
 {
  bone1Name=spine_02
  bone2Name=spine_01
  swing1LimitDegrees=45
  swing1Motion=Free
  swing2LimitDegrees=45
  swing2Motion=Free
  twistLimitDegrees=45
  twistMotion=Free
 }
 {bone1Name=dyn_jacket_l_fr_03, bone2Name=dyn_jacket_l_fr_02, swing1LimitDegrees=7, swing2LimitDegrees=25, twistLimitDegrees=10}
 {bone1Name=dyn_jacket_l_fr_04, bone2Name=dyn_jacket_l_fr_03, swing1LimitDegrees=7, swing2LimitDegrees=25, twistLimitDegrees=10}
 {bone1Name=dyn_jacket_a_l_01, bone2Name=dyn_jacket_l_01, swing1LimitDegrees=15, swing2LimitDegrees=35, twistLimitDegrees=10}
 {bone1Name=dyn_jacket_a_l_02, bone2Name=dyn_jacket_a_l_01, swing1LimitDegrees=7, swing2LimitDegrees=35, twistLimitDegrees=10}
 {bone1Name=dyn_jacket_b_l_01, bone2Name=dyn_jacket_l_01, swing1LimitDegrees=45, swing2LimitDegrees=35, twistLimitDegrees=2}
 {bone1Name=dyn_jacket_b_l_02, bone2Name=dyn_jacket_b_l_01, swing1LimitDegrees=7, swing2LimitDegrees=35, twistLimitDegrees=2}
 {bone1Name=dyn_jacket_a_r_01, bone2Name=dyn_jacket_r_01, swing1LimitDegrees=20, swing2LimitDegrees=25, twistLimitDegrees=10}
 {bone1Name=dyn_jacket_a_r_02, bone2Name=dyn_jacket_a_r_01, swing1LimitDegrees=7, swing2LimitDegrees=25, twistLimitDegrees=10}
 {bone1Name=dyn_jacket_b_r_01, bone2Name=dyn_jacket_r_01, swing1LimitDegrees=35, swing2LimitDegrees=35, twistLimitDegrees=2}
 {bone1Name=dyn_jacket_b_r_02, bone2Name=dyn_jacket_b_r_01, swing1LimitDegrees=7, swing2LimitDegrees=25, twistLimitDegrees=10}
 {bone1Name=dyn_jacket_r_fr_04, bone2Name=dyn_jacket_r_fr_03, swing1LimitDegrees=7, swing2LimitDegrees=25, twistLimitDegrees=10}
 {bone1Name=dyn_jacket_r_fr_03, bone2Name=dyn_jacket_r_fr_02, swing1LimitDegrees=7, swing2LimitDegrees=25, twistLimitDegrees=10}
 {bone1Name=dyn_jacket_a_l_03, bone2Name=dyn_jacket_a_l_02, swing1LimitDegrees=45, swing2LimitDegrees=45, twistLimitDegrees=2}
 {bone1Name=dyn_jacket_l_fr_05, bone2Name=dyn_jacket_l_fr_04, swing1LimitDegrees=45, swing2LimitDegrees=45, twistLimitDegrees=2}
 {bone1Name=dyn_jacket_r_fr_05, bone2Name=dyn_jacket_r_fr_04, swing1LimitDegrees=45, swing2LimitDegrees=45, twistLimitDegrees=2}
 {bone1Name=dyn_jacket_r_bk_03, bone2Name=dyn_jacket_r_bk_02, swing1LimitDegrees=45, swing2LimitDegrees=45, twistLimitDegrees=2}
 {bone1Name=dyn_jacket_a_r_03, bone2Name=dyn_jacket_a_r_02, swing1LimitDegrees=45, swing2LimitDegrees=45, twistLimitDegrees=2}
 {bone1Name=dyn_jacket_l_bk_03, bone2Name=dyn_jacket_l_bk_02, swing1LimitDegrees=45, swing2LimitDegrees=45, twistLimitDegrees=2}
 {
  bone1Name=calf_r
  bone2Name=pelvis
  swing1LimitDegrees=45
  swing1Motion=Free
  swing2LimitDegrees=45
  swing2Motion=Free
  twistLimitDegrees=45
  twistMotion=Free
 }
 {
  bone1Name=dyn_laces_MATR
  bone2Name=pelvis
  swing1LimitDegrees=45
  swing1Motion=Free
  swing2LimitDegrees=45
  swing2Motion=Free
  twistLimitDegrees=45
  twistMotion=Free
 }
 {bone1Name=dyn_laces_l_01, bone2Name=dyn_laces_MATR, swing1LimitDegrees=25, swing2LimitDegrees=45, twistLimitDegrees=2}
 {bone1Name=dyn_laces_l_02, bone2Name=dyn_laces_l_01, swing1LimitDegrees=45, swing2LimitDegrees=45, twistLimitDegrees=2}
 {bone1Name=dyn_laces_r_01, bone2Name=dyn_laces_MATR, swing1LimitDegrees=25, swing2LimitDegrees=45, twistLimitDegrees=2}
 {bone1Name=dyn_laces_r_02, bone2Name=dyn_laces_r_01, swing1LimitDegrees=45, swing2LimitDegrees=45, twistLimitDegrees=2}
 {
  bone1Name=calf_twist_01_l
  bone2Name=calf_l
  swing1LimitDegrees=45
  swing1Motion=Free
  swing2LimitDegrees=45
  swing2Motion=Free
  twistLimitDegrees=45
  twistMotion=Free
 }
 {
  bone1Name=calf_l
  bone2Name=pelvis
  swing1LimitDegrees=45
  swing1Motion=Free
  swing2LimitDegrees=45
  swing2Motion=Free
  twistLimitDegrees=45
  twistMotion=Free
 }
 {
  bone1Name=dyn_groin_collision
  bone2Name=pelvis
  swing1LimitDegrees=45
  swing1Motion=Free
  swing2LimitDegrees=45
  swing2Motion=Free
  twistLimitDegrees=45
  twistMotion=Free
 }
 {
  bone1Name=dyn_bag
  bone2Name=pelvis
  swing1LimitDegrees=45
  swing1Motion=Free
  swing2LimitDegrees=45
  swing2Motion=Free
  twistLimitDegrees=45
  twistMotion=Free
 }
 {
  bone1Name=dyn_jacket_a_r_02
  bone2Name=dyn_jacket_r_fr_04
  swing1LimitDegrees=5
  swing2LimitDegrees=45
  swing2Motion=Free
  twistLimitDegrees=45
  twistMotion=Free
 }
 {
  bone1Name=dyn_jacket_b_r_02
  bone2Name=dyn_jacket_r_bk_02
  swing1LimitDegrees=8
  swing2LimitDegrees=45
  swing2Motion=Free
  twistLimitDegrees=45
  twistMotion=Free
 }
 {
  bone1Name=dyn_jacket_a_l_02
  bone2Name=dyn_jacket_l_fr_04
  swing1LimitDegrees=5
  swing2LimitDegrees=45
  swing2Motion=Free
  twistLimitDegrees=45
  twistMotion=Free
 }
 {
  bone1Name=dyn_jacket_b_l_02
  bone2Name=dyn_jacket_l_bk_02
  swing1LimitDegrees=8
  swing2LimitDegrees=45
  swing2Motion=Free
  twistLimitDegrees=45
  twistMotion=Free
 }
 {
  bone1Name=thigh_twist_01_r
  bone2Name=pelvis
  swing1LimitDegrees=45
  swing1Motion=Free
  swing2LimitDegrees=45
  swing2Motion=Free
  twistLimitDegrees=45
  twistMotion=Free
 }
 {
  bone1Name=dyn_jacket_bk_MSTR
  bone2Name=pelvis
  swing1LimitDegrees=45
  swing1Motion=Free
  swing2LimitDegrees=45
  swing2Motion=Free
  twistLimitDegrees=45
  twistMotion=Free
 }
 {
  bone1Name=foot_l
  bone2Name=calf_l
  swing1LimitDegrees=45
  swing1Motion=Free
  swing2LimitDegrees=45
  swing2Motion=Free
  twistLimitDegrees=45
  twistMotion=Free
 }
 {
  bone1Name=calf_twist_01_r
  bone2Name=calf_r
  swing1LimitDegrees=45
  swing1Motion=Free
  swing2LimitDegrees=45
  swing2Motion=Free
  twistLimitDegrees=45
  twistMotion=Free
 }
 {
  bone1Name=foot_r
  bone2Name=calf_r
  swing1LimitDegrees=45
  swing1Motion=Free
  swing2LimitDegrees=45
  swing2Motion=Free
  twistLimitDegrees=45
  twistMotion=Free
 }
]
