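# Physics asset for the F_MED_LightningDragon female body's dynamic jacket and laces, named as the sliding variant.
# Capsule chains for front, back, branching side panels and laces hang from MSTR anchor bodies on spine_02 or pelvis, with swing and tight twist limits.
# Pelvis, spine, groin and bag capsules supply body shapes, alongside a root capsule and a large, low-set root sphere.
# Two extra left-side constraints tie side strands to front and back chains.
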PhysicsAsset /BRCosmetics/Characters/Player/Female/Medium/Bodies/F_MED_LightningDragon/Meshes/F_MED_LightningDragon_Dyn_Jacket_Sliding
bodies {
 dyn_bag [
  {shapeType=Capsule, length=11.25, radius=4.125, center=<9.3679 -1.263 -1.2843>, rotation=<rot 90 0 0>}
  {shapeType=Capsule, length=11.25, radius=4.125, center=<9.3679 2.2819 -1.2681>, rotation=<rot 90 0 0>}
 ]
 dyn_groin_collision [{shapeType=Capsule, length=24.475, radius=19.4622, center=<-0.0451 -0.8793 -9.4455>, rotation=<rot 0.0004 0 92.7469>}]
 dyn_jacket_a_l_01 [{shapeType=Capsule, length=5.5803, radius=6.5415, shapeName=dyn_jacket_l_bk_01_capsule, center=<7 2 4>, rotation=<rot 0.0001 0.0006 99.9999>}]
 dyn_jacket_a_l_02 [{shapeType=Capsule, length=4.3303, radius=7.7915, shapeName=dyn_jacket_l_bk_02_capsule, center=<7 2 4>, rotation=<rot 0.0001 0.0006 99.9998>}]
 dyn_jacket_a_l_03 [{shapeType=Capsule, length=2.875, radius=7.125, shapeName=dyn_jacket_l_bk_02_capsule, center=<8 0 3>, rotation=<rot 0 0 90>}]
 dyn_jacket_a_r_01 [{shapeType=Capsule, length=5.5803, radius=6.5415, shapeName=dyn_jacket_l_bk_01_capsule, center=<7 -2 4>, rotation=<rot 0 -0.0001 85>}]
 dyn_jacket_a_r_02 [{shapeType=Capsule, length=5.3303, radius=6.7915, shapeName=dyn_jacket_l_bk_02_capsule, center=<7 -2 4>, rotation=<rot 0 -0.0001 85>}]
 dyn_jacket_a_r_03 [{shapeType=Capsule, length=2.875, radius=7.125, shapeName=dyn_jacket_l_bk_02_capsule, center=<8 0 3>, rotation=<rot 0 0 90>}]
 dyn_jacket_b_l_01 [{shapeType=Capsule, length=9.125, radius=6.8, shapeName=dyn_jacket_l_bk_01_capsule, center=<7 -2 4>, rotation=<rot -0.0005 0.0006 89.9999>}]
 dyn_jacket_b_l_02 [{shapeType=Capsule, length=9.125, radius=6.8, shapeName=dyn_jacket_l_bk_02_capsule, center=<7 -2 4>, rotation=<rot -0.0005 0.0006 90>}]
 dyn_jacket_b_r_01 [{shapeType=Capsule, length=11, radius=6.3, shapeName=dyn_jacket_l_bk_01_capsule, center=<7 2 4>, rotation=<rot -0.0005 0.0004 90>}]
 dyn_jacket_b_r_02 [{shapeType=Capsule, length=11, radius=6.3, shapeName=dyn_jacket_l_bk_02_capsule, center=<7 2 4>, rotation=<rot -0.0005 0.0004 90>}]
 dyn_jacket_bk_MSTR [{shapeType=Capsule, length=1, radius=1}]
 dyn_jacket_l_01 [{shapeType=Capsule, length=9.0803, radius=6.2915, shapeName=dyn_jacket_l_01_capsule, center=<7 0 4>, rotation=<rot 0 0.0001 85>}]
 dyn_jacket_l_MSTR [{shapeType=Capsule, length=1, radius=1}]
 dyn_jacket_l_bk_01 [{shapeType=Capsule, length=16.5803, radius=7.0415, shapeName=dyn_jacket_l_bk_01_capsule, center=<7 0 4>, rotation=<rot -0.0001 0 95>}]
 dyn_jacket_l_bk_02 [{shapeType=Capsule, length=16.5803, radius=7.0415, shapeName=dyn_jacket_l_bk_02_capsule, center=<7 0 2.5>, rotation=<rot -0.0001 0 95>}]
 dyn_jacket_l_bk_03 [{shapeType=Capsule, length=2.75, radius=7.25, shapeName=dyn_jacket_l_bk_02_capsule, center=<8 0 3>, rotation=<rot 0 0 90>}]
 dyn_jacket_l_bk_MSTR [{shapeType=Capsule, length=1, radius=1}]
 dyn_jacket_l_fr_01 [{shapeType=Capsule, length=5.125, radius=6.875, shapeName=dyn_jacket_l_fr_01_capsule, center=<6 0 5>, rotation=<rot 90 0 0>}]
 dyn_jacket_l_fr_02 [{shapeType=Capsule, length=7, radius=7.475, shapeName=dyn_jacket_l_fr_02_capsule, center=<6 -2 5>, rotation=<rot 0 0 90>}]
 dyn_jacket_l_fr_03 [{shapeType=Capsule, length=7, radius=7.475, shapeName=dyn_jacket_l_fr_02_capsule, center=<6 -2 5>, rotation=<rot 0 0 90>}]
 dyn_jacket_l_fr_04 [{shapeType=Capsule, length=5.75, radius=8.725, shapeName=dyn_jacket_l_fr_02_capsule, center=<6 -2 5>, rotation=<rot 0 0 90>}]
 dyn_jacket_l_fr_05 [{shapeType=Capsule, length=2.875, radius=7.125, shapeName=dyn_jacket_l_fr_02_capsule, center=<8 0 3>, rotation=<rot 0 0 90>}]
 dyn_jacket_l_fr_MSTR [{shapeType=Capsule, length=1, radius=1}]
 dyn_jacket_r_01 [{shapeType=Capsule, length=9.0803, radius=6.2915, shapeName=dyn_jacket_l_01_capsule, center=<7 0 4>, rotation=<rot 90 0 0>}]
 dyn_jacket_r_MSTR [{shapeType=Capsule, length=1, radius=1}]
 dyn_jacket_r_bk_01 [{shapeType=Capsule, length=16.5803, radius=7.0415, shapeName=dyn_jacket_l_bk_01_capsule, center=<7 0 4>, rotation=<rot -0.0001 0 85>}]
 dyn_jacket_r_bk_02 [{shapeType=Capsule, length=16.5803, radius=7.0415, shapeName=dyn_jacket_l_bk_02_capsule, center=<7 0 2.5>, rotation=<rot -0.0001 0 85>}]
 dyn_jacket_r_bk_03 [{shapeType=Capsule, length=2.75, radius=7.25, shapeName=dyn_jacket_l_bk_02_capsule, center=<8 0 3>, rotation=<rot 0 0 90>}]
 dyn_jacket_r_bk_MSTR [{shapeType=Capsule, length=1, radius=1}]
 dyn_jacket_r_fr_01 [{shapeType=Capsule, length=5, radius=7, shapeName=dyn_jacket_l_fr_01_capsule, center=<6 0 5>, rotation=<rot 90 0 0>}]
 dyn_jacket_r_fr_02 [{shapeType=Capsule, length=9.5, radius=6.875, shapeName=dyn_jacket_l_fr_02_capsule, center=<6 3 5>, rotation=<rot 0 0 90>}]
 dyn_jacket_r_fr_03 [{shapeType=Capsule, length=9.5, radius=6.875, shapeName=dyn_jacket_l_fr_02_capsule, center=<6 3 5>, rotation=<rot 0 0 90>}]
 dyn_jacket_r_fr_04 [{shapeType=Capsule, length=8.625, radius=7.75, shapeName=dyn_jacket_l_fr_02_capsule, center=<6 3 5>, rotation=<rot 0 0 90>}]
 dyn_jacket_r_fr_05 [{shapeType=Capsule, length=2.875, radius=7.125, shapeName=dyn_jacket_l_fr_02_capsule, center=<8 0 3>, rotation=<rot 0 0 90>}]
 dyn_jacket_r_fr_MSTR [{shapeType=Capsule, length=1, radius=1}]
 dyn_laces_MATR [{shapeType=Capsule, length=1, radius=1}]
 dyn_laces_l_01 [{shapeType=Capsule, length=6.25, radius=2.25, shapeName=dyn_jacket_l_fr_02_capsule, center=<7.8787 0 1.5>, rotation=<rot 90 0 0>}]
 dyn_laces_l_02 [{shapeType=Capsule, length=5.875, radius=2.375, shapeName=dyn_jacket_l_fr_02_capsule, center=<6.5258 0 1.5>, rotation=<rot 90 0 0>}]
 dyn_laces_r_01 [{shapeType=Capsule, length=6.25, radius=2.25, shapeName=dyn_jacket_l_fr_02_capsule, center=<7.8787 0 1.5>, rotation=<rot 90 0 0>}]
 dyn_laces_r_02 [{shapeType=Capsule, length=6.875, radius=2.375, shapeName=dyn_jacket_l_fr_02_capsule, center=<8.5258 0 1.5>, rotation=<rot 90 0 0>}]
 pelvis [{shapeType=Capsule, length=10.725, radius=12.2122, center=<-7.1958 -1.3188 0>}]
 root [
  {shapeType=Sphere, radius=122.25, shapeName=root_sphere, center=<2.0835 -14.0653 -114.8142>}
  {shapeType=Capsule, length=18.5, radius=30.375, shapeName=root_capsule, center=<0 4.0579 -8.3646>, rotation=<rot -0.4349 20.0189 85.018>}
 ]
 spine_01 [{shapeType=Capsule, length=7.2756, radius=10.4006, center=<-3.1705 0.9137 0>}]
 spine_02 [{shapeType=Capsule, length=5.6506, radius=10.4006, center=<0 0.6443 0>}]
}
constraints [
 {bone1Name=dyn_jacket_r_bk_01, bone2Name=dyn_jacket_r_bk_MSTR, swing1LimitDegrees=5, swing2LimitDegrees=35, twistLimitDegrees=2}
 {bone1Name=dyn_jacket_r_bk_02, bone2Name=dyn_jacket_r_bk_01, swing1LimitDegrees=5, swing2LimitDegrees=35, twistLimitDegrees=2}
 {bone1Name=dyn_jacket_l_bk_01, bone2Name=dyn_jacket_l_bk_MSTR, swing1LimitDegrees=5, swing2LimitDegrees=35, twistLimitDegrees=2}
 {bone1Name=dyn_jacket_l_bk_02, bone2Name=dyn_jacket_l_bk_01, swing1LimitDegrees=5, swing2LimitDegrees=35, twistLimitDegrees=2}
 {bone1Name=dyn_jacket_l_fr_01, bone2Name=dyn_jacket_l_fr_MSTR, swing1LimitDegrees=5, swing2LimitDegrees=35, twistLimitDegrees=2}
 {bone1Name=dyn_jacket_l_fr_02, bone2Name=dyn_jacket_l_fr_01, swing1LimitDegrees=5, swing2LimitDegrees=35, twistLimitDegrees=2}
 {bone1Name=dyn_jacket_r_fr_01, bone2Name=dyn_jacket_r_fr_MSTR, swing1LimitDegrees=5, swing2LimitDegrees=35, twistLimitDegrees=2}
 {bone1Name=dyn_jacket_r_fr_02, bone2Name=dyn_jacket_r_fr_01, swing1LimitDegrees=5, swing2LimitDegrees=35, twistLimitDegrees=2}
 {bone1Name=dyn_jacket_l_01, bone2Name=dyn_jacket_l_MSTR, swing1LimitDegrees=5, swing2LimitDegrees=35, twistLimitDegrees=2}
 {bone1Name=dyn_jacket_r_01, bone2Name=dyn_jacket_r_MSTR, swing1LimitDegrees=5, swing2LimitDegrees=35, twistLimitDegrees=2}
 {
  bone1Name=dyn_jacket_r_bk_MSTR
  bone2Name=dyn_jacket_bk_MSTR
  swing1LimitDegrees=45
  swing1Motion=Free
  swing2LimitDegrees=45
  swing2Motion=Free
  twistLimitDegrees=45
  twistMotion=Free
 }
 {
  bone1Name=dyn_jacket_l_bk_MSTR
  bone2Name=dyn_jacket_bk_MSTR
  swing1LimitDegrees=45
  swing1Motion=Free
  swing2LimitDegrees=45
  swing2Motion=Free
  twistLimitDegrees=45
  twistMotion=Free
 }
 {
  bone1Name=dyn_jacket_r_MSTR
  bone2Name=spine_02
  swing1LimitDegrees=45
  swing1Motion=Free
  swing2LimitDegrees=45
  swing2Motion=Free
  twistLimitDegrees=45
  twistMotion=Free
 }
 {
  bone1Name=dyn_jacket_l_MSTR
  bone2Name=spine_02
  swing1LimitDegrees=45
  swing1Motion=Free
  swing2LimitDegrees=45
  swing2Motion=Free
  twistLimitDegrees=45
  twistMotion=Free
 }
 {
  bone1Name=dyn_jacket_l_fr_MSTR
  bone2Name=spine_02
  swing1LimitDegrees=45
  swing1Motion=Free
  swing2LimitDegrees=45
  swing2Motion=Free
  twistLimitDegrees=45
  twistMotion=Free
 }
 {
  bone1Name=dyn_jacket_r_fr_MSTR
  bone2Name=spine_02
  swing1LimitDegrees=45
  swing1Motion=Free
  swing2LimitDegrees=45
  swing2Motion=Free
  twistLimitDegrees=45
  twistMotion=Free
 }
 {
  bone1Name=spine_01
  bone2Name=pelvis
  swing1LimitDegrees=45
  swing1Motion=Free
  swing2LimitDegrees=45
  swing2Motion=Free
  twistLimitDegrees=45
  twistMotion=Free
 }
 {
  bone1Name=spine_02
  bone2Name=spine_01
  swing1LimitDegrees=45
  swing1Motion=Free
  swing2LimitDegrees=45
  swing2Motion=Free
  twistLimitDegrees=45
  twistMotion=Free
 }
 {bone1Name=dyn_jacket_l_fr_03, bone2Name=dyn_jacket_l_fr_02, swing1LimitDegrees=5, swing2LimitDegrees=35, twistLimitDegrees=2}
 {bone1Name=dyn_jacket_l_fr_04, bone2Name=dyn_jacket_l_fr_03, swing1LimitDegrees=5, swing2LimitDegrees=35, twistLimitDegrees=2}
 {bone1Name=dyn_jacket_a_l_01, bone2Name=dyn_jacket_l_01, swing1LimitDegrees=20, swing2LimitDegrees=35, twistLimitDegrees=2}
 {bone1Name=dyn_jacket_a_l_02, bone2Name=dyn_jacket_a_l_01, swing1LimitDegrees=5, swing2LimitDegrees=35, twistLimitDegrees=2}
 {bone1Name=dyn_jacket_b_l_01, bone2Name=dyn_jacket_l_01, swing1LimitDegrees=20, swing2LimitDegrees=35, twistLimitDegrees=2}
 {bone1Name=dyn_jacket_b_l_02, bone2Name=dyn_jacket_b_l_01, swing1LimitDegrees=5, swing2LimitDegrees=35, twistLimitDegrees=2}
 {bone1Name=dyn_jacket_a_r_01, bone2Name=dyn_jacket_r_01, swing1LimitDegrees=5, swing2LimitDegrees=35, twistLimitDegrees=2}
 {bone1Name=dyn_jacket_a_r_02, bone2Name=dyn_jacket_a_r_01, swing1LimitDegrees=5, swing2LimitDegrees=35, twistLimitDegrees=2}
 {bone1Name=dyn_jacket_b_r_01, bone2Name=dyn_jacket_r_01, swing1LimitDegrees=5, swing2LimitDegrees=35, twistLimitDegrees=2}
 {bone1Name=dyn_jacket_b_r_02, bone2Name=dyn_jacket_b_r_01, swing1LimitDegrees=5, swing2LimitDegrees=35, twistLimitDegrees=2}
 {bone1Name=dyn_jacket_r_fr_04, bone2Name=dyn_jacket_r_fr_03, swing1LimitDegrees=5, swing2LimitDegrees=35, twistLimitDegrees=2}
 {bone1Name=dyn_jacket_r_fr_03, bone2Name=dyn_jacket_r_fr_02, swing1LimitDegrees=5, swing2LimitDegrees=35, twistLimitDegrees=2}
 {bone1Name=dyn_jacket_a_l_03, bone2Name=dyn_jacket_a_l_02, swing1LimitDegrees=45, swing2LimitDegrees=45, twistLimitDegrees=2}
 {bone1Name=dyn_jacket_l_fr_05, bone2Name=dyn_jacket_l_fr_04, swing1LimitDegrees=45, swing2LimitDegrees=45, twistLimitDegrees=2}
 {bone1Name=dyn_jacket_r_fr_05, bone2Name=dyn_jacket_r_fr_04, swing1LimitDegrees=45, swing2LimitDegrees=45, twistLimitDegrees=2}
 {bone1Name=dyn_jacket_r_bk_03, bone2Name=dyn_jacket_r_bk_02, swing1LimitDegrees=45, swing2LimitDegrees=45, twistLimitDegrees=2}
 {bone1Name=dyn_jacket_a_r_03, bone2Name=dyn_jacket_a_r_02, swing1LimitDegrees=45, swing2LimitDegrees=45, twistLimitDegrees=2}
 {bone1Name=dyn_jacket_l_bk_03, bone2Name=dyn_jacket_l_bk_02, swing1LimitDegrees=45, swing2LimitDegrees=45, twistLimitDegrees=2}
 {
  bone1Name=dyn_laces_MATR
  bone2Name=pelvis
  swing1LimitDegrees=45
  swing1Motion=Free
  swing2LimitDegrees=45
  swing2Motion=Free
  twistLimitDegrees=45
  twistMotion=Free
 }
 {bone1Name=dyn_laces_l_01, bone2Name=dyn_laces_MATR, swing1LimitDegrees=25, swing2LimitDegrees=45, twistLimitDegrees=2}
 {bone1Name=dyn_laces_l_02, bone2Name=dyn_laces_l_01, swing1LimitDegrees=45, swing2LimitDegrees=45, twistLimitDegrees=2}
 {bone1Name=dyn_laces_r_01, bone2Name=dyn_laces_MATR, swing1LimitDegrees=25, swing2LimitDegrees=45, twistLimitDegrees=2}
 {bone1Name=dyn_laces_r_02, bone2Name=dyn_laces_r_01, swing1LimitDegrees=45, swing2LimitDegrees=45, twistLimitDegrees=2}
 {
  bone1Name=dyn_groin_collision
  bone2Name=pelvis
  swing1LimitDegrees=45
  swing1Motion=Free
  swing2LimitDegrees=45
  swing2Motion=Free
  twistLimitDegrees=45
  twistMotion=Free
 }
 {
  bone1Name=dyn_bag
  bone2Name=pelvis
  swing1LimitDegrees=45
  swing1Motion=Free
  swing2LimitDegrees=45
  swing2Motion=Free
  twistLimitDegrees=45
  twistMotion=Free
 }
 {
  bone1Name=dyn_jacket_bk_MSTR
  bone2Name=pelvis
  swing1LimitDegrees=45
  swing1Motion=Free
  swing2LimitDegrees=45
  swing2Motion=Free
  twistLimitDegrees=45
  twistMotion=Free
 }
 {
  bone1Name=pelvis
  bone2Name=root
  swing1LimitDegrees=45
  swing1Motion=Free
  swing2LimitDegrees=45
  swing2Motion=Free
  twistLimitDegrees=45
  twistMotion=Free
 }
 {
  bone1Name=dyn_jacket_b_l_02
  bone2Name=dyn_jacket_l_bk_02
  swing1LimitDegrees=45
  swing1Motion=Free
  swing2LimitDegrees=45
  swing2Motion=Free
  twistLimitDegrees=45
  twistMotion=Free
 }
 {
  bone1Name=dyn_jacket_a_l_02
  bone2Name=dyn_jacket_l_fr_04
  swing1LimitDegrees=5
  swing2LimitDegrees=45
  swing2Motion=Free
  twistLimitDegrees=45
  twistMotion=Free
 }
]
