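# Post-process style Anim Blueprint for the female LightningDragon cosmetic body, adding secondary motion to jacket, hood, bag, sleeves and pants.
# It copies the parent mesh pose, runs two deform Control Rigs, then applies attach constraints, look-ats and AnimDynamics.
# Output uses the cached RBAN pose when rigid body is enabled.
# Otherwise jacket fallbacks switch between slide offsets, front-end/emote trails with thigh planar limits, and variable-weighted transforms and trails driven by thigh rotation and movement.
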AnimBlueprint /BRCosmetics/Characters/Player/Female/Medium/Bodies/F_MED_LightningDragon/Meshes/F_MED_LightningDragon_AnimBP
target_skeleton=&/F_MED_LightningDragon_Skeleton
graphs [
 AnimGraph {nodes [
   n1 = Look At - Bone: dyn_jacket_a_l_01(Alpha=1) {
    alphaBoolBlend {blendOption=Linear}
    boneToModify {boneName=dyn_jacket_a_l_01}
    lookAtLocation=<0 5 -17.0327>
    lookAtTarget {boneReference {boneName=calf_l}}
    lookAt_Axis {axis=<1 0>}
   }
   Look At - Bone: dyn_jacket_b_l_01(ComponentPose=@n1, Alpha=1) {
    alphaBoolBlend {blendOption=Linear}
    boneToModify {boneName=dyn_jacket_b_l_01}
    lookAtLocation=<0 -10 -16>
    lookAtTarget {boneReference {boneName=calf_l}}
    lookAt_Axis {axis=<1 0>}
   }
   n2 = Look At - Bone: dyn_jacket_l_01(ComponentPose=Local To Component(LocalPose=Use cached pose 'prePose'()), Alpha=1) {
    alphaBoolBlend {blendOption=Linear}
    boneToModify {boneName=dyn_jacket_l_01}
    lookAtLocation {x=0, z=-13}
    lookAtTarget {boneReference {boneName=calf_l}}
    lookAt_Axis {axis=<1 0>}
   }
   n3 = Look At - Bone: dyn_jacket_r_01(ComponentPose=@n2, Alpha=1) {
    alphaBoolBlend {blendOption=Linear}
    boneToModify {boneName=dyn_jacket_r_01}
    lookAtLocation {x=0, z=13}
    lookAtTarget {boneReference {boneName=calf_r}}
    lookAt_Axis {axis=<1 0>}
   }
   n4 = Look At - Bone: dyn_jacket_r_fr_02(ComponentPose=@n3, Alpha=1) {
    alphaBoolBlend {blendOption=Linear}
    boneToModify {boneName=dyn_jacket_r_fr_02}
    lookAtLocation=<0 -15 4.5>
    lookAtTarget {boneReference {boneName=calf_r}}
    lookAt_Axis {axis=<1 0>}
   }
   n5 = "Transform (Modify) Bone - Bone: dyn_jacket_l_fr_02"(ComponentPose=@n4, Alpha=Get thigh_rot_up_l() -> (thigh_rot_up_l:double), Translation="-5,0,5", Rotation="100,30,-0.000", Scale="1,1,1") {
    rotationMode=BMM_Additive
    rotationSpace=BCS_BoneSpace
    translationMode=BMM_Additive
    translationSpace=BCS_BoneSpace
    alphaBoolBlend {blendOption=Linear}
    boneToModify {boneName=dyn_jacket_l_fr_02}
    rotation {pitch=100, yaw=30}
    translation {x=-5, z=5}
   }
   n6 = "Transform (Modify) Bone - Bone: dyn_jacket_r_fr_02"(
    ComponentPose=@n5
    Alpha=Get thigh_rot_up_r() -> (thigh_rot_up_r:double)
    Translation="-5,0,5"
    Rotation="10.000,-0.000,0.000"
    Scale="1,1,1"
   ) {
    rotationMode=BMM_Additive
    rotationSpace=BCS_BoneSpace
    translationMode=BMM_Additive
    translationSpace=BCS_BoneSpace
    alphaBoolBlend {blendOption=Linear}
    boneToModify {boneName=dyn_jacket_r_fr_02}
    rotation=<rot 10.0001 0 0>
    translation {x=-5, z=5}
   }
   n7 = Trail controller - Bone: dyn_jacket_l_bk_04(ComponentPose=@n6, Alpha=0.15) {
    bUsePlanarLimit=true
    chainLength=4
    alphaScaleBiasClamp {bInterpResult=true}
    planarLimits [{drivingBone {boneName=thigh_twist_01_l}, planeTransform {location=<0 -15 0>, rotation=<rot 0 0 -90>, scale=<1 1 1>}}]
    rotationLimits [
     {limitMax=<180 180 5>, limitMin=<-180 -180 -5>}
     {limitMax=<180 180 5>, limitMin=<-180 -180 -5>}
     {limitMax=<180 180 180>}
     {limitMax=<180 180 180>}
    ]
    rotationOffsets [<0 0 0>, <0 0 0>, <0 0 0>, <0 0 0>]
    trailBone {boneName=dyn_jacket_l_bk_04}
    trailRelaxationSpeed {editorCurveData {keys [<at 0 8 RCIM_Cubic>, {arriveTangent=-11.3652, interpMode=RCIM_Cubic, leaveTangent=-11.3652, tangentMode=RCTM_User, time=1, value=3}]}}
   }
   n8 = Trail controller - Bone: dyn_jacket_r_bk_04(ComponentPose=@n7, Alpha=0.15) {
    bUsePlanarLimit=true
    chainLength=4
    alphaScaleBiasClamp {bInterpResult=true}
    planarLimits [{drivingBone {boneName=thigh_twist_01_r}, planeTransform {location=<0 15 0>, rotation=<rot 0 0 90>, scale=<1 1 1>}}]
    rotationLimits [
     {limitMax=<180 180 5>, limitMin=<-180 -180 -5>}
     {limitMax=<180 180 5>, limitMin=<-180 -180 -5>}
     {limitMax=<180 180 180>}
     {limitMax=<180 180 180>}
    ]
    rotationOffsets [<0 0 0>, <0 0 0>, <0 0 0>, <0 0 0>]
    trailBone {boneName=dyn_jacket_r_bk_04}
    trailRelaxationSpeed {editorCurveData {keys [<at 0 8 RCIM_Cubic>, {arriveTangent=-11.3652, interpMode=RCIM_Cubic, leaveTangent=-11.3652, tangentMode=RCTM_User, time=1, value=3}]}}
   }
   n9 = Trail controller - Bone: dyn_jacket_b_l_03(ComponentPose=@n8, Alpha=0.2) {
    bUsePlanarLimit=true
    chainLength=3
    maxDeltaTime=0.0333
    alphaScaleBiasClamp {bInterpResult=true}
    planarLimits [{drivingBone {boneName=thigh_twist_01_l}, planeTransform {location=<0 0 -10>, rotation=<rot 0 0 -170>, scale=<1 1 1>}}]
    rotationLimits [{limitMax=<180 180 5>, limitMin=<-180 -180 -5>}, {limitMax=<180 180 5>, limitMin=<-180 -180 -5>}, {limitMax=<180 180 180>}]
    rotationOffsets [<0 0 0>, <0 0 0>, <0 0 0>]
    trailBone {boneName=dyn_jacket_b_l_03}
   }
   n10 = Trail controller - Bone: dyn_jacket_b_r_03(ComponentPose=@n9, Alpha=0.2) {
    bUsePlanarLimit=true
    chainLength=3
    maxDeltaTime=0.0333
    alphaScaleBiasClamp {bInterpResult=true}
    planarLimits [{drivingBone {boneName=thigh_twist_01_r}, planeTransform {location=<0 0 12>, rotation=<rot 0 0 10>, scale=<1 1 1>}}]
    rotationLimits [{limitMax=<180 180 5>, limitMin=<-180 -180 -5>}, {limitMax=<180 180 5>, limitMin=<-180 -180 -5>}, {limitMax=<180 180 180>}]
    rotationOffsets [<0 0 0>, <0 0 0>, <0 0 0>]
    trailBone {boneName=dyn_jacket_b_r_03}
   }
   n11 = Trail controller - Bone: dyn_jacket_a_l_04(ComponentPose=@n10, Alpha=0.2) {
    bUsePlanarLimit=true
    chainLength=4
    alphaScaleBiasClamp {bInterpResult=true}
    planarLimits [{drivingBone {boneName=thigh_twist_01_l}, planeTransform {location=<0 0 -10>, rotation=<rot 0 0 170>, scale=<1 1 1>}}]
    rotationLimits [
     {limitMax=<180 180 5>, limitMin=<-180 -180 -5>}
     {limitMax=<180 180 5>, limitMin=<-180 -180 -5>}
     {limitMax=<180 180 180>}
     {limitMax=<180 180 180>}
    ]
    rotationOffsets [<0 0 0>, <0 0 0>, <0 0 0>, <0 0 0>]
    trailBone {boneName=dyn_jacket_a_l_04}
    trailRelaxationSpeed {editorCurveData {keys [<at 0 8 RCIM_Cubic>, {arriveTangent=-11.3652, interpMode=RCIM_Cubic, leaveTangent=-11.3652, tangentMode=RCTM_User, time=1, value=3}]}}
   }
   n12 = Trail controller - Bone: dyn_jacket_l_fr_06(ComponentPose=@n11, Alpha=0.2) {
    bUsePlanarLimit=true
    chainLength=5
    maxDeltaTime=0.0333
    alphaScaleBiasClamp {bInterpResult=true}
    planarLimits [{drivingBone {boneName=thigh_twist_01_l}, planeTransform {location=<0 10 0>, rotation=<rot 0 0 115>, scale=<1 1 1>}}]
    rotationLimits [
     {limitMax=<180 180 5>, limitMin=<-180 -180 -5>}
     {limitMax=<180 180 5>, limitMin=<-180 -180 -5>}
     {limitMax=<180 180 180>}
     {limitMax=<180 180 180>}
     {limitMax=<180 180 180>}
    ]
    rotationOffsets [<0 0 0>, <0 0 0>, <0 0 0>, <0 0 0>, <0 0 0>]
    trailBone {boneName=dyn_jacket_l_fr_06}
    trailRelaxationSpeed {editorCurveData {keys [<at 0 8 RCIM_Cubic>, {arriveTangent=-11.3652, interpMode=RCIM_Cubic, leaveTangent=-11.3652, tangentMode=RCTM_User, time=1, value=3}]}}
   }
   n13 = Trail controller - Bone: dyn_jacket_a_r_04(ComponentPose=@n12, Alpha=0.2) {
    bUsePlanarLimit=true
    chainLength=4
    alphaScaleBiasClamp {bInterpResult=true}
    planarLimits [{drivingBone {boneName=thigh_twist_01_r}, planeTransform {location=<0 0 12>, rotation=<rot 0 0 -10>, scale=<1 1 1>}}]
    rotationLimits [
     {limitMax=<180 180 5>, limitMin=<-180 -180 -5>}
     {limitMax=<180 180 5>, limitMin=<-180 -180 -5>}
     {limitMax=<180 180 180>}
     {limitMax=<180 180 180>}
    ]
    rotationOffsets [<0 0 0>, <0 0 0>, <0 0 0>, <0 0 0>]
    trailBone {boneName=dyn_jacket_a_r_04}
    trailRelaxationSpeed {editorCurveData {keys [<at 0 8 RCIM_Cubic>, {arriveTangent=-11.3652, interpMode=RCIM_Cubic, leaveTangent=-11.3652, tangentMode=RCTM_User, time=1, value=3}]}}
   }
   n14 = Trail controller - Bone: dyn_jacket_r_fr_06(ComponentPose=@n13, Alpha=0.2) {
    bUsePlanarLimit=true
    chainLength=5
    maxDeltaTime=0.0333
    alphaScaleBiasClamp {bInterpResult=true}
    planarLimits [{drivingBone {boneName=thigh_twist_01_r}, planeTransform {location=<0 -10 0>, rotation=<rot 0 0 -75>, scale=<1 1 1>}}]
    rotationLimits [
     {limitMax=<180 180 5>, limitMin=<-180 -180 -5>}
     {limitMax=<180 180 5>, limitMin=<-180 -180 -5>}
     {limitMax=<180 180 180>}
     {limitMax=<180 180 180>}
     {limitMax=<180 180 180>}
    ]
    rotationOffsets [<0 0 0>, <0 0 0>, <0 0 0>, <0 0 0>, <0 0 0>]
    trailBone {boneName=dyn_jacket_r_fr_06}
    trailRelaxationSpeed {editorCurveData {keys [<at 0 8 RCIM_Cubic>, {arriveTangent=-11.3652, interpMode=RCIM_Cubic, leaveTangent=-11.3652, tangentMode=RCTM_User, time=1, value=3}]}}
   }
   n15 = "Transform (Modify) Bone - Bone: dyn_jacket_l_fr_02"(
    ComponentPose=Local To Component(LocalPose=Use cached pose 'baseNoRBANPose'())
    Alpha=Get thigh_rot_up_01_l() -> (thigh_rot_up_01_l:double)
    Translation="2,0,0"
    Rotation="-10,-35,-40"
    Scale="1,1,1"
   ) {
    rotationMode=BMM_Additive
    rotationSpace=BCS_BoneSpace
    translationMode=BMM_Additive
    translationSpace=BCS_BoneSpace
    alphaBoolBlend {blendOption=Linear}
    boneToModify {boneName=dyn_jacket_l_fr_02}
    rotation=<rot -10 -35 -40>
    translation {x=2}
   }
   n16 = "Transform (Modify) Bone - Bone: dyn_jacket_l_fr_02"(
    ComponentPose=@n15
    Alpha=Get thigh_rot_up_02_l() -> (thigh_rot_up_02_l:double)
    Translation="22.800,10.545,3.317"
    Rotation="-8.615,-75.649,27.616"
    Scale="1,1,1"
   ) {
    rotationMode=BMM_Additive
    rotationSpace=BCS_BoneSpace
    translationMode=BMM_Additive
    translationSpace=BCS_BoneSpace
    alphaBoolBlend {blendOption=Linear}
    boneToModify {boneName=dyn_jacket_l_fr_02}
    rotation=<rot -8.615 -75.649 27.616>
    translation=<22.8 10.545 3.317>
   }
   n17 = Get thigh_rot_up_l()
   n18 = "Transform (Modify) Bone - Bone: dyn_jacket_a_l_01"(ComponentPose=@n16, Alpha=@n17, Translation="-4.132,-6.297,6.174", Rotation="0.001,-100.000,0.015", Scale="1,1,1") {
    rotationMode=BMM_Additive
    rotationSpace=BCS_BoneSpace
    translationMode=BMM_Additive
    translationSpace=BCS_BoneSpace
    alphaBoolBlend {blendOption=Linear}
    boneToModify {boneName=dyn_jacket_a_l_01}
    rotation=<rot 0.001 -100 0.015>
    translation=<-4.132 -6.297 6.174>
   }
   n19 = "Transform (Modify) Bone - Bone: dyn_jacket_b_l_01"(ComponentPose=@n18, Alpha=@n17, Translation="-12,4.3,7.5", Rotation="35,-95,0", Scale="1,1,1") {
    rotationMode=BMM_Additive
    rotationSpace=BCS_BoneSpace
    translationMode=BMM_Additive
    translationSpace=BCS_BoneSpace
    alphaBoolBlend {blendOption=Linear}
    boneToModify {boneName=dyn_jacket_b_l_01}
    rotation {pitch=35, yaw=-95}
    translation=<-12 4.3 7.5>
   }
   n20 = "Transform (Modify) Bone - Bone: dyn_jacket_l_bk_MSTR"(ComponentPose=@n19, Alpha=@n17, Translation="-19,7.6,-6.3", Rotation="95,0,30", Scale="1,1,1") {
    rotationMode=BMM_Additive
    rotationSpace=BCS_BoneSpace
    translationMode=BMM_Additive
    translationSpace=BCS_BoneSpace
    alphaBoolBlend {blendOption=Linear}
    boneToModify {boneName=dyn_jacket_l_bk_MSTR}
    rotation {pitch=95, roll=30}
    translation=<-19 7.6 -6.3>
   }
   n21 = Get thigh_rot_up_r()
   n22 = "Transform (Modify) Bone - Bone: dyn_jacket_r_fr_02"(ComponentPose=@n20, Alpha=@n21, Translation="21.036,3.697,7.451", Rotation="-49.980,89.994,-10.006", Scale="1,1,1") {
    rotationMode=BMM_Additive
    rotationSpace=BCS_BoneSpace
    translationMode=BMM_Additive
    translationSpace=BCS_BoneSpace
    alphaBoolBlend {blendOption=Linear}
    boneToModify {boneName=dyn_jacket_r_fr_02}
    rotation=<rot -49.98 89.9939 -10.0059>
    translation=<21.036 3.697 7.451>
   }
   n23 = "Transform (Modify) Bone - Bone: dyn_jacket_a_r_01"(ComponentPose=@n22, Alpha=@n21, Translation="-1.823,5.988,5.500", Rotation="-0.000,100.000,0.000", Scale="1,1,1") {
    rotationMode=BMM_Additive
    rotationSpace=BCS_BoneSpace
    translationMode=BMM_Additive
    translationSpace=BCS_BoneSpace
    alphaBoolBlend {blendOption=Linear}
    boneToModify {boneName=dyn_jacket_a_r_01}
    rotation=<rot 0 100 0>
    translation=<-1.823 5.988 5.5>
   }
   n24 = "Transform (Modify) Bone - Bone: dyn_jacket_b_r_01"(ComponentPose=@n23, Alpha=@n21, Translation="-12.162,-3.451,6.529", Rotation="34.410,101.038,3.476", Scale="1,1,1") {
    rotationMode=BMM_Additive
    rotationSpace=BCS_BoneSpace
    translationMode=BMM_Additive
    translationSpace=BCS_BoneSpace
    alphaBoolBlend {blendOption=Linear}
    boneToModify {boneName=dyn_jacket_b_r_01}
    rotation=<rot 34.41 101.038 3.476>
    translation=<-12.162 -3.451 6.529>
   }
   n25 = "Transform (Modify) Bone - Bone: dyn_jacket_r_bk_MSTR"(ComponentPose=@n24, Alpha=@n21, Translation="-20.352,-11.291,-8.846", Rotation="88.050,99.391,69.052", Scale="1,1,1") {
    rotationMode=BMM_Additive
    rotationSpace=BCS_BoneSpace
    translationMode=BMM_Additive
    translationSpace=BCS_BoneSpace
    alphaBoolBlend {blendOption=Linear}
    boneToModify {boneName=dyn_jacket_r_bk_MSTR}
    rotation=<rot 88.05 99.3907 69.0517>
    translation=<-20.352 -11.291 -8.846>
   }
   n26 = Get BwdMovement() -> (BwdMovement:double)
   n27 = Trail controller - Bone: dyn_jacket_l_bk_04(ComponentPose=@n25, Alpha=@n26) {
    bUsePlanarLimit=true
    chainLength=4
    alphaScaleBiasClamp {bInterpResult=true}
    rotationLimits [
     {limitMax=<180 180 5>, limitMin=<-180 -180 -5>}
     {limitMax=<180 180 5>, limitMin=<-180 -180 -5>}
     {limitMax=<180 180 180>}
     {limitMax=<180 180 180>}
    ]
    rotationOffsets [<0 0 0>, <0 0 0>, <0 0 0>, <0 0 0>]
    trailBone {boneName=dyn_jacket_l_bk_04}
    trailRelaxationSpeed {editorCurveData {keys [<at 0 8 RCIM_Cubic>, {arriveTangent=-11.3652, interpMode=RCIM_Cubic, leaveTangent=-11.3652, tangentMode=RCTM_User, time=1, value=3}]}}
   }
   n28 = Trail controller - Bone: dyn_jacket_r_bk_04(ComponentPose=@n27, Alpha=@n26) {
    bUsePlanarLimit=true
    chainLength=4
    alphaScaleBiasClamp {bInterpResult=true}
    rotationLimits [
     {limitMax=<180 180 5>, limitMin=<-180 -180 -5>}
     {limitMax=<180 180 5>, limitMin=<-180 -180 -5>}
     {limitMax=<180 180 180>}
     {limitMax=<180 180 180>}
    ]
    rotationOffsets [<0 0 0>, <0 0 0>, <0 0 0>, <0 0 0>]
    trailBone {boneName=dyn_jacket_r_bk_04}
    trailRelaxationSpeed {editorCurveData {keys [<at 0 8 RCIM_Cubic>, {arriveTangent=-11.3652, interpMode=RCIM_Cubic, leaveTangent=-11.3652, tangentMode=RCTM_User, time=1, value=3}]}}
   }
   n29 = Trail controller - Bone: dyn_jacket_b_l_03(ComponentPose=@n28, Alpha=Get BwdLeftMovement() -> (BwdLeftMovement:double)) {
    bUsePlanarLimit=true
    chainLength=3
    maxDeltaTime=0.0333
    alphaScaleBiasClamp {bInterpResult=true}
    rotationLimits [{limitMax=<180 180 5>, limitMin=<-180 -180 -5>}, {limitMax=<180 180 5>, limitMin=<-180 -180 -5>}, {limitMax=<180 180 180>}]
    rotationOffsets [<0 0 0>, <0 0 0>, <0 0 0>]
    trailBone {boneName=dyn_jacket_b_l_03}
   }
   n30 = Trail controller - Bone: dyn_jacket_b_r_03(ComponentPose=@n29, Alpha=Get BwdRightMovement() -> (BwdRightMovement:double)) {
    bUsePlanarLimit=true
    chainLength=3
    maxDeltaTime=0.0333
    alphaScaleBiasClamp {bInterpResult=true}
    rotationLimits [{limitMax=<180 180 5>, limitMin=<-180 -180 -5>}, {limitMax=<180 180 5>, limitMin=<-180 -180 -5>}, {limitMax=<180 180 180>}]
    rotationOffsets [<0 0 0>, <0 0 0>, <0 0 0>]
    trailBone {boneName=dyn_jacket_b_r_03}
   }
   n31 = Get LeftMovement() -> (LeftMovement:double)
   n32 = Trail controller - Bone: dyn_jacket_a_l_04(ComponentPose=@n30, Alpha=@n31) {
    bUsePlanarLimit=true
    chainLength=4
    alphaScaleBiasClamp {bInterpResult=true}
    rotationLimits [
     {limitMax=<180 180 5>, limitMin=<-180 -180 -5>}
     {limitMax=<180 180 5>, limitMin=<-180 -180 -5>}
     {limitMax=<180 180 180>}
     {limitMax=<180 180 180>}
    ]
    rotationOffsets [<0 0 0>, <0 0 0>, <0 0 0>, <0 0 0>]
    trailBone {boneName=dyn_jacket_a_l_04}
    trailRelaxationSpeed {editorCurveData {keys [<at 0 8 RCIM_Cubic>, {arriveTangent=-11.3652, interpMode=RCIM_Cubic, leaveTangent=-11.3652, tangentMode=RCTM_User, time=1, value=3}]}}
   }
   n33 = Trail controller - Bone: dyn_jacket_l_fr_06(ComponentPose=@n32, Alpha=@n31) {
    bUsePlanarLimit=true
    chainLength=5
    maxDeltaTime=0.0333
    alphaScaleBiasClamp {bInterpResult=true}
    rotationLimits [
     {limitMax=<180 180 5>, limitMin=<-180 -180 -5>}
     {limitMax=<180 180 5>, limitMin=<-180 -180 -5>}
     {limitMax=<180 180 180>}
     {limitMax=<180 180 180>}
     {limitMax=<180 180 180>}
    ]
    rotationOffsets [<0 0 0>, <0 0 0>, <0 0 0>, <0 0 0>, <0 0 0>]
    trailBone {boneName=dyn_jacket_l_fr_06}
    trailRelaxationSpeed {editorCurveData {keys [<at 0 8 RCIM_Cubic>, {arriveTangent=-11.3652, interpMode=RCIM_Cubic, leaveTangent=-11.3652, tangentMode=RCTM_User, time=1, value=3}]}}
   }
   n34 = Get RightMovement() -> (RightMovement:double)
   n35 = Trail controller - Bone: dyn_jacket_a_r_04(ComponentPose=@n33, Alpha=@n34) {
    bUsePlanarLimit=true
    chainLength=4
    alphaScaleBiasClamp {bInterpResult=true}
    rotationLimits [
     {limitMax=<180 180 5>, limitMin=<-180 -180 -5>}
     {limitMax=<180 180 5>, limitMin=<-180 -180 -5>}
     {limitMax=<180 180 180>}
     {limitMax=<180 180 180>}
    ]
    rotationOffsets [<0 0 0>, <0 0 0>, <0 0 0>, <0 0 0>]
    trailBone {boneName=dyn_jacket_a_r_04}
    trailRelaxationSpeed {editorCurveData {keys [<at 0 8 RCIM_Cubic>, {arriveTangent=-11.3652, interpMode=RCIM_Cubic, leaveTangent=-11.3652, tangentMode=RCTM_User, time=1, value=3}]}}
   }
   n36 = Trail controller - Bone: dyn_jacket_r_fr_06(ComponentPose=@n35, Alpha=@n34) {
    bUsePlanarLimit=true
    chainLength=5
    maxDeltaTime=0.0333
    alphaScaleBiasClamp {bInterpResult=true}
    rotationLimits [
     {limitMax=<180 180 5>, limitMin=<-180 -180 -5>}
     {limitMax=<180 180 5>, limitMin=<-180 -180 -5>}
     {limitMax=<180 180 180>}
     {limitMax=<180 180 180>}
     {limitMax=<180 180 180>}
    ]
    rotationOffsets [<0 0 0>, <0 0 0>, <0 0 0>, <0 0 0>, <0 0 0>]
    trailBone {boneName=dyn_jacket_r_fr_06}
    trailRelaxationSpeed {editorCurveData {keys [<at 0 8 RCIM_Cubic>, {arriveTangent=-11.3652, interpMode=RCIM_Cubic, leaveTangent=-11.3652, tangentMode=RCTM_User, time=1, value=3}]}}
   }
   n37 = Blend Poses by bool(
    BlendPose_0=Component To Local(ComponentPose=@n14)
    BlendPose_1=Component To Local(ComponentPose=@n36)
    BlendTime_1=0.3
    bActiveValue=Get isInFrontEndOrEmoting() -> (isInFrontEndOrEmoting:bool)
   ) {blendType=Linear}
   n38 = "Transform (Modify) Bone - Bone: dyn_jacket_l_bk_01"(
    ComponentPose=Local To Component(LocalPose=Use cached pose 'baseNoRBANPose'())
    Alpha=1
    Translation="-12,0,3"
    Rotation="0,0,0"
    Scale="1,1,1"
   ) {
    rotationMode=BMM_Additive
    rotationSpace=BCS_BoneSpace
    scaleMode=BMM_Additive
    scaleSpace=BCS_BoneSpace
    translationMode=BMM_Additive
    translationSpace=BCS_BoneSpace
    alphaBoolBlend {blendOption=Linear}
    boneToModify {boneName=dyn_jacket_l_bk_01}
    translation {x=-12, z=3}
   }
   n39 = "Transform (Modify) Bone - Bone: dyn_jacket_b_l_01"(ComponentPose=@n38, Alpha=1, Translation="-11,0,2", Rotation="0,0,0", Scale="1,1,1") {
    rotationMode=BMM_Additive
    rotationSpace=BCS_BoneSpace
    scaleMode=BMM_Additive
    scaleSpace=BCS_BoneSpace
    translationMode=BMM_Additive
    translationSpace=BCS_BoneSpace
    alphaBoolBlend {blendOption=Linear}
    boneToModify {boneName=dyn_jacket_b_l_01}
    translation {x=-11, z=2}
   }
   n40 = "Transform (Modify) Bone - Bone: dyn_jacket_r_bk_01"(ComponentPose=@n39, Alpha=1, Translation="-12,0,3", Rotation="0,0,0", Scale="1,1,1") {
    rotationMode=BMM_Additive
    rotationSpace=BCS_BoneSpace
    scaleMode=BMM_Additive
    scaleSpace=BCS_BoneSpace
    translationMode=BMM_Additive
    translationSpace=BCS_BoneSpace
    alphaBoolBlend {blendOption=Linear}
    boneToModify {boneName=dyn_jacket_r_bk_01}
    translation {x=-12, z=3}
   }
   n41 = "Transform (Modify) Bone - Bone: dyn_jacket_b_r_01"(ComponentPose=@n40, Alpha=1, Translation="-11,0,2", Rotation="0,0,0", Scale="1,1,1") {
    rotationMode=BMM_Additive
    rotationSpace=BCS_BoneSpace
    scaleMode=BMM_Additive
    scaleSpace=BCS_BoneSpace
    translationMode=BMM_Additive
    translationSpace=BCS_BoneSpace
    alphaBoolBlend {blendOption=Linear}
    boneToModify {boneName=dyn_jacket_b_r_01}
    translation {x=-11, z=2}
   }
   n42 = Component To Local(ComponentPose=@n41)
   n43 = Get bIsSliding()
   Output Pose(Result=Blend Poses by bool(
     BlendPose_0=Use cached pose 'RBANPose'()
     BlendPose_1=Blend Poses by bool(BlendPose_0=@n42, BlendPose_1=@n37, BlendTime_1=0.3, bActiveValue=@n43) {blendType=Linear}
     bActiveValue=Get bRigidBodyEnabled()
    ) {blendType=Linear}) {layerGroup=None}
   n44 = Control Rig(Source=Copy Pose From Mesh() {bUseAttachedParent=true}) {
    alphaBoolBlend {blendOption=Linear}
    controlRigAssetReference {blueprintRigClass=&/F_MED_LightningDragon_Deform_CtrlRig_twists.F_MED_LightningDragon_Deform_CtrlRig_twists_C}
   }
   n45 = Control Rig(Source=@n44) {
    alphaBoolBlend {blendOption=Linear}
    controlRigAssetReference {blueprintRigClass=&/F_MED_LightningDragon_DeformRig_ControlRig.F_MED_LightningDragon_DeformRig_ControlRig_C}
   }
   n46 = "Constraint - attach_backpack (PPPP)"!(
    ComponentPose=Local To Component(LocalPose=@n45)
    Alpha=1
    ConstraintWeights_0=0.1443
    ConstraintWeights_1:float=0.8
    ConstraintWeights_2:float=0.0139
    ConstraintWeights_3:float=0.0418
   ) -> (Pose) {
    alphaBoolBlend {blendOption=Linear}
    boneToModify {boneName=attach_backpack}
    constraintSetup [{targetBone {boneName=spine_04}}, {targetBone {boneName=spine_05}}, {targetBone {boneName=clavicle_l}}, {targetBone {boneName=clavicle_r}}]
    constraintWeights [1,1,1,1]
   }
   n47 = "Constraint - attach_cape (PPP)"!(ComponentPose=@n46, Alpha=1, ConstraintWeights_0=0.923, ConstraintWeights_1:float=0.028, ConstraintWeights_2:float=0.049) -> (Pose) {
    alphaBoolBlend {blendOption=Linear}
    boneToModify {boneName=attach_cape}
    constraintSetup [{targetBone {boneName=spine_05}}, {targetBone {boneName=clavicle_l}}, {targetBone {boneName=clavicle_r}}]
    constraintWeights [1,1,1]
   }
   n48 = "Transform (Modify) Bone - Bone: dyn_hood_bk"(ComponentPose=@n47, Alpha=Get head_rot_bk() -> (head_rot_bk:double), Translation="-6,0,2", Rotation="15,0,0", Scale="1,1,1") {
    rotationMode=BMM_Additive
    rotationSpace=BCS_BoneSpace
    translationMode=BMM_Additive
    translationSpace=BCS_BoneSpace
    alphaBoolBlend {blendOption=Linear}
    boneToModify {boneName=dyn_hood_bk}
    rotation {pitch=15}
    translation {x=-6, z=2}
   }
   n49 = "Transform (Modify) Bone - Bone: dyn_hood_r"(ComponentPose=@n48, Alpha=Get head_rot_r() -> (head_rot_r:double), Translation="-3,0,2", Rotation="0,0,0", Scale="1,1,1") {
    rotationMode=BMM_Additive
    rotationSpace=BCS_BoneSpace
    translationMode=BMM_Additive
    translationSpace=BCS_BoneSpace
    alphaBoolBlend {blendOption=Linear}
    boneToModify {boneName=dyn_hood_r}
    translation {x=-3, z=2}
   }
   n50 = Look At - Bone: dyn_bag_MSTR(ComponentPose=@n49, Alpha=1) {
    bUseLookUpAxis=true
    alphaBoolBlend {blendOption=Linear}
    boneToModify {boneName=dyn_bag_MSTR}
    lookAtLocation {x=0}
    lookAtTarget {boneReference {boneName=dyn_bag_TGT}}
    lookAt_Axis {axis=<1 0>}
    lookUp_Axis {axis=<0 1>}
   }
   n51 = Anim Dynamics - Bone: dyn_sleeves_r_MSTR!(ComponentPose=@n50, Alpha=1, GravityScale:float=1, ExternalForce="0,0,0") -> (Pose) {
    angularDampingOverride=1
    angularSpringConstant=500
    bGravityOverrideInSimSpace=true
    bOverrideAngularDamping=true
    bUseGravityOverride=true
    simulationSpace=BoneRelative
    alphaBoolBlend {blendOption=Linear}
    boundBone {boneName=dyn_sleeves_r_MSTR}
    componentAppliedLinearAccClamp=<3000 3000 3000>
    componentLinearAccScale {y=0.05}
    componentLinearVelScale {y=0.1}
    gravityOverride {y=500, z=300}
    physicsBodyDefinitions [{
      sphereCollisionRadius=0
      boundBone {boneName=dyn_sleeves_r_MSTR}
      boxExtents=<15 20 20>
      constraintSetup {angularLimitsMax=<0 0 26>, angularLimitsMin=<0 0 -30>, angularTarget=<0 0 0>}
      localJointOffset=<5 0 0>
     }]
    relativeSpaceBone {boneName=lowerarm_r}
   }
   n52 = Anim Dynamics - Bone: dyn_sleeves_r_up!(ComponentPose=@n51, Alpha=1, GravityScale:float=1, ExternalForce="0,0,0") -> (Pose) {
    angularDampingOverride=0.999
    angularSpringConstant=500
    bGravityOverrideInSimSpace=true
    bOverrideAngularDamping=true
    bUseGravityOverride=true
    simulationSpace=BoneRelative
    alphaBoolBlend {blendOption=Linear}
    boundBone {boneName=dyn_sleeves_r_up}
    componentAppliedLinearAccClamp=<3000 3000 3000>
    componentLinearAccScale=<0.2 0.2 0.2>
    componentLinearVelScale=<0.2 0.2 0.2>
    gravityOverride {y=500, z=100}
    physicsBodyDefinitions [{
      sphereCollisionRadius=0
      boundBone {boneName=dyn_sleeves_r_up}
      boxExtents=<15 20 20>
      constraintSetup {angularLimitsMax=<0 10 2>, angularLimitsMin=<0 0 -2>, angularTarget=<0 0 0>}
      localJointOffset=<5 0 0>
     }]
    relativeSpaceBone {boneName=lowerarm_r}
   }
   n53 = Anim Dynamics - Bone: dyn_sleeves_r_dn!(ComponentPose=@n52, Alpha=1, GravityScale:float=1, ExternalForce="0,0,0") -> (Pose) {
    angularDampingOverride=0.99
    angularSpringConstant=1000
    bAngularSpring=true
    bGravityOverrideInSimSpace=true
    bOverrideAngularDamping=true
    bUseGravityOverride=true
    simulationSpace=RootRelative
    alphaBoolBlend {blendOption=Linear}
    boundBone {boneName=dyn_sleeves_r_dn}
    componentAppliedLinearAccClamp=<3000 3000 3000>
    componentLinearAccScale=<0.2 0.2 0.2>
    componentLinearVelScale=<0.3 0.3 0.3>
    gravityOverride {z=-500}
    physicsBodyDefinitions [{
      sphereCollisionRadius=0
      boundBone {boneName=dyn_sleeves_r_dn}
      boxExtents=<20 20 10>
      constraintSetup {angularLimitsMax=<0 6 3>, angularLimitsMin=<0 -6 -3>, angularTarget=<0 0 0>}
      localJointOffset=<5 0 4>
     }]
    relativeSpaceBone {boneName=lowerarm_r}
   }
   n54 = Anim Dynamics - Bone: dyn_sleeves_r_ft!(ComponentPose=@n53, Alpha=1, GravityScale:float=1, ExternalForce="0,0,0") -> (Pose) {
    angularDampingOverride=1
    angularSpringConstant=500
    bGravityOverrideInSimSpace=true
    bOverrideAngularDamping=true
    bUseGravityOverride=true
    simulationSpace=BoneRelative
    alphaBoolBlend {blendOption=Linear}
    boundBone {boneName=dyn_sleeves_r_ft}
    componentAppliedLinearAccClamp=<3000 3000 3000>
    componentLinearAccScale=<0.05 0.15 0.05>
    componentLinearVelScale=<0.3 0.3 0.3>
    gravityOverride=<-500 -300>
    physicsBodyDefinitions [{
      sphereCollisionRadius=0
      boundBone {boneName=dyn_sleeves_r_ft}
      boxExtents=<10 20 20>
      constraintSetup {angularLimitsMax=<0 8 5>, angularLimitsMin=<0 -2 -5>, angularTarget=<0 0 0>}
      localJointOffset=<5 1 2>
     }]
    relativeSpaceBone {boneName=lowerarm_r}
   }
   n55 = Anim Dynamics - Bone: dyn_sleeves_r_bk!(ComponentPose=@n54, Alpha=1, GravityScale:float=1, ExternalForce="0,0,0") -> (Pose) {
    angularDampingOverride=1
    angularSpringConstant=500
    bGravityOverrideInSimSpace=true
    bOverrideAngularDamping=true
    bUseGravityOverride=true
    simulationSpace=BoneRelative
    alphaBoolBlend {blendOption=Linear}
    boundBone {boneName=dyn_sleeves_r_bk}
    componentAppliedLinearAccClamp=<3000 3000 3000>
    componentLinearAccScale=<0.05 0.15 0.05>
    componentLinearVelScale=<0.3 0.3 0.3>
    gravityOverride=<-500 300>
    physicsBodyDefinitions [{
      sphereCollisionRadius=0
      boundBone {boneName=dyn_sleeves_r_bk}
      boxExtents=<10 20 20>
      constraintSetup {angularLimitsMax=<0 8 5>, angularLimitsMin=<0 -2 -5>, angularTarget=<0 0 0>}
      localJointOffset=<5 -1 2>
     }]
    relativeSpaceBone {boneName=lowerarm_r}
   }
   n56 = "Transform (Modify) Bone - Bone: dyn_sleeves_r_MSTR"(ComponentPose=@n55, Alpha=Get hand_rot_r() -> (hand_rot_r:double), Translation="0,0,-2.5", Rotation="0,0,0", Scale="1,1,1") {
    translationMode=BMM_Additive
    translationSpace=BCS_BoneSpace
    alphaBoolBlend {blendOption=Linear}
    boneToModify {boneName=dyn_sleeves_r_MSTR}
    translation {z=-2.5}
   }
   n57 = Anim Dynamics - Bone: dyn_sleeves_l_MSTR!(ComponentPose=@n56, Alpha=1, GravityScale:float=1, ExternalForce="0,0,0") -> (Pose) {
    angularDampingOverride=1
    angularSpringConstant=500
    bGravityOverrideInSimSpace=true
    bOverrideAngularDamping=true
    bUseGravityOverride=true
    simulationSpace=BoneRelative
    alphaBoolBlend {blendOption=Linear}
    boundBone {boneName=dyn_sleeves_l_MSTR}
    componentAppliedLinearAccClamp=<3000 3000 3000>
    componentLinearAccScale {y=0.05}
    componentLinearVelScale {y=0.1}
    gravityOverride {y=-500, z=-300}
    physicsBodyDefinitions [{
      sphereCollisionRadius=0
      boundBone {boneName=dyn_sleeves_l_MSTR}
      boxExtents=<15 20 20>
      constraintSetup {angularLimitsMax=<0 0 26>, angularLimitsMin=<0 0 -30>, angularTarget=<0 0 0>}
      localJointOffset=<5 0 0>
     }]
    relativeSpaceBone {boneName=lowerarm_l}
   }
   n58 = Anim Dynamics - Bone: dyn_sleeves_l_up!(ComponentPose=@n57, Alpha=1, GravityScale:float=1, ExternalForce="0,0,0") -> (Pose) {
    angularDampingOverride=0.999
    angularSpringConstant=500
    bGravityOverrideInSimSpace=true
    bOverrideAngularDamping=true
    bUseGravityOverride=true
    simulationSpace=BoneRelative
    alphaBoolBlend {blendOption=Linear}
    boundBone {boneName=dyn_sleeves_l_up}
    componentAppliedLinearAccClamp=<3000 3000 3000>
    componentLinearAccScale=<0.2 0.2 0.2>
    componentLinearVelScale=<0.2 0.2 0.2>
    gravityOverride {y=-500, z=100}
    physicsBodyDefinitions [{
      sphereCollisionRadius=0
      boundBone {boneName=dyn_sleeves_l_up}
      boxExtents=<15 20 20>
      constraintSetup {angularLimitsMax=<0 10 2>, angularLimitsMin=<0 0 -2>, angularTarget=<0 0 0>}
      localJointOffset=<5 0 0>
     }]
    relativeSpaceBone {boneName=lowerarm_l}
   }
   n59 = Anim Dynamics - Bone: dyn_sleeves_l_dn!(ComponentPose=@n58, Alpha=1, GravityScale:float=1, ExternalForce="0,0,0") -> (Pose) {
    angularDampingOverride=0.99
    angularSpringConstant=1000
    bAngularSpring=true
    bGravityOverrideInSimSpace=true
    bOverrideAngularDamping=true
    bUseGravityOverride=true
    simulationSpace=RootRelative
    alphaBoolBlend {blendOption=Linear}
    boundBone {boneName=dyn_sleeves_l_dn}
    componentAppliedLinearAccClamp=<3000 3000 3000>
    componentLinearAccScale=<0.3 0.3 0.3>
    componentLinearVelScale=<0.3 0.3 0.3>
    gravityOverride {z=-300}
    physicsBodyDefinitions [{
      sphereCollisionRadius=0
      boundBone {boneName=dyn_sleeves_l_dn}
      boxExtents=<20 20 10>
      constraintSetup {angularLimitsMax=<0 6 3>, angularLimitsMin=<0 -6 -3>, angularTarget=<0 0 0>}
      localJointOffset=<5 0 4>
     }]
    relativeSpaceBone {boneName=lowerarm_r}
   }
   n60 = Anim Dynamics - Bone: dyn_sleeves_l_ft!(ComponentPose=@n59, Alpha=1, GravityScale:float=1, ExternalForce="0,0,0") -> (Pose) {
    angularDampingOverride=1
    angularSpringConstant=500
    bGravityOverrideInSimSpace=true
    bOverrideAngularDamping=true
    bUseGravityOverride=true
    simulationSpace=BoneRelative
    alphaBoolBlend {blendOption=Linear}
    boundBone {boneName=dyn_sleeves_l_ft}
    componentAppliedLinearAccClamp=<3000 3000 3000>
    componentLinearAccScale=<0.05 0.15 0.05>
    componentLinearVelScale=<0.3 0.3 0.3>
    gravityOverride=<500 300>
    physicsBodyDefinitions [{
      sphereCollisionRadius=0
      boundBone {boneName=dyn_sleeves_l_ft}
      boxExtents=<10 20 20>
      constraintSetup {angularLimitsMax=<0 8 5>, angularLimitsMin=<0 -2 -5>, angularTarget=<0 0 0>}
      localJointOffset=<5 1 2>
     }]
    relativeSpaceBone {boneName=lowerarm_l}
   }
   n61 = Anim Dynamics - Bone: dyn_sleeves_l_bk!(ComponentPose=@n60, Alpha=1, GravityScale:float=1, ExternalForce="0,0,0") -> (Pose) {
    angularDampingOverride=1
    angularSpringConstant=500
    bGravityOverrideInSimSpace=true
    bOverrideAngularDamping=true
    bUseGravityOverride=true
    simulationSpace=BoneRelative
    alphaBoolBlend {blendOption=Linear}
    boundBone {boneName=dyn_sleeves_l_bk}
    componentAppliedLinearAccClamp=<3000 3000 3000>
    componentLinearAccScale=<0.05 0.15 0.05>
    componentLinearVelScale=<0.3 0.3 0.3>
    gravityOverride=<500 -300>
    physicsBodyDefinitions [{
      sphereCollisionRadius=0
      boundBone {boneName=dyn_sleeves_l_bk}
      boxExtents=<10 20 20>
      constraintSetup {angularLimitsMax=<0 8 5>, angularLimitsMin=<0 -2 -5>, angularTarget=<0 0 0>}
      localJointOffset=<5 -1 2>
     }]
    relativeSpaceBone {boneName=lowerarm_l}
   }
   n62 = Anim Dynamics - Bone: dyn_pants_bk_r!(ComponentPose=@n61, Alpha=1, GravityScale:float=1, ExternalForce="0,0,0") -> (Pose) {
    angularDampingOverride=1
    angularSpringConstant=500
    bGravityOverrideInSimSpace=true
    bOverrideAngularDamping=true
    bUseGravityOverride=true
    simulationSpace=RootRelative
    alphaBoolBlend {blendOption=Linear}
    boundBone {boneName=dyn_pants_bk_r}
    componentAppliedLinearAccClamp=<3000 3000 3000>
    componentLinearAccScale=<0.1 0.1 0.1>
    componentLinearVelScale=<0.2 0.2 0.2>
    gravityOverride {z=-500}
    physicsBodyDefinitions [{
      sphereCollisionRadius=0
      boundBone {boneName=dyn_pants_bk_r}
      boxExtents=<20 10 10>
      constraintSetup {angularLimitsMax=<10 10 10>, angularLimitsMin=<-10 -10 -10>, angularTarget=<0 0 0>}
      localJointOffset=<5 0 0>
     }]
    relativeSpaceBone {boneName=calf_r}
   }
   n63 = Anim Dynamics - Bone: dyn_pants_ft_r!(ComponentPose=@n62, Alpha=1, GravityScale:float=1, ExternalForce="0,0,0") -> (Pose) {
    angularDampingOverride=1
    angularSpringConstant=500
    bGravityOverrideInSimSpace=true
    bOverrideAngularDamping=true
    bUseGravityOverride=true
    simulationSpace=RootRelative
    alphaBoolBlend {blendOption=Linear}
    boundBone {boneName=dyn_pants_ft_r}
    componentAppliedLinearAccClamp=<3000 3000 3000>
    componentLinearAccScale=<0.1 0.1 0.1>
    componentLinearVelScale=<0.2 0.2 0.2>
    gravityOverride {z=-500}
    physicsBodyDefinitions [{
      sphereCollisionRadius=0
      boundBone {boneName=dyn_pants_ft_r}
      boxExtents=<20 10 10>
      constraintSetup {angularLimitsMax=<10 10 10>, angularLimitsMin=<-10 -10 -10>, angularTarget=<0 0 0>}
      localJointOffset=<5 0 0>
     }]
    relativeSpaceBone {boneName=calf_r}
   }
   n64 = Anim Dynamics - Bone: dyn_pants_otr_r!(ComponentPose=@n63, Alpha=1, GravityScale:float=1, ExternalForce="0,0,0") -> (Pose) {
    angularDampingOverride=1
    angularSpringConstant=500
    bGravityOverrideInSimSpace=true
    bOverrideAngularDamping=true
    bUseGravityOverride=true
    simulationSpace=RootRelative
    alphaBoolBlend {blendOption=Linear}
    boundBone {boneName=dyn_pants_otr_r}
    componentAppliedLinearAccClamp=<3000 3000 3000>
    componentLinearAccScale=<0.1 0.1 0.1>
    componentLinearVelScale=<0.2 0.2 0.2>
    gravityOverride {z=-500}
    physicsBodyDefinitions [{
      sphereCollisionRadius=0
      boundBone {boneName=dyn_pants_otr_r}
      boxExtents=<20 10 10>
      constraintSetup {angularLimitsMax=<10 10 10>, angularLimitsMin=<-10 -10 -10>, angularTarget=<0 0 0>}
      localJointOffset=<5 0 0>
     }]
    relativeSpaceBone {boneName=calf_r}
   }
   n65 = Anim Dynamics - Bone: dyn_pants_inr_r!(ComponentPose=@n64, Alpha=1, GravityScale:float=1, ExternalForce="0,0,0") -> (Pose) {
    angularDampingOverride=1
    angularSpringConstant=500
    bGravityOverrideInSimSpace=true
    bOverrideAngularDamping=true
    bUseGravityOverride=true
    simulationSpace=RootRelative
    alphaBoolBlend {blendOption=Linear}
    boundBone {boneName=dyn_pants_inr_r}
    componentAppliedLinearAccClamp=<3000 3000 3000>
    componentLinearAccScale=<0.1 0.1 0.1>
    componentLinearVelScale=<0.2 0.2 0.2>
    gravityOverride {z=-500}
    physicsBodyDefinitions [{
      sphereCollisionRadius=0
      boundBone {boneName=dyn_pants_inr_r}
      boxExtents=<20 10 10>
      constraintSetup {angularLimitsMax=<10 10 10>, angularLimitsMin=<-10 -10 -10>, angularTarget=<0 0 0>}
      localJointOffset=<5 0 0>
     }]
    relativeSpaceBone {boneName=calf_r}
   }
   n66 = Anim Dynamics - Bone: dyn_pants_bk_l!(ComponentPose=@n65, Alpha=1, GravityScale:float=1, ExternalForce="0,0,0") -> (Pose) {
    angularDampingOverride=1
    angularSpringConstant=500
    bGravityOverrideInSimSpace=true
    bOverrideAngularDamping=true
    bUseGravityOverride=true
    simulationSpace=RootRelative
    alphaBoolBlend {blendOption=Linear}
    boundBone {boneName=dyn_pants_bk_l}
    componentAppliedLinearAccClamp=<3000 3000 3000>
    componentLinearAccScale=<0.1 0.1 0.1>
    componentLinearVelScale=<0.2 0.2 0.2>
    gravityOverride {z=-500}
    physicsBodyDefinitions [{
      sphereCollisionRadius=0
      boundBone {boneName=dyn_pants_bk_l}
      boxExtents=<20 10 10>
      constraintSetup {angularLimitsMax=<10 10 10>, angularLimitsMin=<-10 -10 -10>, angularTarget=<0 0 0>}
      localJointOffset=<5 0 0>
     }]
    relativeSpaceBone {boneName=calf_r}
   }
   n67 = Anim Dynamics - Bone: dyn_pants_ft_l!(ComponentPose=@n66, Alpha=1, GravityScale:float=1, ExternalForce="0,0,0") -> (Pose) {
    angularDampingOverride=1
    angularSpringConstant=500
    bGravityOverrideInSimSpace=true
    bOverrideAngularDamping=true
    bUseGravityOverride=true
    simulationSpace=RootRelative
    alphaBoolBlend {blendOption=Linear}
    boundBone {boneName=dyn_pants_ft_l}
    componentAppliedLinearAccClamp=<3000 3000 3000>
    componentLinearAccScale=<0.1 0.1 0.1>
    componentLinearVelScale=<0.2 0.2 0.2>
    gravityOverride {z=-500}
    physicsBodyDefinitions [{
      sphereCollisionRadius=0
      boundBone {boneName=dyn_pants_ft_l}
      boxExtents=<20 10 10>
      constraintSetup {angularLimitsMax=<10 10 10>, angularLimitsMin=<-10 -10 -10>, angularTarget=<0 0 0>}
      localJointOffset=<5 0 0>
     }]
    relativeSpaceBone {boneName=calf_r}
   }
   n68 = Anim Dynamics - Bone: dyn_pants_otr_l!(ComponentPose=@n67, Alpha=1, GravityScale:float=1, ExternalForce="0,0,0") -> (Pose) {
    angularDampingOverride=1
    angularSpringConstant=500
    bGravityOverrideInSimSpace=true
    bOverrideAngularDamping=true
    bUseGravityOverride=true
    simulationSpace=RootRelative
    alphaBoolBlend {blendOption=Linear}
    boundBone {boneName=dyn_pants_otr_l}
    componentAppliedLinearAccClamp=<3000 3000 3000>
    componentLinearAccScale=<0.1 0.1 0.1>
    componentLinearVelScale=<0.2 0.2 0.2>
    gravityOverride {z=-500}
    physicsBodyDefinitions [{
      sphereCollisionRadius=0
      boundBone {boneName=dyn_pants_otr_l}
      boxExtents=<20 10 10>
      constraintSetup {angularLimitsMax=<10 10 10>, angularLimitsMin=<-10 -10 -10>, angularTarget=<0 0 0>}
      localJointOffset=<5 0 0>
     }]
    relativeSpaceBone {boneName=calf_r}
   }
   n69 = Anim Dynamics - Bone: dyn_pants_inr_l!(ComponentPose=@n68, Alpha=1, GravityScale:float=1, ExternalForce="0,0,0") -> (Pose) {
    angularDampingOverride=1
    angularSpringConstant=500
    bGravityOverrideInSimSpace=true
    bOverrideAngularDamping=true
    bUseGravityOverride=true
    simulationSpace=RootRelative
    alphaBoolBlend {blendOption=Linear}
    boundBone {boneName=dyn_pants_inr_l}
    componentAppliedLinearAccClamp=<3000 3000 3000>
    componentLinearAccScale=<0.1 0.1 0.1>
    componentLinearVelScale=<0.2 0.2 0.2>
    gravityOverride {z=-500}
    physicsBodyDefinitions [{
      sphereCollisionRadius=0
      boundBone {boneName=dyn_pants_inr_l}
      boxExtents=<20 10 10>
      constraintSetup {angularLimitsMax=<10 10 10>, angularLimitsMin=<-10 -10 -10>, angularTarget=<0 0 0>}
      localJointOffset=<5 0 0>
     }]
    relativeSpaceBone {boneName=calf_r}
   }
   n70 = Anim Dynamics - Bone: dyn_arm_straps_r_01!(ComponentPose=@n69, Alpha=1, GravityScale:float=1, ExternalForce="0,0,0") -> (Pose) {
    angularDampingOverride=1
    angularSpringConstant=500
    bAngularSpring=true
    bGravityOverrideInSimSpace=true
    bOverrideAngularDamping=true
    bUseGravityOverride=true
    simulationSpace=BoneRelative
    alphaBoolBlend {blendOption=Linear}
    boundBone {boneName=dyn_arm_straps_r_01}
    componentAppliedLinearAccClamp=<3000 3000 3000>
    componentLinearAccScale=<0.25 0.25 0.25>
    componentLinearVelScale=<0.2 0.2 0.2>
    gravityOverride {x=-600}
    physicsBodyDefinitions [{
      sphereCollisionRadius=0
      boundBone {boneName=dyn_arm_straps_r_01}
      boxExtents=<15 20 20>
      constraintSetup {angularLimitsMax=<0 2 0>, angularLimitsMin=<0 -45 0>, angularTarget=<0 0 0>}
      localJointOffset=<2 0 0>
     }]
    relativeSpaceBone {boneName=upperarm_r}
   }
   n71 = Anim Dynamics - Bone: dyn_arm_straps_r_02!(ComponentPose=@n70, Alpha=1, GravityScale:float=1, ExternalForce="0,0,0") -> (Pose) {
    angularDampingOverride=0.999
    angularSpringConstant=500
    bAngularSpring=true
    bGravityOverrideInSimSpace=true
    bOverrideAngularDamping=true
    bUseGravityOverride=true
    simulationSpace=BoneRelative
    alphaBoolBlend {blendOption=Linear}
    boundBone {boneName=dyn_arm_straps_r_02}
    componentAppliedLinearAccClamp=<3000 3000 3000>
    componentLinearAccScale=<0.25 0.25 0.25>
    componentLinearVelScale=<0.2 0.2 0.2>
    gravityOverride {x=-600}
    physicsBodyDefinitions [{
      sphereCollisionRadius=0
      boundBone {boneName=dyn_arm_straps_r_02}
      boxExtents=<15 20 20>
      constraintSetup {angularLimitsMax=<0 2 0>, angularLimitsMin=<0 -45 0>, angularTarget=<0 0 0>}
      localJointOffset=<5 0 0>
     }]
    relativeSpaceBone {boneName=upperarm_r}
   }
   n72 = Anim Dynamics - Bone: dyn_arm_straps_r_03!(ComponentPose=@n71, Alpha=1, GravityScale:float=1, ExternalForce="0,0,0") -> (Pose) {
    angularDampingOverride=0.9999
    angularSpringConstant=500
    bAngularSpring=true
    bGravityOverrideInSimSpace=true
    bOverrideAngularDamping=true
    bUseGravityOverride=true
    simulationSpace=BoneRelative
    alphaBoolBlend {blendOption=Linear}
    boundBone {boneName=dyn_arm_straps_r_03}
    componentAppliedLinearAccClamp=<3000 3000 3000>
    componentLinearAccScale=<0.25 0.25 0.25>
    componentLinearVelScale=<0.2 0.2 0.2>
    gravityOverride {x=-600}
    physicsBodyDefinitions [{
      sphereCollisionRadius=0
      boundBone {boneName=dyn_arm_straps_r_03}
      boxExtents=<15 20 20>
      constraintSetup {angularLimitsMax=<0 2 0>, angularLimitsMin=<0 -45 0>, angularTarget=<0 0 0>}
      localJointOffset=<3 0 0>
     }]
    relativeSpaceBone {boneName=upperarm_r}
   }
   n73 = Anim Dynamics - Bone: dyn_arm_straps_l_01!(ComponentPose=@n72, Alpha=1, GravityScale:float=1, ExternalForce="0,0,0") -> (Pose) {
    angularDampingOverride=1
    angularSpringConstant=500
    bAngularSpring=true
    bGravityOverrideInSimSpace=true
    bOverrideAngularDamping=true
    bUseGravityOverride=true
    simulationSpace=BoneRelative
    alphaBoolBlend {blendOption=Linear}
    boundBone {boneName=dyn_arm_straps_l_01}
    componentAppliedLinearAccClamp=<3000 3000 3000>
    componentLinearAccScale=<0.25 0.25 0.25>
    componentLinearVelScale=<0.2 0.2 0.2>
    gravityOverride {x=600}
    physicsBodyDefinitions [{
      sphereCollisionRadius=0
      boundBone {boneName=dyn_arm_straps_l_01}
      boxExtents=<15 20 20>
      constraintSetup {angularLimitsMax=<0 2 0>, angularLimitsMin=<0 -45 0>, angularTarget=<0 0 0>}
      localJointOffset=<2 0 0>
     }]
    relativeSpaceBone {boneName=upperarm_l}
   }
   n74 = Anim Dynamics - Bone: dyn_arm_straps_l_02!(ComponentPose=@n73, Alpha=1, GravityScale:float=1, ExternalForce="0,0,0") -> (Pose) {
    angularDampingOverride=0.999
    angularSpringConstant=500
    bAngularSpring=true
    bGravityOverrideInSimSpace=true
    bOverrideAngularDamping=true
    bUseGravityOverride=true
    simulationSpace=BoneRelative
    alphaBoolBlend {blendOption=Linear}
    boundBone {boneName=dyn_arm_straps_l_02}
    componentAppliedLinearAccClamp=<3000 3000 3000>
    componentLinearAccScale=<0.25 0.25 0.25>
    componentLinearVelScale=<0.2 0.2 0.2>
    gravityOverride {x=600}
    physicsBodyDefinitions [{
      sphereCollisionRadius=0
      boundBone {boneName=dyn_arm_straps_l_02}
      boxExtents=<15 20 20>
      constraintSetup {angularLimitsMax=<0 2 0>, angularLimitsMin=<0 -45 0>, angularTarget=<0 0 0>}
      localJointOffset=<5 0 0>
     }]
    relativeSpaceBone {boneName=upperarm_l}
   }
   n75 = Anim Dynamics - Bone: dyn_arm_straps_l_03!(ComponentPose=@n74, Alpha=1, GravityScale:float=1, ExternalForce="0,0,0") -> (Pose) {
    angularDampingOverride=0.9999
    angularSpringConstant=500
    bAngularSpring=true
    bGravityOverrideInSimSpace=true
    bOverrideAngularDamping=true
    bUseGravityOverride=true
    simulationSpace=BoneRelative
    alphaBoolBlend {blendOption=Linear}
    boundBone {boneName=dyn_arm_straps_l_03}
    componentAppliedLinearAccClamp=<3000 3000 3000>
    componentLinearAccScale=<0.25 0.25 0.25>
    componentLinearVelScale=<0.2 0.2 0.2>
    gravityOverride {x=600}
    physicsBodyDefinitions [{
      sphereCollisionRadius=0
      boundBone {boneName=dyn_arm_straps_l_03}
      boxExtents=<15 20 20>
      constraintSetup {angularLimitsMax=<0 2 0>, angularLimitsMin=<0 -45 0>, angularTarget=<0 0 0>}
      localJointOffset=<3 0 0>
     }]
    relativeSpaceBone {boneName=upperarm_l}
   }
   Save cached pose 'prePose'(Pose=Component To Local(ComponentPose=@n75)) {cacheName=prePose}
   n76 = "Constraint - dyn_groin_collision (PPP)"!(
    ComponentPose=Local To Component(LocalPose=Use cached pose 'prePose'())
    Alpha=1
    ConstraintWeights_0=0.4
    ConstraintWeights_1:float=0.4
    ConstraintWeights_2:float=0.2
   ) -> (Pose) {
    alphaBoolBlend {blendOption=Linear}
    boneToModify {boneName=dyn_groin_collision}
    constraintSetup [{targetBone {boneName=thigh_twist_01_l}}, {targetBone {boneName=thigh_twist_01_r}}, {targetBone {boneName=pelvis}}]
    constraintWeights [1,1,1]
   }
   n77 = "Constraint - dyn_laces_MATR (PPPP)"!(
    ComponentPose=@n76
    Alpha=1
    ConstraintWeights_0=0.0255
    ConstraintWeights_1:float=0.9395
    ConstraintWeights_2:float=0.0133
    ConstraintWeights_3:float=0.0217
   ) -> (Pose) {
    alphaBoolBlend {blendOption=Linear}
    boneToModify {boneName=dyn_laces_MATR}
    constraintSetup [
     {targetBone {boneName=spine_01}}
     {targetBone {boneName=pelvis}}
     {targetBone {boneName=thigh_twist_01_l}}
     {targetBone {boneName=thigh_twist_01_r}}
    ]
    constraintWeights [1,1,1,1]
   }
   n78 = "Transform (Modify) Bone - Bone: dyn_groin_collision"(ComponentPose=@n77, Alpha=Get thigh_rot_up_l(), Translation="0,0,0", Rotation="0,0,35", Scale="1,1,1") {
    rotationMode=BMM_Additive
    rotationSpace=BCS_BoneSpace
    translationMode=BMM_Additive
    translationSpace=BCS_BoneSpace
    alphaBoolBlend {blendOption=Linear}
    boneToModify {boneName=dyn_groin_collision}
    rotation {roll=35}
   }
   n79 = "Transform (Modify) Bone - Bone: dyn_groin_collision"(ComponentPose=@n78, Alpha=Get thigh_rot_up_r(), Translation="0,0,0", Rotation="0,0,-35", Scale="1,1,1") {
    rotationMode=BMM_Additive
    rotationSpace=BCS_BoneSpace
    translationMode=BMM_Additive
    translationSpace=BCS_BoneSpace
    alphaBoolBlend {blendOption=Linear}
    boneToModify {boneName=dyn_groin_collision}
    rotation {roll=-35}
   }
   n80 = "Transform (Modify) Bone - Bone: dyn_groin_collision"!(ComponentPose=@n79, bAlphaBoolEnabled:bool=True, Translation="2,0,-3", Rotation="0,0,0", Scale="1,1,1") -> (Pose) {
    alphaInputType=Bool
    rotationMode=BMM_Additive
    rotationSpace=BCS_BoneSpace
    translationMode=BMM_Additive
    translationSpace=BCS_BoneSpace
    alphaBoolBlend {blendOption=Linear}
    boneToModify {boneName=dyn_groin_collision}
    translation {x=2, z=-3}
   }
   n81 = "Transform (Modify) Bone - Bone: dyn_groin_collision"(ComponentPose=@n80, bAlphaBoolEnabled=True, Translation="0,0,2", Rotation="0,0,0", Scale="1,1,1") {
    alphaInputType=Bool
    rotationMode=BMM_Additive
    rotationSpace=BCS_BoneSpace
    translationMode=BMM_Additive
    translationSpace=BCS_BoneSpace
    alphaBoolBlend {blendOption=Linear}
    boneToModify {boneName=dyn_groin_collision}
    translation {z=2}
   }
   n82 = Get thigh_rot_up_l()
   n83 = "Transform (Modify) Bone - Bone: dyn_jacket_l_fr_MSTR"(ComponentPose=@n81, Alpha=@n82, Translation="0,-16,8", Rotation="0,0,-60", Scale="1,1,1") {
    rotationMode=BMM_Additive
    rotationSpace=BCS_BoneSpace
    translationMode=BMM_Additive
    translationSpace=BCS_BoneSpace
    alphaBoolBlend {blendOption=Linear}
    boneToModify {boneName=dyn_jacket_l_fr_MSTR}
    rotation {roll=-60}
    translation {y=-16, z=8}
   }
   n84 = "Transform (Modify) Bone - Bone: dyn_jacket_l_MSTR"(ComponentPose=@n83, Alpha=@n82, Translation="0,-5,5", Rotation="0,-5,-25", Scale="1,1,1") {
    rotationMode=BMM_Additive
    rotationSpace=BCS_BoneSpace
    translationMode=BMM_Additive
    translationSpace=BCS_BoneSpace
    alphaBoolBlend {blendOption=Linear}
    boneToModify {boneName=dyn_jacket_l_MSTR}
    rotation {roll=-25, yaw=-5}
    translation {y=-5, z=5}
   }
   n85 = Get thigh_rot_up_r()
   n86 = "Transform (Modify) Bone - Bone: dyn_jacket_r_fr_MSTR"(ComponentPose=@n84, Alpha=@n85, Translation="0,10,5", Rotation="0,0,60", Scale="1,1,1") {
    rotationMode=BMM_Additive
    rotationSpace=BCS_BoneSpace
    translationMode=BMM_Additive
    translationSpace=BCS_BoneSpace
    alphaBoolBlend {blendOption=Linear}
    boneToModify {boneName=dyn_jacket_r_fr_MSTR}
    rotation {roll=60}
    translation {y=10, z=5}
   }
   n87 = "Transform (Modify) Bone - Bone: dyn_jacket_r_MSTR"(ComponentPose=@n86, Alpha=@n85, Translation="0,0,2", Rotation="0,0,0", Scale="1,1,1") {
    rotationMode=BMM_Additive
    rotationSpace=BCS_BoneSpace
    translationMode=BMM_Additive
    translationSpace=BCS_BoneSpace
    alphaBoolBlend {blendOption=Linear}
    boneToModify {boneName=dyn_jacket_r_MSTR}
    translation {z=2}
   }
   n88 = Get pawnSpeedRange() -> (pawnSpeedRange:double)
   n89 = "Transform (Modify) Bone - Bone: dyn_jacket_l_MSTR"(ComponentPose=@n87, Alpha=@n88, Translation="0,0,1", Rotation="0,0,0", Scale="1,1,1") {
    rotationMode=BMM_Additive
    rotationSpace=BCS_BoneSpace
    translationMode=BMM_Additive
    translationSpace=BCS_BoneSpace
    alphaBoolBlend {blendOption=Linear}
    boneToModify {boneName=dyn_jacket_l_MSTR}
    translation {z=1}
   }
   n90 = "Transform (Modify) Bone - Bone: dyn_jacket_l_bk_MSTR"(ComponentPose=@n89, Alpha=@n88, Translation="0,0,3", Rotation="0,0,0", Scale="1,1,1") {
    rotationMode=BMM_Additive
    rotationSpace=BCS_BoneSpace
    translationMode=BMM_Additive
    translationSpace=BCS_BoneSpace
    alphaBoolBlend {blendOption=Linear}
    boneToModify {boneName=dyn_jacket_l_bk_MSTR}
    translation {z=3}
   }
   n91 = RigidBody!(ComponentPose=@n90, Alpha, OverrideWorldGravity:Vector=Get GravityOverride(), ExternalForce) -> (Pose) {
    bClampLinearTranslationLimitToRefPose=true
    bOverrideWorldGravity=true
    overridePhysicsAsset=&/F_MED_LightningDragon_Dyn_Jacket
    simulationSpace=BaseBoneSpace
    alphaBoolBlend {blendOption=Linear}
    baseBoneRef {boneName=spine_01}
    componentAppliedLinearAccClamp=<3000 1000 1000>
    componentLinearAccScale {x=0.1}
    componentLinearVelScale {x=0.2}
    overrideWorldGravity {z=-980}
    simSpaceSettings {maxAngularAcceleration=100, maxAngularVelocity=5, maxLinearAcceleration=500, maxLinearVelocity=100, velocityScaleZ=0, worldAlpha=1}
   }
   n92 = Get pushOutLeftJacketAlpha() -> (pushOutLeftJacketAlpha:double)
   n93 = "Transform (Modify) Bone - Bone: dyn_jacket_l_fr_01"(ComponentPose=@n91, Alpha=@n92, Translation="0,0,3", Rotation="0,0,0", Scale="1,1,1") {
    rotationMode=BMM_Additive
    rotationSpace=BCS_BoneSpace
    translationMode=BMM_Additive
    translationSpace=BCS_BoneSpace
    alphaBoolBlend {blendOption=Linear}
    alphaScaleBiasClamp {bInterpResult=true, interpSpeedIncreasing=30}
    boneToModify {boneName=dyn_jacket_l_fr_01}
    translation {z=3}
   }
   n94 = "Transform (Modify) Bone - Bone: dyn_jacket_l_MSTR"(ComponentPose=@n93, Alpha=@n92, Translation="0,0,2", Rotation="0,0,0", Scale="1,1,1") {
    rotationMode=BMM_Additive
    rotationSpace=BCS_BoneSpace
    translationMode=BMM_Additive
    translationSpace=BCS_BoneSpace
    alphaBoolBlend {blendOption=Linear}
    alphaScaleBiasClamp {bInterpResult=true, interpSpeedIncreasing=30}
    boneToModify {boneName=dyn_jacket_l_MSTR}
    translation {z=2}
   }
   n95 = Get movingLeftSideAlpha() -> (movingLeftSideAlpha:double)
   n96 = "Transform (Modify) Bone - Bone: dyn_jacket_l_fr_01"(ComponentPose=@n94, Alpha=@n95, Translation="0,-5,5", Rotation="0,0,0", Scale="1,1,1") {
    rotationMode=BMM_Additive
    rotationSpace=BCS_BoneSpace
    translationMode=BMM_Additive
    translationSpace=BCS_BoneSpace
    alphaBoolBlend {blendOption=Linear}
    alphaScaleBiasClamp {bInterpResult=true}
    boneToModify {boneName=dyn_jacket_l_fr_01}
    translation {y=-5, z=5}
   }
   n97 = "Transform (Modify) Bone - Bone: dyn_jacket_l_MSTR"(ComponentPose=@n96, Alpha=@n95, Translation="0,0,3", Rotation="0,0,0", Scale="1,1,1") {
    rotationMode=BMM_Additive
    rotationSpace=BCS_BoneSpace
    translationMode=BMM_Additive
    translationSpace=BCS_BoneSpace
    alphaBoolBlend {blendOption=Linear}
    alphaScaleBiasClamp {bInterpResult=true}
    boneToModify {boneName=dyn_jacket_l_MSTR}
    translation {z=3}
   }
   n98 = Get movingRightSideAlpha() -> (movingRightSideAlpha:double)
   n99 = "Transform (Modify) Bone - Bone: dyn_jacket_r_fr_02"(ComponentPose=@n97, Alpha=@n98, Translation="0,0,3", Rotation="0,0,0", Scale="1,1,1") {
    rotationMode=BMM_Additive
    rotationSpace=BCS_BoneSpace
    translationMode=BMM_Additive
    translationSpace=BCS_BoneSpace
    alphaBoolBlend {blendOption=Linear}
    alphaScaleBiasClamp {bInterpResult=true}
    boneToModify {boneName=dyn_jacket_r_fr_02}
    translation {z=3}
   }
   n100 = "Transform (Modify) Bone - Bone: dyn_jacket_r_MSTR"(ComponentPose=@n99, Alpha=@n98, Translation="0,0,2", Rotation="0,0,0", Scale="1,1,1") {
    rotationMode=BMM_Additive
    rotationSpace=BCS_BoneSpace
    translationMode=BMM_Additive
    translationSpace=BCS_BoneSpace
    alphaBoolBlend {blendOption=Linear}
    alphaScaleBiasClamp {bInterpResult=true}
    boneToModify {boneName=dyn_jacket_r_MSTR}
    translation {z=2}
   }
   n101 = Control Rig(Source=Use cached pose 'prePose'()) {
    alphaBoolBlend {blendOption=Linear}
    controlRigAssetReference {blueprintRigClass=&/F_MED_LightningDragon_Jacket_CtrlRig.F_MED_LightningDragon_Jacket_CtrlRig_C}
   }
   n102 = "Constraint - dyn_groin_collision (PPP)"(ComponentPose=Local To Component(LocalPose=@n101), Alpha=1, ConstraintWeights_0=0.4, ConstraintWeights_1=0.4, ConstraintWeights_2=0.2) {
    alphaBoolBlend {blendOption=Linear}
    boneToModify {boneName=dyn_groin_collision}
    constraintSetup [{targetBone {boneName=thigh_twist_01_l}}, {targetBone {boneName=thigh_twist_01_r}}, {targetBone {boneName=pelvis}}]
    constraintWeights [1,1,1]
   }
   n103 = "Constraint - dyn_laces_MATR (PPPP)"(ComponentPose=@n102, Alpha=1, ConstraintWeights_0=0.0255, ConstraintWeights_1=0.9395, ConstraintWeights_2=0.0133, ConstraintWeights_3=0.0217) {
    alphaBoolBlend {blendOption=Linear}
    boneToModify {boneName=dyn_laces_MATR}
    constraintSetup [
     {targetBone {boneName=spine_01}}
     {targetBone {boneName=pelvis}}
     {targetBone {boneName=thigh_twist_01_l}}
     {targetBone {boneName=thigh_twist_01_r}}
    ]
    constraintWeights [1,1,1,1]
   }
   n104 = "Transform (Modify) Bone - Bone: dyn_groin_collision"!(ComponentPose=@n103, Alpha:float=Get thigh_rot_up_l(), Translation="0,0,0", Rotation="0,0,35", Scale="1,1,1") -> (Pose) {
    rotationMode=BMM_Additive
    rotationSpace=BCS_BoneSpace
    translationMode=BMM_Additive
    translationSpace=BCS_BoneSpace
    alphaBoolBlend {blendOption=Linear}
    boneToModify {boneName=dyn_groin_collision}
    rotation {roll=35}
   }
   n105 = "Transform (Modify) Bone - Bone: dyn_groin_collision"(ComponentPose=@n104, Alpha=Get thigh_rot_up_r(), Translation="0,0,0", Rotation="0,0,-35", Scale="1,1,1") {
    rotationMode=BMM_Additive
    rotationSpace=BCS_BoneSpace
    translationMode=BMM_Additive
    translationSpace=BCS_BoneSpace
    alphaBoolBlend {blendOption=Linear}
    boneToModify {boneName=dyn_groin_collision}
    rotation {roll=-35}
   }
   n106 = RigidBody(ComponentPose=@n105, OverrideWorldGravity="0,0,-980") {
    bClampLinearTranslationLimitToRefPose=true
    bOverrideWorldGravity=true
    overridePhysicsAsset=&F_MED_LightningDragon_Dyn_Jacket
    simulationSpace=BaseBoneSpace
    alphaBoolBlend {blendOption=Linear}
    baseBoneRef {boneName=spine_01}
    componentAppliedLinearAccClamp=<3000 1000 1000>
    componentLinearAccScale {x=0.1}
    componentLinearVelScale {x=0.2}
    overrideWorldGravity {z=-980}
    simSpaceSettings {maxAngularAcceleration=100, maxAngularVelocity=5, maxLinearAcceleration=500, maxLinearVelocity=100, velocityScaleZ=0, worldAlpha=1}
   }
   n107 = Blend Poses by bool(
    BlendPose_0=Component To Local(ComponentPose=@n106)
    BlendPose_1=Component To Local(ComponentPose=@n100)
    BlendTime_1=0.3
    bActiveValue=Get bIsPlayingEmote()
   ) {blendType=Linear}
   n108 = "Transform (Modify) Bone - Bone: dyn_jacket_l_fr_MSTR"(ComponentPose=Local To Component(LocalPose=Use cached pose 'prePose'()), Alpha=1, Translation="0,-15,0", Rotation="35,0,-65", Scale="1,1,1") {
    rotationMode=BMM_Additive
    rotationSpace=BCS_BoneSpace
    translationMode=BMM_Additive
    translationSpace=BCS_BoneSpace
    alphaBoolBlend {blendOption=Linear}
    boneToModify {boneName=dyn_jacket_l_fr_MSTR}
    rotation {pitch=35, roll=-65}
    translation {y=-15}
   }
   n109 = "Transform (Modify) Bone - Bone: dyn_jacket_l_MSTR"(ComponentPose=@n108, Alpha=1, Translation="0,0,4", Rotation="0,0,-10", Scale="1,1,1") {
    rotationMode=BMM_Additive
    rotationSpace=BCS_BoneSpace
    translationMode=BMM_Additive
    translationSpace=BCS_BoneSpace
    alphaBoolBlend {blendOption=Linear}
    boneToModify {boneName=dyn_jacket_l_MSTR}
    rotation {roll=-10}
    translation {z=4}
   }
   n110 = "Transform (Modify) Bone - Bone: dyn_jacket_r_fr_MSTR"(ComponentPose=@n109, Alpha=1, Translation="0,7,5", Rotation="0,0,60", Scale="1,1,1") {
    rotationMode=BMM_Additive
    rotationSpace=BCS_BoneSpace
    translationMode=BMM_Additive
    translationSpace=BCS_BoneSpace
    alphaBoolBlend {blendOption=Linear}
    boneToModify {boneName=dyn_jacket_r_fr_MSTR}
    rotation {roll=60}
    translation {y=7, z=5}
   }
   n111 = "Transform (Modify) Bone - Bone: dyn_jacket_l_fr_MSTR"(ComponentPose=@n110, Alpha=Get pushOutLeftJacketAlpha(), Translation="0,-5,10", Rotation="0,0,0", Scale="1,1,1") {
    rotationMode=BMM_Additive
    rotationSpace=BCS_BoneSpace
    translationMode=BMM_Additive
    translationSpace=BCS_BoneSpace
    alphaBoolBlend {blendOption=Linear}
    alphaScaleBiasClamp {bInterpResult=true, interpSpeedIncreasing=30}
    boneToModify {boneName=dyn_jacket_l_fr_MSTR}
    translation {y=-5, z=10}
   }
   n112 = RigidBody(ComponentPose=@n111, OverrideWorldGravity=Get GravityOverride()) {
    bClampLinearTranslationLimitToRefPose=true
    bOverrideWorldGravity=true
    overridePhysicsAsset=&/F_MED_LightningDragon_Dyn_Jacket_DBNO
    simulationSpace=BaseBoneSpace
    alphaBoolBlend {blendOption=Linear}
    baseBoneRef {boneName=spine_01}
    componentAppliedLinearAccClamp=<3000 1000 1000>
    componentLinearAccScale {x=0.1}
    componentLinearVelScale {x=0.2}
    overrideWorldGravity {z=-980}
    simSpaceSettings {maxAngularAcceleration=100, maxAngularVelocity=5, maxLinearAcceleration=500, maxLinearVelocity=100, velocityScaleZ=0, worldAlpha=1}
   }
   n113 = Get pushOutLeftJacketAlpha()
   n114 = "Transform (Modify) Bone - Bone: dyn_jacket_l_fr_01"(ComponentPose=@n112, Alpha=@n113, Translation="0,0,5", Rotation="0,0,0", Scale="1,1,1") {
    rotationMode=BMM_Additive
    rotationSpace=BCS_BoneSpace
    translationMode=BMM_Additive
    translationSpace=BCS_BoneSpace
    alphaBoolBlend {blendOption=Linear}
    alphaScaleBiasClamp {bInterpResult=true, interpSpeedIncreasing=30}
    boneToModify {boneName=dyn_jacket_l_fr_01}
    translation {z=5}
   }
   n115 = "Transform (Modify) Bone - Bone: dyn_jacket_l_MSTR"(ComponentPose=@n114, Alpha=@n113, Translation="-0.000,-0.000,2", Rotation="-0.000,0.000,-5", Scale="1,1,1") {
    rotationMode=BMM_Additive
    rotationSpace=BCS_BoneSpace
    translationMode=BMM_Additive
    translationSpace=BCS_BoneSpace
    alphaBoolBlend {blendOption=Linear}
    alphaScaleBiasClamp {bInterpResult=true, interpSpeedIncreasing=30}
    boneToModify {boneName=dyn_jacket_l_MSTR}
    rotation {roll=-5}
    translation {z=2}
   }
   n116 = Component To Local(ComponentPose=@n115)
   n117 = Get bIsDBNO()
   n118 = "Transform (Modify) Bone - Bone: dyn_bag"(
    ComponentPose=Local To Component(LocalPose=Use cached pose 'prePose'())
    Alpha=1
    Translation="0,0,0"
    Rotation="-6.5,-0.000,0.000"
    Scale="1,1,1"
   ) {
    rotationMode=BMM_Additive
    rotationSpace=BCS_BoneSpace
    alphaBoolBlend {blendOption=Linear}
    boneToModify {boneName=dyn_bag}
    rotation {pitch=-6.5}
   }
   n119 = "Constraint - dyn_groin_collision (PPP)"(ComponentPose=@n118, Alpha=1, ConstraintWeights_0=0.7, ConstraintWeights_1=0.3, ConstraintWeights_2=0) {
    alphaBoolBlend {blendOption=Linear}
    boneToModify {boneName=dyn_groin_collision}
    constraintSetup [{targetBone {boneName=thigh_twist_01_l}}, {targetBone {boneName=thigh_twist_01_r}}, {targetBone {boneName=pelvis}}]
    constraintWeights [1,1,1]
   }
   n120 = "Constraint - dyn_jacket_bk_MSTR (P)"(ComponentPose=@n119, Alpha=1, ConstraintWeights_0=1) {
    alphaBoolBlend {blendOption=Linear}
    boneToModify {boneName=dyn_jacket_bk_MSTR}
    constraintSetup [{targetBone {boneName=spine_01}}]
    constraintWeights [1]
   }
   n121 = "Transform (Modify) Bone - Bone: dyn_jacket_bk_MSTR"!(ComponentPose=@n120, bAlphaBoolEnabled:bool=True, Translation="-6.3,0,-8.3", Rotation="44.782,0.071,0.024", Scale="1,1,1") -> (Pose) {
    alphaInputType=Bool
    rotationMode=BMM_Additive
    rotationSpace=BCS_BoneSpace
    translationMode=BMM_Additive
    translationSpace=BCS_BoneSpace
    alphaBoolBlend {blendInTime=0.5, blendOption=Linear}
    boneToModify {boneName=dyn_jacket_bk_MSTR}
    rotation=<rot 44.7819 0.071 0.024>
    translation {x=-6.3, z=-8.3}
   }
   n122 = "Transform (Modify) Bone - Bone: dyn_jacket_l_bk_MSTR"!(ComponentPose=@n121, bAlphaBoolEnabled:bool=True, Translation="-3,0.000,2.000", Rotation="0,0,0", Scale="1,1,1") -> (Pose) {
    alphaInputType=Bool
    rotationMode=BMM_Additive
    rotationSpace=BCS_BoneSpace
    translationMode=BMM_Additive
    translationSpace=BCS_BoneSpace
    alphaBoolBlend {blendInTime=0.5, blendOption=Linear}
    boneToModify {boneName=dyn_jacket_l_bk_MSTR}
    translation {x=-3, z=2}
   }
   n123 = "Transform (Modify) Bone - Bone: dyn_jacket_r_bk_MSTR"!(ComponentPose=@n122, bAlphaBoolEnabled:bool=True, Translation="-3,0,2", Rotation="4.884,0.007,-0.000", Scale="1,1,1") -> (Pose) {
    alphaInputType=Bool
    rotationMode=BMM_Additive
    rotationSpace=BCS_BoneSpace
    translationMode=BMM_Additive
    translationSpace=BCS_BoneSpace
    alphaBoolBlend {blendInTime=0.5, blendOption=Linear}
    boneToModify {boneName=dyn_jacket_r_bk_MSTR}
    rotation=<rot 4.884 0.007 0>
    translation {x=-3, z=2}
   }
   n124 = "Transform (Modify) Bone - Bone: dyn_jacket_r_fr_MSTR"!(ComponentPose=@n123, bAlphaBoolEnabled:bool=True, Translation="5,0,0", Rotation="-18.977,50.896,45.218", Scale="1,1,1") -> (Pose) {
    alphaInputType=Bool
    rotationMode=BMM_Additive
    rotationSpace=BCS_BoneSpace
    translationMode=BMM_Additive
    translationSpace=BCS_BoneSpace
    alphaBoolBlend {blendInTime=0.5, blendOption=Linear}
    boneToModify {boneName=dyn_jacket_r_fr_MSTR}
    rotation=<rot -18.977 50.896 45.218>
    translation {x=5}
   }
   n125 = "Transform (Modify) Bone - Bone: dyn_jacket_r_MSTR"!(ComponentPose=@n124, bAlphaBoolEnabled:bool=True, Translation="-7.879,7.216,0.281", Rotation="11.202,62.638,34.579", Scale="1,1,1") -> (Pose) {
    alphaInputType=Bool
    rotationMode=BMM_Additive
    rotationSpace=BCS_BoneSpace
    translationMode=BMM_Additive
    translationSpace=BCS_BoneSpace
    alphaBoolBlend {blendInTime=0.5, blendOption=Linear}
    boneToModify {boneName=dyn_jacket_r_MSTR}
    rotation=<rot 11.202 62.638 34.579>
    translation=<-7.879 7.216 0.281>
   }
   n126 = "Transform (Modify) Bone - Bone: dyn_jacket_l_fr_MSTR"!(ComponentPose=@n125, bAlphaBoolEnabled:bool=True, Translation="5.386,-0.508,-1.842", Rotation="-7.541,-20,-59.873", Scale="1,1,1") -> (Pose) {
    alphaInputType=Bool
    rotationMode=BMM_Additive
    rotationSpace=BCS_BoneSpace
    translationMode=BMM_Additive
    translationSpace=BCS_BoneSpace
    alphaBoolBlend {blendInTime=0.5, blendOption=Linear}
    boneToModify {boneName=dyn_jacket_l_fr_MSTR}
    rotation=<rot -7.541 -20 -59.873>
    translation=<5.3861 -0.508 -1.842>
   }
   n127 = "Transform (Modify) Bone - Bone: dyn_jacket_l_MSTR"!(ComponentPose=@n126, bAlphaBoolEnabled:bool=True, Translation="-4.182,-0.708,0.554", Rotation="14.942,-45.175,-1.338", Scale="1,1,1") -> (Pose) {
    alphaInputType=Bool
    rotationMode=BMM_Additive
    rotationSpace=BCS_BoneSpace
    translationMode=BMM_Additive
    translationSpace=BCS_BoneSpace
    alphaBoolBlend {blendInTime=0.5, blendOption=Linear}
    boneToModify {boneName=dyn_jacket_l_MSTR}
    rotation=<rot 14.942 -45.1749 -1.338>
    translation=<-4.182 -0.708 0.554>
   }
   n128 = RigidBody(ComponentPose=@n127, OverrideWorldGravity="0,0,-980") {
    bClampLinearTranslationLimitToRefPose=true
    bOverrideWorldGravity=true
    overridePhysicsAsset=&/F_MED_LightningDragon_Dyn_Jacket_Sliding
    simulationSpace=BaseBoneSpace
    alphaBoolBlend {blendOption=Linear}
    baseBoneRef {boneName=spine_01}
    componentAppliedLinearAccClamp=<3000 1000 1000>
    componentLinearAccScale {x=0.1}
    componentLinearVelScale {x=0.2}
    overrideWorldGravity {z=-980}
    simSpaceSettings {maxAngularAcceleration=100, maxAngularVelocity=5, maxLinearAcceleration=500, maxLinearVelocity=100, velocityScaleZ=0, worldAlpha=1}
   }
   Save cached pose 'RBANPose'(Pose=Blend Poses by bool(
     BlendPose_0=Component To Local(ComponentPose=@n128)
     BlendPose_1=Blend Poses by bool(BlendPose_0=@n116, BlendPose_1=@n107, BlendTime_1=0.3, bActiveValue=@n117) {blendType=Linear}
     BlendTime_1=0.3
     bActiveValue=Get bIsSliding()
    ) {blendType=Linear}) {cacheName=RBANPose}
   n129 = "Constraint - dyn_jacket_l_bk_MSTR (P)"(ComponentPose=Local To Component(LocalPose=Use cached pose 'prePose'()), Alpha=1, ConstraintWeights_0=1) {
    alphaBoolBlend {blendOption=Linear}
    boneToModify {boneName=dyn_jacket_l_bk_MSTR}
    constraintSetup [{targetBone {boneName=thigh_twist_01_l}}]
    constraintWeights [1]
   }
   n130 = "Constraint - dyn_jacket_l_MSTR (P)"(ComponentPose=@n129, Alpha=1, ConstraintWeights_0=1) {
    alphaBoolBlend {blendOption=Linear}
    boneToModify {boneName=dyn_jacket_l_MSTR}
    constraintSetup [{targetBone {boneName=thigh_twist_01_l}}]
    constraintWeights [1]
   }
   n131 = "Constraint - dyn_jacket_l_fr_02 (P)"(ComponentPose=@n130, Alpha=1, ConstraintWeights_0=1) {
    alphaBoolBlend {blendOption=Linear}
    boneToModify {boneName=dyn_jacket_l_fr_02}
    constraintSetup [{targetBone {boneName=thigh_twist_01_l}}]
    constraintWeights [1]
   }
   n132 = "Constraint - dyn_laces_l_01 (P)"(ComponentPose=@n131, Alpha=1, ConstraintWeights_0=1) {
    alphaBoolBlend {blendOption=Linear}
    boneToModify {boneName=dyn_laces_l_01}
    constraintSetup [{targetBone {boneName=thigh_twist_01_l}}]
    constraintWeights [1]
   }
   n133 = "Constraint - dyn_jacket_r_bk_MSTR (P)"(ComponentPose=@n132, Alpha=1, ConstraintWeights_0=1) {
    alphaBoolBlend {blendOption=Linear}
    boneToModify {boneName=dyn_jacket_r_bk_MSTR}
    constraintSetup [{targetBone {boneName=thigh_twist_01_r}}]
    constraintWeights [1]
   }
   n134 = "Constraint - dyn_jacket_r_MSTR (P)"(ComponentPose=@n133, Alpha=1, ConstraintWeights_0=1) {
    alphaBoolBlend {blendOption=Linear}
    boneToModify {boneName=dyn_jacket_r_MSTR}
    constraintSetup [{targetBone {boneName=thigh_twist_01_r}}]
    constraintWeights [1]
   }
   n135 = "Constraint - dyn_jacket_r_fr_02 (P)"(ComponentPose=@n134, Alpha=1, ConstraintWeights_0=1) {
    alphaBoolBlend {blendOption=Linear}
    boneToModify {boneName=dyn_jacket_r_fr_02}
    constraintSetup [{targetBone {boneName=thigh_twist_01_r}}]
    constraintWeights [1]
   }
   n136 = "Constraint - dyn_laces_r_01 (P)"(ComponentPose=@n135, Alpha=1, ConstraintWeights_0=1) {
    alphaBoolBlend {blendOption=Linear}
    boneToModify {boneName=dyn_laces_r_01}
    constraintSetup [{targetBone {boneName=thigh_twist_01_r}}]
    constraintWeights [1]
   }
   n137 = "Transform (Modify) Bone - Bone: dyn_laces_l_01"(ComponentPose=@n136, Alpha=Get thigh_rot_up_l(), Translation="10.000,0.000,1", Rotation="0,0,0", Scale="1,1,1") {
    rotationMode=BMM_Additive
    rotationSpace=BCS_BoneSpace
    translationMode=BMM_Additive
    translationSpace=BCS_BoneSpace
    alphaBoolBlend {blendOption=Linear}
    boneToModify {boneName=dyn_laces_l_01}
    translation {x=10, z=1}
   }
   n138 = "Transform (Modify) Bone - Bone: dyn_laces_r_01"(ComponentPose=@n137, Alpha=Get thigh_rot_up_r(), Translation="10.000,-0.000,1", Rotation="0,0,0", Scale="1,1,1") {
    rotationMode=BMM_Additive
    rotationSpace=BCS_BoneSpace
    translationMode=BMM_Additive
    translationSpace=BCS_BoneSpace
    alphaBoolBlend {blendOption=Linear}
    boneToModify {boneName=dyn_laces_r_01}
    translation {x=10, z=1}
   }
   Save cached pose 'baseNoRBANPose'(Pose=Component To Local(ComponentPose=@n138)) {cacheName=baseNoRBANPose}
  ]}
 EventGraph {nodes [
   Event BlueprintUpdateAnimation()
   > n1 = Sequence!(execute) -> (then_0, then_1, then_2:Exec, then_3:Exec)
   n2 = Get SkelMesh()
   > Set thigh_rot_up_l(thigh_rot_up_l:double=GetAlphaFromReferenceTransformDelta(SkeletalMeshComponent=@n2, BoneName=thigh_l, BaseName=pelvis, DeltaComponentType=QuaternionTwist, InputRange=(X=0,Y=100), TwistAxis="0,0,1")) -> (Output_Get:double)
   > Set thigh_rot_up_01_l(thigh_rot_up_01_l:double=GetAlphaFromReferenceTransformDelta(SkeletalMeshComponent=@n2, BoneName=thigh_l, BaseName=pelvis, DeltaComponentType=QuaternionTwist, InputRange=(X=0,Y=50), TwistAxis="0,0,1")) -> (Output_Get:double)
   > Set thigh_rot_up_02_l(thigh_rot_up_02_l:double=GetAlphaFromReferenceTransformDelta(SkeletalMeshComponent=@n2, BoneName=thigh_l, BaseName=pelvis, DeltaComponentType=QuaternionTwist, InputRange=(X=40,Y=100), TwistAxis="0,0,1")) -> (Output_Get:double)
   > Set thigh_rot_up_r(thigh_rot_up_r:double=GetAlphaFromReferenceTransformDelta(SkeletalMeshComponent=@n2, BoneName=thigh_r, BaseName=pelvis, DeltaComponentType=QuaternionTwist, InputRange=(X=0,Y=100), TwistAxis="0,0,1")) -> (Output_Get:double)
   > Set head_rot_bk(head_rot_bk:double=GetAlphaFromReferenceTransformDelta(SkeletalMeshComponent=@n2, BoneName=head, BaseName=spine_05, DeltaComponentType=QuaternionTwist, InputRange=(X=0,Y=60), TwistAxis="0,0,1")) -> (Output_Get:double)
   > Set head_rot_r(head_rot_r:double=GetAlphaFromReferenceTransformDelta(SkeletalMeshComponent=@n2, BoneName=head, BaseName=spine_05, DeltaComponentType=QuaternionTwist, InputRange=(X=0,Y=40), TwistAxis="0,1,0")) -> (Output_Get:double)
   n3 = Get SkelMesh()
   > Set hand_rot_r(hand_rot_r:double=GetAlphaFromReferenceTransformDelta(
     SkeletalMeshComponent=@n3
     BoneName=hand_r
     BaseName=lowerarm_twist_01_r
     DeltaComponentType=QuaternionTwist
     InputRange=(X=0,Y=-50)
     TwistAxis="0,1,1"
    )) -> (Output_Get:double)
   > Set pushOutLeftJacketAlpha(pushOutLeftJacketAlpha:double=float * float(
     A=GetAlphaFromReferenceTransformDelta(SkeletalMeshComponent=@n3, BoneName=thigh_l, BaseName=pelvis, DeltaComponentType=QuaternionTwist, InputRange=(X=15,Y=30), TwistAxis="0,1,0")
     B=Get movingLeftSideAlpha()
    ) {is=title K2Node_PromotableOperator}) -> (Output_Get:double)
   > n4 = Branch(Condition=AND Boolean(A=Get bIsAccelerating2D(), B=Get bIsOnGround() -> (bIsOnGround:bool)))
   > Set pullOutAlpha(pullOutAlpha:double=GetAlphaFromReferenceTransformDelta(SkeletalMeshComponent=@n3, BoneName=thigh_l, BaseName=pelvis, DeltaComponentType=QuaternionTwist, InputRange=(X=0,Y=10), TwistAxis="0,1,0")) -> (Output_Get:double)
   Set pullOutAlpha(execute=@n4.else, pullOutAlpha=0)
   n5 = Gravity Override State Function DEPRECATED(
    execute=@n1.then_1
    CharacterPartAnimInstance=Self-Reference()
    GameTimeInSeconds=GetGameTimeInSeconds()
    GravityOverrideParams=Get Gravity Override Params()
   )
   > Set GravityOverride(GravityOverride=@n5)
   > Set isInFrontEndOrEmoting(isInFrontEndOrEmoting:bool=OR Boolean(A=Get bIsInFrontEnd(), B=Get bIsPlayingEmote())) -> (Output_Get:bool)
   n6 = PlayerMovement(execute=@n1.then_2, AnimInstance=Self-Reference(), XDotProductCoefficient=1, YDotProductCoefficient=1, ZDotProductCoefficient=1)
   n7 = float * float(
    A=MapRangeClamped(Value=@n6.YMovement, InRangeA=0, InRangeB=-1, OutRangeA=1, OutRangeB=0)
    B=MapRangeClamped(Value="Absolute (Float)"(A=@n6.ZMovement), InRangeA=0, InRangeB=1, OutRangeA=1, OutRangeB=0)
   )
   > Set BwdMovement(BwdMovement:double=MapRangeClamped(Value=@n7, InRangeA=0, InRangeB=1, OutRangeB=0.5)) -> (Output_Get:double)
   n8 = MapRangeClamped(Value=@n6, InRangeA=0, InRangeB=-1, OutRangeA=1, OutRangeB=0)
   > Set LeftMovement(LeftMovement:double=MapRangeClamped(Value=float * float(A=@n7, B=@n8), InRangeA=0, InRangeB=1, OutRangeB=0.5)) -> (Output_Get:double)
   n9 = MapRangeClamped(Value=@n6, InRangeA=0, InRangeB=-1, OutRangeA=1, OutRangeB=0.5)
   > Set BwdLeftMovement(BwdLeftMovement:double=MapRangeClamped(Value=float * float(A=@n7, B=@n9), InRangeA=0, InRangeB=1, OutRangeB=0.5)) -> (Output_Get:double)
   n10 = MapRangeClamped(Value=@n6, InRangeA=0, InRangeB=1, OutRangeA=1, OutRangeB=0)
   > Set RightMovement(RightMovement:double=MapRangeClamped(Value=float * float(A=@n7, B=@n10), InRangeA=0, InRangeB=1, OutRangeB=0.5)) -> (Output_Get:double)
   n11 = MapRangeClamped(Value=@n6, InRangeA=0, InRangeB=1, OutRangeA=1, OutRangeB=0.5)
   > Set BwdRightMovement(BwdRightMovement:double=MapRangeClamped(Value=float * float(A=@n7, B=@n11), InRangeA=0, InRangeB=1, OutRangeB=0.5)) -> (Output_Get:double)
   > Set movingLeftSideAlpha(movingLeftSideAlpha:double=MapRangeClamped(Value=@n6, InRangeA=0, InRangeB=-1)) -> (Output_Get:double)
   > Set movingRightSideAlpha(movingRightSideAlpha:double=MapRangeClamped(Value=@n6, InRangeA=0, InRangeB=1)) -> (Output_Get:double)
   Set pawnSpeedRange(execute=@n1.then_3, pawnSpeedRange:double=MapRangeClamped(Value=Get PawnSpeed2D(), InRangeA=300, InRangeB=550)) -> (Output_Get:double)
   Event BlueprintInitializeAnimation()
   n12 = GetOwnerSkeletalMeshComponent()
   > Is Valid(InputObject=@n12)
   > Set SkelMesh(SkelMesh=@n12)
   Event OnEndSkydiving()
   > ResetDynamics(InTeleportType=TeleportPhysics)
   > ResetDynamics(InTeleportType=ResetPhysics)
   NOT Boolean(A=Get bIsCrouching())
  ]}
]
variables [
 BwdLeftMovement {type=number}
 BwdMovement {type=number}
 BwdRightMovement {type=number}
 Gravity Override Params {
  type=GravityOverrideParamsStruct
  default {
   skydive_FloatingDown {gravityOverride=<0 -700 -100>, windAmplitude=<200 -500 200>, windOffset=<0 -600 100>}
   skydive_GoingUp {gravityOverride=<0 -300 500>, windAmplitude=<200 -500 300>}
   skydive_Parachuting {gravityOverride=<0 -500 -980>, windAmplitude=<200 -150 400>, windFrequency=<1.5 4 6>}
  }
 }
 GravityOverride {type=Vector}
 hand_rot_r {type=number}
 head_rot_bk {type=number}
 head_rot_r {type=number}
 isInFrontEndOrEmoting {type=boolean}
 LeftMovement {type=number}
 movingLeftSideAlpha {type=number}
 movingRightSideAlpha {type=number}
 pawnSpeedRange {type=number}
 pullOutAlpha {type=number}
 pushOutLeftJacketAlpha {type=number}
 RightMovement {type=number}
 SkelMesh {type=SkeletalMeshComponent}
 thigh_rot_up_01_l {type=number}
 thigh_rot_up_02_l {type=number}
 thigh_rot_up_l {type=number}
 thigh_rot_up_r {type=number}
]
