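# Physics asset for the LightningDragon female medium face-accessory ribbons, giving braid and hair strap bones secondary motion.
# Three braid chains and two hair strap chains hang from small root bodies linked to the head by free constraints.
# Braid swing limits loosen toward the tips, while straps swing wider on one axis, all with tight twist.
# Larger capsules on head, neck, upper spine, right upper arm and a hood bone provide collision shapes.
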PhysicsAsset /BRCosmetics/Characters/Player/Female/Medium/Bodies/F_MED_LightningDragon/Meshes/Parts/F_MED_LightningDragon_FaceAcc_Ribbons_Dyn
bodies {
 dyn_braid_a_l_01 [{shapeType=Capsule, length=3.5, radius=1.375, center=<4 0 0>, rotation=<rot 90 0 0>}]
 dyn_braid_a_l_02 [{shapeType=Capsule, length=3.5, radius=1.375, center=<4 0 0>, rotation=<rot 90 0 0>}]
 dyn_braid_a_l_03 [{shapeType=Capsule, length=3.5, radius=1.375, center=<4 0 0>, rotation=<rot 90 0 0>}]
 dyn_braid_a_l_MSTR [{shapeType=Capsule, length=0.505, radius=0.505, rotation=<rot 0 -90 0>}]
 dyn_braid_a_r_01 [{shapeType=Capsule, length=3.5, radius=1.375, center=<4 0 0>, rotation=<rot 90 0 0>}]
 dyn_braid_a_r_02 [{shapeType=Capsule, length=3.5, radius=1.375, center=<4 0 0>, rotation=<rot 90 0 0>}]
 dyn_braid_a_r_03 [{shapeType=Capsule, length=3.5, radius=1.375, center=<4 0 0>, rotation=<rot 90 0 0>}]
 dyn_braid_a_r_MSTR [{shapeType=Capsule, length=0.505, radius=0.505, rotation=<rot 0 -90 0>}]
 dyn_braid_b_l_01 [{shapeType=Capsule, length=3.5, radius=1.375, center=<4 0 0>, rotation=<rot 90 0 0>}]
 dyn_braid_b_l_02 [{shapeType=Capsule, length=3.5, radius=1.375, center=<4 0 0>, rotation=<rot 90 0 0>}]
 dyn_braid_b_l_03 [{shapeType=Capsule, length=3.5, radius=1.375, center=<4 0 0>, rotation=<rot 90 0 0>}]
 dyn_braid_b_l_MSTR [{shapeType=Capsule, length=0.505, radius=0.505, rotation=<rot 0 -90 0>}]
 dyn_hair_strap_a_base_l [{shapeType=Capsule, length=0.505, radius=0.505, rotation=<rot 0 -90 0>}]
 dyn_hair_strap_a_base_r [{shapeType=Capsule, length=0.505, radius=0.505, rotation=<rot 0 -90 0>}]
 dyn_hair_strap_a_l [{shapeType=Capsule, length=2.125, radius=1.375, center=<4 0 0>, rotation=<rot 90 0 0>}]
 dyn_hair_strap_a_r [{shapeType=Capsule, length=1, radius=2, center=<4 0 0>, rotation=<rot 90 0 0>}]
 dyn_hair_strap_b_l [{shapeType=Capsule, length=2.125, radius=1.625, center=<2 0 0>, rotation=<rot 90 0 0>}]
 dyn_hair_strap_b_r [{shapeType=Capsule, length=2.25, radius=2.125, center=<3 0 0>, rotation=<rot 90 0 0>}]
 dyn_hair_strap_c_l [{shapeType=Capsule, length=2.125, radius=1.625, center=<2 0 0>, rotation=<rot 90 0 0>}]
 dyn_hair_strap_c_r [{shapeType=Capsule, length=2, radius=2.375, center=<3 0 0>, rotation=<rot 82.4367 0.0001 0.0001>}]
 dyn_hood_collision [{shapeType=Capsule, length=6, radius=11.75, center=<-0.8781 1.5389 0>}]
 head [{shapeType=Capsule, length=6.8004, radius=4.591, shapeName=head_capsule, center=<2.1489 4.3282 0>, rotation=<rot 90 180 -179.455>}]
 neck_01 [{shapeType=Capsule, length=1.5504, radius=7.841, shapeName=head_capsule, center=<5.282 -2.1719 0>, rotation=<rot 90 180 159.0582>}]
 spine_05 [{shapeType=Capsule, length=14.5, radius=9.875, center=<5.1703 0.581 -0.0001>}]
 upperarm_r [{shapeType=Capsule, length=15.375, radius=5.875, shapeName=spine_05_capsule, center=<-14.3648 -0.4877 0.875>, rotation=<rot 90 -90 -90.1513>}]
}
constraints [
 {
  bone1Name=dyn_hair_strap_a_base_l
  bone2Name=head
  swing1LimitDegrees=45
  swing1Motion=Free
  swing2LimitDegrees=45
  swing2Motion=Free
  twistLimitDegrees=45
  twistMotion=Free
 }
 {
  bone1Name=dyn_hair_strap_a_base_r
  bone2Name=head
  swing1LimitDegrees=45
  swing1Motion=Free
  swing2LimitDegrees=45
  swing2Motion=Free
  twistLimitDegrees=45
  twistMotion=Free
 }
 {bone1Name=dyn_hair_strap_a_l, bone2Name=dyn_hair_strap_a_base_l, swing1LimitDegrees=15, swing2LimitDegrees=45, twistLimitDegrees=2}
 {bone1Name=dyn_hair_strap_b_l, bone2Name=dyn_hair_strap_a_l, swing1LimitDegrees=15, swing2LimitDegrees=45, twistLimitDegrees=2}
 {bone1Name=dyn_hair_strap_b_r, bone2Name=dyn_hair_strap_a_r, swing1LimitDegrees=15, swing2LimitDegrees=45, twistLimitDegrees=2}
 {bone1Name=dyn_hair_strap_a_r, bone2Name=dyn_hair_strap_a_base_r, swing1LimitDegrees=15, swing2LimitDegrees=45, twistLimitDegrees=2}
 {bone1Name=dyn_hair_strap_c_r, bone2Name=dyn_hair_strap_b_r, swing1LimitDegrees=15, swing2LimitDegrees=45, twistLimitDegrees=2}
 {
  bone1Name=head
  bone2Name=neck_01
  swing1LimitDegrees=45
  swing1Motion=Free
  swing2LimitDegrees=45
  swing2Motion=Free
  twistLimitDegrees=45
  twistMotion=Free
 }
 {
  bone1Name=dyn_braid_a_l_MSTR
  bone2Name=head
  swing1LimitDegrees=45
  swing1Motion=Free
  swing2LimitDegrees=45
  swing2Motion=Free
  twistLimitDegrees=45
  twistMotion=Free
 }
 {bone1Name=dyn_braid_a_l_01, bone2Name=dyn_braid_a_l_MSTR, swing1LimitDegrees=15, swing2LimitDegrees=15, twistLimitDegrees=2}
 {bone1Name=dyn_braid_a_l_02, bone2Name=dyn_braid_a_l_01, swing1LimitDegrees=25, swing2LimitDegrees=25, twistLimitDegrees=2}
 {bone1Name=dyn_braid_a_l_03, bone2Name=dyn_braid_a_l_02, swing1LimitDegrees=45, swing2LimitDegrees=45, twistLimitDegrees=2}
 {
  bone1Name=dyn_braid_a_r_MSTR
  bone2Name=head
  swing1LimitDegrees=45
  swing1Motion=Free
  swing2LimitDegrees=45
  swing2Motion=Free
  twistLimitDegrees=45
  twistMotion=Free
 }
 {bone1Name=dyn_braid_a_r_01, bone2Name=dyn_braid_a_r_MSTR, swing1LimitDegrees=15, swing2LimitDegrees=15, twistLimitDegrees=2}
 {bone1Name=dyn_braid_a_r_02, bone2Name=dyn_braid_a_r_01, swing1LimitDegrees=25, swing2LimitDegrees=25, twistLimitDegrees=2}
 {bone1Name=dyn_braid_a_r_03, bone2Name=dyn_braid_a_r_02, swing1LimitDegrees=45, swing2LimitDegrees=45, twistLimitDegrees=2}
 {
  bone1Name=dyn_braid_b_l_MSTR
  bone2Name=head
  swing1LimitDegrees=45
  swing1Motion=Free
  swing2LimitDegrees=45
  swing2Motion=Free
  twistLimitDegrees=45
  twistMotion=Free
 }
 {bone1Name=dyn_braid_b_l_01, bone2Name=dyn_braid_b_l_MSTR, swing1LimitDegrees=15, swing2LimitDegrees=15, twistLimitDegrees=2}
 {bone1Name=dyn_braid_b_l_02, bone2Name=dyn_braid_b_l_01, swing1LimitDegrees=25, swing2LimitDegrees=25, twistLimitDegrees=2}
 {bone1Name=dyn_braid_b_l_03, bone2Name=dyn_braid_b_l_02, swing1LimitDegrees=45, swing2LimitDegrees=45, twistLimitDegrees=2}
 {bone1Name=dyn_hair_strap_c_l, bone2Name=dyn_hair_strap_b_l, swing1LimitDegrees=15, swing2LimitDegrees=45, twistLimitDegrees=2}
 {
  bone1Name=neck_01
  bone2Name=spine_05
  swing1LimitDegrees=45
  swing1Motion=Free
  swing2LimitDegrees=45
  swing2Motion=Free
  twistLimitDegrees=45
  twistMotion=Free
 }
 {
  bone1Name=upperarm_r
  bone2Name=spine_05
  swing1LimitDegrees=45
  swing1Motion=Free
  swing2LimitDegrees=45
  swing2Motion=Free
  twistLimitDegrees=45
  twistMotion=Free
 }
 {
  bone1Name=dyn_hood_collision
  bone2Name=spine_05
  swing1LimitDegrees=45
  swing1Motion=Free
  swing2LimitDegrees=45
  swing2Motion=Free
  twistLimitDegrees=45
  twistMotion=Free
 }
]
