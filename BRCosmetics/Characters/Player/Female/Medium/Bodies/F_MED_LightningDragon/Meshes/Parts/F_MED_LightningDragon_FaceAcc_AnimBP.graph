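# Post-body AnimBP for the LightningDragon female face accessory, driving secondary motion on earrings, hair straps and hood collision.
# Copied parent and owner-mesh poses are layered at the upper arms, then Anim Dynamics swings the earrings into a cached pose.
# RigidBodyEnabled selects head-space ribbon simulation with backward-tilt offsets, or a fallback aiming straps at spine_05 with twist offsets.
# Head-versus-spine_05 quaternion twist, computed each update, supplies those offset alphas.
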AnimBlueprint /BRCosmetics/Characters/Player/Female/Medium/Bodies/F_MED_LightningDragon/Meshes/Parts/F_MED_LightningDragon_FaceAcc_AnimBP
target_skeleton=&/F_MED_LightningDragon_FaceAcc_Skeleton
graphs [
 AnimGraph {nodes [
   n1 = Get head_rot_bk() -> (head_rot_bk:double)
   n2 = "Transform (Modify) Bone - Bone: dyn_hair_strap_a_base_l"(ComponentPose=Local To Component(LocalPose=Use cached pose 'perPose'()), Alpha=@n1, Translation="-1,0,2", Rotation="50,0,0", Scale="1,1,1") {
    rotationMode=BMM_Additive
    rotationSpace=BCS_BoneSpace
    translationMode=BMM_Additive
    translationSpace=BCS_BoneSpace
    alphaBoolBlend {blendOption=Linear}
    boneToModify {boneName=dyn_hair_strap_a_base_l}
    rotation {pitch=50}
    translation {x=-1, z=2}
   }
   n3 = "Transform (Modify) Bone - Bone: dyn_hair_strap_a_base_r"(ComponentPose=@n2, Alpha=@n1, Translation="-1,0,2", Rotation="50,0,0", Scale="1,1,1") {
    rotationMode=BMM_Additive
    rotationSpace=BCS_BoneSpace
    translationMode=BMM_Additive
    translationSpace=BCS_BoneSpace
    alphaBoolBlend {blendOption=Linear}
    boneToModify {boneName=dyn_hair_strap_a_base_r}
    rotation {pitch=50}
    translation {x=-1, z=2}
   }
   n4 = "Transform (Modify) Bone - Bone: dyn_hood_collision"(ComponentPose=@n3, Alpha=@n1, Translation="-3,-4,0", Rotation="0,0,0", Scale="1,1,1") {
    rotationMode=BMM_Additive
    rotationSpace=BCS_BoneSpace
    translationMode=BMM_Additive
    translationSpace=BCS_BoneSpace
    alphaBoolBlend {blendOption=Linear}
    boneToModify {boneName=dyn_hood_collision}
    translation=<-3 -4>
   }
   n5 = RigidBody!(ComponentPose=@n4, Alpha, OverrideWorldGravity:Vector="0,0,-980", ExternalForce) -> (Pose) {
    bClampLinearTranslationLimitToRefPose=true
    bOverrideWorldGravity=true
    overridePhysicsAsset=&/F_MED_LightningDragon_FaceAcc_Ribbons_Dyn
    simulationSpace=BaseBoneSpace
    alphaBoolBlend {blendOption=Linear}
    baseBoneRef {boneName=head}
    componentAppliedLinearAccClamp=<2000 2000 2000>
    componentLinearAccScale=<0.3 0.3 0.2>
    componentLinearVelScale=<0.3 0.3 0.2>
    overrideWorldGravity {z=-980}
   }
   n6 = "Transform (Modify) Bone - Bone: dyn_hair_strap_a_base_r"(ComponentPose=@n5, Alpha=1, Translation="0,0,0.5", Rotation="0,0,0", Scale="1,1,1") {
    rotationMode=BMM_Additive
    rotationSpace=BCS_BoneSpace
    translationMode=BMM_Additive
    translationSpace=BCS_BoneSpace
    alphaBoolBlend {blendOption=Linear}
    boneToModify {boneName=dyn_hair_strap_a_base_r}
    translation {z=0.5}
   }
   n7 = Look At - Bone: dyn_hair_strap_a_r(ComponentPose=Local To Component(LocalPose=Use cached pose 'perPose'()), Alpha=1) {
    alphaBoolBlend {blendOption=Linear}
    boneToModify {boneName=dyn_hair_strap_a_r}
    lookAtLocation=<0 -31.285 1.3253>
    lookAtTarget {boneReference {boneName=spine_05}}
    lookAt_Axis {axis=<1 0>}
   }
   n8 = Look At - Bone: dyn_hair_strap_a_l(ComponentPose=@n7, Alpha=1) {
    alphaBoolBlend {blendOption=Linear}
    boneToModify {boneName=dyn_hair_strap_a_l}
    lookAtLocation=<0 -25.9032 -1.0173>
    lookAtTarget {boneReference {boneName=spine_05}}
    lookAt_Axis {axis=<1 0>}
   }
   n9 = Look At - Bone: dyn_hair_strap_b_r(ComponentPose=@n8, Alpha=1) {
    alphaBoolBlend {blendOption=Linear}
    boneToModify {boneName=dyn_hair_strap_b_r}
    lookAtLocation=<-0.0006 -24.7189 0.8355>
    lookAtTarget {boneReference {boneName=spine_05}}
    lookAt_Axis {axis=<1 0>}
   }
   n10 = Look At - Bone: dyn_hair_strap_b_l(ComponentPose=@n9, Alpha=1) {
    alphaBoolBlend {blendOption=Linear}
    boneToModify {boneName=dyn_hair_strap_b_l}
    lookAtLocation=<0.0007 -24.8106 -0.3662>
    lookAtTarget {boneReference {boneName=spine_05}}
    lookAt_Axis {axis=<1 0>}
   }
   n11 = Get head_twist_right() -> (head_twist_right:double)
   n12 = "Transform (Modify) Bone - Bone: dyn_hair_strap_b_l"(ComponentPose=@n10, Alpha=@n11, Translation="0,0,1", Rotation="0,0,45", Scale="1,1,1") {
    rotationMode=BMM_Additive
    rotationSpace=BCS_BoneSpace
    translationMode=BMM_Additive
    translationSpace=BCS_BoneSpace
    alphaBoolBlend {blendOption=Linear}
    boneToModify {boneName=dyn_hair_strap_b_l}
    rotation {roll=45}
    translation {z=1}
   }
   n13 = "Transform (Modify) Bone - Bone: dyn_hair_strap_b_r"(ComponentPose=@n12, Alpha=@n11, Translation="0,0,1.5", Rotation="0,0,45", Scale="1,1,1") {
    rotationMode=BMM_Additive
    rotationSpace=BCS_BoneSpace
    translationMode=BMM_Additive
    translationSpace=BCS_BoneSpace
    alphaBoolBlend {blendOption=Linear}
    boneToModify {boneName=dyn_hair_strap_b_r}
    rotation {roll=45}
    translation {z=1.5}
   }
   n14 = Get head_twist_left() -> (head_twist_left:double)
   n15 = "Transform (Modify) Bone - Bone: dyn_hair_strap_b_l"(ComponentPose=@n13, Alpha=@n14, Translation="0,0,2", Rotation="0,0,-45", Scale="1,1,1") {
    rotationMode=BMM_Additive
    rotationSpace=BCS_BoneSpace
    translationMode=BMM_Additive
    translationSpace=BCS_BoneSpace
    alphaBoolBlend {blendOption=Linear}
    boneToModify {boneName=dyn_hair_strap_b_l}
    rotation {roll=-45}
    translation {z=2}
   }
   n16 = "Transform (Modify) Bone - Bone: dyn_hair_strap_b_r"(ComponentPose=@n15, Alpha=@n14, Translation="0,0,2.5", Rotation="0,0,-45", Scale="1,1,1") {
    rotationMode=BMM_Additive
    rotationSpace=BCS_BoneSpace
    translationMode=BMM_Additive
    translationSpace=BCS_BoneSpace
    alphaBoolBlend {blendOption=Linear}
    boneToModify {boneName=dyn_hair_strap_b_r}
    rotation {roll=-45}
    translation {z=2.5}
   }
   Output Pose(Result=Blend Poses by bool(BlendPose_0=Component To Local(ComponentPose=@n6), BlendPose_1=Component To Local(ComponentPose=@n16), bActiveValue=Get bRigidBodyEnabled()) {blendType=Linear}) {layerGroup=None}
   n17 = Copy Pose From Mesh(SourceMeshComponent=Get SkelMesh())
   n18 = Copy Pose From Mesh() {bUseAttachedParent=true}
   n19 = Anim Dynamics - Bone: dyn_earring_l(
    ComponentPose=Local To Component(LocalPose=Layered blend per bone(BasePose=@n18, BlendPoses_0=@n17) {layerSetup [{branchFilters [{boneName=upperarm_l}, {boneName=upperarm_r}]}]})
    Alpha=1
    ExternalForce="0,0,0"
   ) {
    angularDampingOverride=0.99
    angularSpringConstant=1000
    bAngularSpring=true
    bOverrideAngularDamping=true
    bUseGravityOverride=true
    simulationSpace=RootRelative
    alphaBoolBlend {blendOption=Linear}
    boundBone {boneName=dyn_earring_l}
    componentAppliedLinearAccClamp=<5000 5000 5000>
    componentLinearAccScale=<0.3 0.3 0.3>
    componentLinearVelScale=<0.3 0.3 0.3>
    gravityOverride {z=-980}
    physicsBodyDefinitions [{
      boundBone {boneName=dyn_earring_l}
      boxExtents=<20 10 10>
      constraintSetup {angularLimitsMax=<0 15 5>, angularLimitsMin=<0 -35 -5>, angularTarget=<0 0 0>}
      localJointOffset=<5 0 0>
     }]
   }
   n20 = Anim Dynamics - Bone: dyn_earring_r(ComponentPose=@n19, Alpha=1, ExternalForce="0,0,0") {
    angularDampingOverride=0.99
    angularSpringConstant=1000
    bAngularSpring=true
    bOverrideAngularDamping=true
    bUseGravityOverride=true
    simulationSpace=RootRelative
    alphaBoolBlend {blendOption=Linear}
    boundBone {boneName=dyn_earring_r}
    componentAppliedLinearAccClamp=<5000 5000 5000>
    componentLinearAccScale=<0.3 0.3 0.3>
    componentLinearVelScale=<0.3 0.3 0.3>
    gravityOverride {z=-980}
    physicsBodyDefinitions [{
      boundBone {boneName=dyn_earring_r}
      boxExtents=<20 10 10>
      constraintSetup {angularLimitsMax=<0 15 5>, angularLimitsMin=<0 -35 -5>, angularTarget=<0 0 0>}
      localJointOffset=<5 0 0>
     }]
   }
   Save cached pose 'perPose'(Pose=Component To Local(ComponentPose=@n20)) {cacheName=perPose}
  ]}
 EventGraph {nodes [
   Event BlueprintUpdateAnimation()
   n1 = Get SkelMesh()
   > Set head_rot_bk(head_rot_bk:double=GetAlphaFromReferenceTransformDelta(SkeletalMeshComponent=@n1, BoneName=head, BaseName=spine_05, DeltaComponentType=QuaternionTwist, InputRange=(X=0,Y=60), TwistAxis="0,0,1")) -> (Output_Get:double)
   > Set head_twist_right(head_twist_right:double=GetAlphaFromReferenceTransformDelta(SkeletalMeshComponent=@n1, BoneName=head, BaseName=spine_05, DeltaComponentType=QuaternionTwist, InputRange=(X=0,Y=60), TwistAxis="1,0,0")) -> (Output_Get:double)
   > Set head_twist_left(head_twist_left:double=GetAlphaFromReferenceTransformDelta(
     SkeletalMeshComponent=@n1
     BoneName=head
     BaseName=spine_05
     DeltaComponentType=QuaternionTwist
     InputRange=(X=-60,Y=0)
     OutputRange=(X=1,Y=0)
     TwistAxis="1,0,0"
    )) -> (Output_Get:double)
   Event BlueprintInitializeAnimation()
   n2 = GetOwnerSkeletalMeshComponent()
   > Is Valid(InputObject=@n2)
   > Set SkelMesh(SkelMesh=@n2)
  ]}
]
variables [
 head_rot_bk {type=number}
 head_twist {type=Rotator}
 head_twist_left {type=number}
 head_twist_right {type=number}
 SkelMesh {type=SkeletalMeshComponent}
]
